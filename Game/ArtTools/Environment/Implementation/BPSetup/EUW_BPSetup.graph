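# Editor utility widget that creates environment Blueprints from a picked static mesh, using a blockout Blueprint as parent.
# It reads blockouts from a shared collection and narrows them by metadata tags through Category, Shape, Size and Feature controls into a toggle grid.
# Creation makes a BP_ asset in a Blueprints folder, sets the mesh, optionally clears texture data and uses Python to apply the chosen Wood, Stone or Metal resource.
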EditorUtilityWidgetBlueprint /Game/ArtTools/Environment/Implementation/BPSetup/EUW_BPSetup
graphs [
 AddSpecificMetadataTagToMap {nodes [
   n1 = Inputs!() -> (execute, BlockoutIn:Object Reference, Tag:Name, TargetMap:Map of Strings to Strings)
   > n2 = GetMetadataTag!(execute, Object:Object Reference=@n1, Tag:Name=@n1.Tag) -> (then, ReturnValue:String)
   > Add!(
    execute
    TargetMap:Map of Strings to Strings=@n1.TargetMap
    Key:"String (by ref)"="To String (Name)"(InName=@n1.Tag)
    Value:"String (by ref)"=@n2
   ) -> (then) {is=title K2Node_CallFunction}
   > Outputs!(then:Exec, BlockoutOut:Object Reference=@n1, MapOut:Map of Strings to Strings=@n1.TargetMap)
  ]}
 ClearTextureData {nodes [
   n1 = ClearTextureData!() -> (then, CDO:Object Reference) {is=title K2Node_FunctionEntry}
   > n2 = Branch(Condition=IsChecked(self=Get UI_Checkbox_KeepTD() -> (UI_Checkbox_KeepTD:Editor Utility Check Box Object Reference)))
   Return Node(execute=[@n3.then, @n2.then])
   n3 = Execute Python Script!(execute=@n2.else, PythonScript="cdo.set_editor_property('texture_data', [None, None, None, None])", CDO:Object Reference=@n1) -> (then, ReturnValue) {inputs [CDO]}
  ]}
 EnsureBPFolderExist {nodes [
   n1 = EnsureBPFolderExist!() -> (then, PackagePath:Name) {is=title K2Node_FunctionEntry}
   n2 = "To String (Name)"(InName=@n1)
   n3 = ParseIntoArray(SourceString=@n2, Delimiter=/)
   n4 = Length(TargetArray:Array of Strings=@n3)
   n5 = Replace(SourceString=@n2, From="Get (a copy)"(Array=@n3, Dimension 1=int - int(A=@n4)), To=Blueprints, SearchCase=CaseSensitive)
   > n6 = DoesDirectoryExist!(execute, DirectoryPath:String=@n5) -> (then, ReturnValue:bool)
   > n7 = Branch(Condition=@n6)
   Return Node!(execute=[@n8.then, @n7.then], BlueprintsFolderPath:String=@n5)
   n8 = MakeDirectory!(execute=@n7.else, DirectoryPath:String=@n5) -> (then, ReturnValue:bool)
  ]}
 EventGraph {nodes [
   n1 = "On Selection Changed (UI_CBox_Category)"!() -> (OutputDelegate, then, SelectedItem:String, SelectionType:ESelectInfo Enum) {is=title K2Node_ComponentBoundEvent}
   > Add(TargetMap=Get ActiveBlockoutFilter() -> (ActiveBlockoutFilter:Map of Strings to Strings), Key=Category, Value=@n1.SelectedItem)
   > RefreshAvailableBlockouts()
   > ClearChildren(self=Get UI_HorizontalBox_Shape() -> (UI_HorizontalBox_Shape:Horizontal Box Object Reference))
   > n2 = GetMetadataValues!(
    execute
    self
    AvailibleBlockoutsIn:Array of Object References=Get AvailableBlockoutsFiltered() -> (AvailableBlockoutsFiltered:Array of Object References)
    SelectionKey:Name=Shape
   ) -> (then, AvailibleValues:Array of Strings)
   > n3 = For Each Loop(Array=@n2)
   > n4 = Construct Object from Class!(
    execute
    Class=&/EUW_BPSetup_Checkbox.EUW_BPSetup_Checkbox_C
    self
    CheckboxKey:String=Shape
    CheckboxValue:String=@n3
   ) -> (then, ReturnValue:EUW BPSetup Checkbox Object Reference)
   > Bind Event to On Checkbox Changed(self:EUW BPSetup Checkbox Object Reference=@n4, Delegate=Create Event()) {is=title K2Node_AddDelegate}
   > AddChild(self=Get UI_HorizontalBox_Shape(), Content=@n4)
   > UpdateData(self:EUW BPSetup Checkbox Object Reference=@n4)
   > Add(
    TargetArray:Array of EUW BPSetup Checkbox Object References=Get ShapeCheckboxes() -> (ShapeCheckboxes:Array of EUW BPSetup Checkbox Object References)
    NewItem:"EUW BPSetup Checkbox Object Reference (by ref)"=@n4
   ) {is=title K2Node_CallArrayFunction}
   "On Clicked (UI_Button_MakeBPButton)"() {is=title K2Node_ComponentBoundEvent}
   > n5 = For Each Loop(Array:Array of EUW BPSetup Blockout Icon Toggle Object References=Get AvailableBlockoutCheckboxesFiltered() -> (AvailableBlockoutCheckboxesFiltered:Array of EUW BPSetup Blockout Icon Toggle Object References)) -> (Array Element:EUW BPSetup Blockout Icon Toggle Object Reference)
   > Branch(Condition=IsChecked(self=Get UI_Checkbox_FilteredBP!(self:EUW BPSetup Blockout Icon Toggle Object Reference=@n5) -> (UI_Checkbox_FilteredBP:Editor Utility Check Box Object Reference)))
   > MakeBlueprint!(
    execute
    self
    StaticMesh:Static Mesh Object Reference=Get StaticMeshProperty() -> (StaticMeshProperty:Static Mesh Object Reference)
    ParentClass:Object Reference=Get Asset!(self:EUW BPSetup Blockout Icon Toggle Object Reference=@n5) -> (Asset:Object Reference)
   ) -> (then)
   "On Clicked (UI_Button_ResetButton)"() {is=title K2Node_ComponentBoundEvent}
   > Clear(TargetArray:Array of EUW BPSetup Checkbox Object References=Get ShapeCheckboxes())
   > Clear(TargetArray=Get SizeCheckboxes() -> (SizeCheckboxes:Array of EUW BPSetup Checkbox Object References))
   > Clear(TargetArray=Get FeatureCheckboxes() -> (FeatureCheckboxes:Array of EUW BPSetup Checkbox Object References))
   n6 = Get UI_HorizontalBox_Shape()
   > ClearChildren(self=@n6)
   > SetIsEnabled(self=@n6, bInIsEnabled=true)
   n7 = Get UI_HorizontalBox_Size() -> (UI_HorizontalBox_Size:Horizontal Box Object Reference)
   > ClearChildren(self=@n7)
   > SetIsEnabled(self=@n7, bInIsEnabled=true)
   n8 = Get UI_HorizontalBox_Feature() -> (UI_HorizontalBox_Feature:Horizontal Box Object Reference)
   > ClearChildren(self=@n8)
   > SetIsEnabled(self=@n8, bInIsEnabled=true)
   > SetIsEnabled(bInIsEnabled=true)
   > SetIsChecked!(
    execute
    self:Check Box Object Reference=Get UI_Checkbox_AddCustomSuffix() -> (UI_Checkbox_AddCustomSuffix:Editor Utility Check Box Object Reference)
    InIsChecked:bool=false
   ) -> (then)
   n9 = Get UI_EditableText_CustomSuffix() -> (UI_EditableText_CustomSuffix:Editor Utility Editable Text Box Object Reference)
   > SetIsEnabled(self=@n9, bInIsEnabled=false)
   > "SetText (Text Box)"!(execute, self:Text Box Object Reference=@n9, InText:Text) -> (then)
   > SetIsEnabled(self=Get UI_CBox_Category() -> (UI_CBox_Category:"Searchable Combo Box (String) Object Reference"), bInIsEnabled=true)
   > SetIsChecked(self=Get UI_Checkbox_KeepTD(), InIsChecked=false)
   > n10 = For Each Loop(Array:Array of Editor Utility Check Box Object References=Get ResourceCheckboxes() -> (ResourceCheckboxes:Array of Editor Utility Check Box Object References)) -> (Array Element:Editor Utility Check Box Object Reference)
   > SetIsEnabled(self=@n10, bInIsEnabled=true)
   > Set CheckedState!(execute, CheckedState:ECheckBoxState Enum=Unchecked, self:Check Box Object Reference=@n10) -> (then, Output_Get:ECheckBoxState Enum=Unchecked)
   Add(execute=@n10.Completed, TargetMap=Get ActiveBlockoutFilter(), Key=Shape) {is=title K2Node_CallFunction}
   > Add(TargetMap=Get ActiveBlockoutFilter(), Key=Size)
   > Add(TargetMap=Get ActiveBlockoutFilter(), Key=Feature)
   > RefreshAvailableBlockouts()
   > GenerateUICheckbox!(execute, self, SelectionToGenerate:String=Shape, BlockoutToParse:Array of Object References=Get AvailableBlockoutsFiltered()) -> (then)
   n11 = "On Check State Changed (UI_Checkbox_AddCustomSuffix)"!() -> (OutputDelegate, then, bIsChecked:bool) {is=title K2Node_ComponentBoundEvent}
   n12 = Get UI_EditableText_CustomSuffix()
   > Set Text(self:Text Box Object Reference=@n12)
   > SetIsEnabled(self=@n12, bInIsEnabled=@n11.bIsChecked)
   n13 = "On Check State Changed (UI_Checkbox_Resource_Wood)"!() -> (OutputDelegate, then, bIsChecked:bool) {is=title K2Node_ComponentBoundEvent}
   > n14 = Branch(Condition=@n13.bIsChecked)
   > SetIsEnabled(self=Get UI_Checkbox_Resource_Stone() -> (UI_Checkbox_Resource_Stone:Editor Utility Check Box Object Reference), bInIsEnabled=false)
   > SetIsEnabled(self=Get UI_Checkbox_Resource_Metal() -> (UI_Checkbox_Resource_Metal:Editor Utility Check Box Object Reference), bInIsEnabled=false)
   SetIsEnabled(execute=@n14.else, self=Get UI_Checkbox_Resource_Stone(), bInIsEnabled=true)
   > SetIsEnabled(self=Get UI_Checkbox_Resource_Metal(), bInIsEnabled=true)
   n15 = "On Check State Changed (UI_Checkbox_Resource_Stone)"!() -> (OutputDelegate, then, bIsChecked:bool) {is=title K2Node_ComponentBoundEvent}
   > n16 = Branch(Condition=@n15.bIsChecked)
   > SetIsEnabled(self=Get UI_Checkbox_Resource_Wood() -> (UI_Checkbox_Resource_Wood:Editor Utility Check Box Object Reference), bInIsEnabled=false)
   > SetIsEnabled(self=Get UI_Checkbox_Resource_Metal(), bInIsEnabled=false)
   SetIsEnabled(execute=@n16.else, self=Get UI_Checkbox_Resource_Wood(), bInIsEnabled=true)
   > SetIsEnabled(self=Get UI_Checkbox_Resource_Metal(), bInIsEnabled=true)
   n17 = "On Check State Changed (UI_Checkbox_Resource_Metal)"!() -> (OutputDelegate, then, bIsChecked:bool) {is=title K2Node_ComponentBoundEvent}
   > n18 = Branch(Condition=@n17.bIsChecked)
   > SetIsEnabled(self=Get UI_Checkbox_Resource_Wood(), bInIsEnabled=false)
   > SetIsEnabled(self=Get UI_Checkbox_Resource_Stone(), bInIsEnabled=false)
   SetIsEnabled(execute=@n18.else, self=Get UI_Checkbox_Resource_Wood(), bInIsEnabled=true)
   > SetIsEnabled(self=Get UI_Checkbox_Resource_Stone(), bInIsEnabled=true)
   n19 = OnBlockoutCheckboxeStateChange!() -> (OutputDelegate, then, Checkbox State:ECheckBoxState Enum=Unchecked, Sender:EUW BPSetup Blockout Icon Toggle Object Reference) {is=title K2Node_CustomEvent}
   > n20 = Switch on ECheckBoxState!(execute, Selection:ECheckBoxState Enum=@n19.Checkbox State) -> (Unchecked:Exec, Checked:Exec, Undetermined:Exec) {is=title K2Node_SwitchEnum}
   > n21 = For Each Loop(Array:Array of EUW BPSetup Blockout Icon Toggle Object References=Get AvailableBlockoutCheckboxesFiltered()) -> (Array Element:EUW BPSetup Blockout Icon Toggle Object Reference)
   > SetIsEnabled(self=@n21, bInIsEnabled=true)
   n22 = For Each Loop(Exec=@n20.Checked, Array=Get AvailableBlockoutCheckboxesFiltered())
   > Branch(Condition="Not Equal (Object)"(A:Object Reference=@n22, B:Object Reference=@n19.Sender) -> (ReturnValue:bool=~) {is=title K2Node_PromotableOperator})
   > SetIsEnabled(self=@n22, bInIsEnabled=false)
   n23 = OnCheckboxStateChange!() -> (OutputDelegate, then, CheckboxState:ECheckBoxState Enum=Unchecked, Sender:EUW BPSetup Checkbox Object Reference, IsChecked:bool=~) {is=title K2Node_CustomEvent}
   > n24 = Switch on ECheckBoxState(Selection=@n23.CheckboxState)
   UpdateFilterFromSelection!(execute=@n24.Checked, self, Sender:EUW BPSetup Checkbox Object Reference=@n23.Sender, RemoveValue:bool=false) -> (then)
   > RefreshAvailableBlockouts()
   > n25 = Switch on String!(execute:Exec, Selection:String=Get CheckboxKey!(self:EUW BPSetup Checkbox Object Reference=@n23.Sender) -> (CheckboxKey:String)) -> (Default:Exec, Shape:Exec, Size:Exec, Feature:Exec) {pinNames [Shape,Size,Feature]}
   SetIsEnabled(execute=@n25.Feature, self=Get UI_HorizontalBox_Size(), bInIsEnabled=false)
   SetIsEnabled(execute=@n25.Shape, self=Get UI_CBox_Category(), bInIsEnabled=false)
   > GenerateUICheckbox(SelectionToGenerate=Size, BlockoutToParse=Get AvailableBlockoutsFiltered())
   > HandleExclusiveCheckboxSelection!(execute, self, IsEnabled:bool=false, Sender:EUW BPSetup Checkbox Object Reference=@n23.Sender) -> (then)
   n26 = Find(TargetMap=Get ActiveBlockoutFilter(), Key=Category)
   n27 = "Equal Exactly (String)"(A=@n26, B=Balcony)
   n28 = "Equal Exactly (String)"(A=@n26, B=BSG)
   > Branch(Condition=OR Boolean(A=OR Boolean(A=@n27, B=@n28), B="Equal Exactly (String)"(A=@n26, B=Prop)))
   n29 = "Equal Exactly (String)"(A=Find(TargetMap=Get ActiveBlockoutFilter(), Key=Category), B=Prop)
   n30 = "Not Equal Exactly (String)"(A=Find(TargetMap=Get ActiveBlockoutFilter(), Key=Shape), B=Adjustable)
   > Branch(Condition=OR Boolean(A=NOT Boolean(A=@n29), B=AND Boolean(A=@n29, B=@n30)))
   > ClearChildren(self=Get UI_HorizontalBox_Feature())
   > GenerateUICheckbox(SelectionToGenerate=Feature, BlockoutToParse=Get AvailableBlockoutsFiltered())
   SetIsEnabled(execute=@n25.Size, self=Get UI_HorizontalBox_Shape(), bInIsEnabled=false)
   > GenerateUICheckbox(SelectionToGenerate=Feature, BlockoutToParse=Get AvailableBlockoutsFiltered())
   UpdateFilterFromSelection(execute=@n24.Unchecked, Sender=@n23.Sender, RemoveValue=true)
   > RefreshAvailableBlockouts()
   > n31 = Switch on String(Selection=Get CheckboxKey(self=@n23.Sender)) {pinNames [Shape,Size,Feature]}
   SetIsEnabled(execute=@n31.Shape, self=Get UI_CBox_Category(), bInIsEnabled=true)
   > n32 = ClearChildren(self=Get UI_HorizontalBox_Size())
   HandleExclusiveCheckboxSelection(execute=[@n36.then, @n32.then], IsEnabled=true, Sender=@n23.Sender)
   n33 = Find(TargetMap=Get ActiveBlockoutFilter(), Key=Category)
   n34 = "Equal Exactly (String)"(A=@n33, B=Balcony)
   n35 = "Equal Exactly (String)"(A=@n33, B=BSG)
   > Branch(Condition=OR Boolean(A=OR Boolean(A=@n34, B=@n35), B="Equal Exactly (String)"(A=@n33, B=Prop)))
   > ClearChildren(self=Get UI_HorizontalBox_Feature())
   SetIsEnabled(execute=@n31.Size, self=Get UI_HorizontalBox_Shape(), bInIsEnabled=true)
   > n36 = ClearChildren(self=Get UI_HorizontalBox_Feature())
   Set NewVar(execute=@n31.Feature, NewVar:bool=false) -> (Output_Get:bool)
   > n37 = For Each Loop(Array:Array of EUW BPSetup Checkbox Object References=Get FeatureCheckboxes()) -> (Array Element:EUW BPSetup Checkbox Object Reference)
   n38 = Branch(execute=@n37.Completed, Condition=Get NewVar() -> (NewVar:bool))
   SetIsEnabled(execute=@n38.else, self=Get UI_HorizontalBox_Size(), bInIsEnabled=true)
   Branch(
    execute=@n37.LoopBody
    Condition=IsChecked(self=Get Checkbox!(self:EUW BPSetup Checkbox Object Reference=@n37) -> (Checkbox:Editor Utility Check Box Object Reference))
   )
   > Set NewVar(NewVar=true)
   RefreshAvailableBlockouts() {is=title K2Node_CustomEvent}
   > Clear(TargetArray:Array of Object References=Get AvailableBlockoutsFiltered())
   > n39 = FilterAvailableBlockouts!(execute, self, BlockoutsToFilter:Array of Object References=Get AvailableBlockouts() -> (AvailableBlockouts:Array of Object References)) -> (then, BlockoutFilteredOut:Array of Object References)
   > n40 = Set AvailableBlockoutsFiltered(AvailableBlockoutsFiltered:Array of Object References=@n39) -> (Output_Get:Array of Object References)
   > PopulateAvailableBPsGrid!(execute, self, FilteredBlockouts:Array of Object References=@n40, MaxColumn:int=7) -> (then)
   Event Construct()
   > Add(
    TargetArray:Array of Editor Utility Check Box Object References=Get ResourceCheckboxes()
    NewItem:"Editor Utility Check Box Object Reference (by ref)"=Get UI_Checkbox_Resource_Wood()
   ) {is=title K2Node_CallArrayFunction}
   > Add(TargetArray=Get ResourceCheckboxes(), NewItem=Get UI_Checkbox_Resource_Stone())
   > Add(TargetArray=Get ResourceCheckboxes(), NewItem=Get UI_Checkbox_Resource_Metal())
   > n41 = GetAvailibleBlockouts!(execute, self) -> (then, AvailibleBlockouts:Array of Object References)
   > n42 = Set AvailableBlockouts(AvailableBlockouts:Array of Object References=@n41) -> (Output_Get:Array of Object References)
   > n43 = GetMetadataValues(AvailibleBlockoutsIn=@n42, SelectionKey=Category)
   > n44 = For Each Loop(Array:Array of Strings=@n43) -> (Array Element:String)
   n45 = Get UI_CBox_Category()
   SetSelectedIndex!(execute=@n44.Completed, self:"Searchable Combo Box (String) Object Reference"=@n45, Index:int=0) -> (then)
   AddOption!(execute=@n44.LoopBody, self:"Searchable Combo Box (String) Object Reference"=@n45, Option:String=@n44) -> (then)
   Event PreConstruct()
   > SetSMPicker!(
    execute
    self
    PickerRef:Single Property View Object Reference=Get UI_SinglePropertyView_SM_Picker() -> (UI_SinglePropertyView_SM_Picker:Single Property View Object Reference)
   ) -> (then)
   n46 = Find(TargetMap=Get ActiveBlockoutFilter(), Key=Category)
   n47 = "Equal Exactly (String)"(A=@n46, B=Balcony)
   n48 = "Equal Exactly (String)"(A=@n46, B=BSG)
   Branch(Condition=OR Boolean(A=OR Boolean(A=@n47, B=@n48), B="Equal Exactly (String)"(A=@n46, B=Prop)))
   n49 = "Equal Exactly (String)"(A=Find(TargetMap=Get ActiveBlockoutFilter(), Key=Category), B=Prop)
   n50 = "Not Equal Exactly (String)"(A=Find(TargetMap=Get ActiveBlockoutFilter(), Key=Shape), B=Adjustable)
   > Branch(Condition=OR Boolean(A=NOT Boolean(A=@n49), B=AND Boolean(A=@n49, B=@n50)))
   > ClearChildren(self=Get UI_HorizontalBox_Feature())
   > GenerateUICheckbox(SelectionToGenerate=Feature, BlockoutToParse=Get AvailableBlockoutsFiltered())
   n51 = "For Each Loop (Map)"!(execute:Exec, MapPin:Map of Strings to Strings=Get ActiveBlockoutFilter(), BreakPin:Exec) -> (then:Exec, KeyPin:String, ValuePin:String, CompletedPin:Exec)
   > PrintString(InString=Append!(A=@n51, B=": ", C:String=@n51.ValuePin) -> (ReturnValue))
   n52 = Stringify as Int!(execute=@n51.CompletedPin, Value:int=Length(TargetArray:Array of Object References=Get AvailableBlockoutsFiltered())) -> (then, ReturnValue:String)
   > PrintString(InString=Append!(A="Availible BPs: ", B=@n52) -> (ReturnValue))
   GenerateUICheckbox(SelectionToGenerate=Feature, BlockoutToParse=Get AvailableBlockoutsFiltered())
   Get SizeBox_Size() -> (SizeBox_Size:Size Box Object Reference)
  ]}
 FilterAvailableBlockouts {nodes [
   n1 = FilterAvailableBlockouts!() -> (then, BlockoutsToFilter:Array of Object References) {is=title K2Node_FunctionEntry}
   > Clear(TargetArray=Get BlockoutFiltered() -> (BlockoutFiltered:Array of Object References))
   > n2 = For Each Loop(Array:Array of Object References=@n1) -> (Array Element:Object Reference)
   Return Node!(execute=@n2.Completed, BlockoutFilteredOut:Array of Object References=Get BlockoutFiltered())
   > Set BP Element(BP Element:Object Reference=@n2) -> (Output_Get:Object Reference)
   > n3 = GetMetadataTag(Object=Get BP Element() -> (BP Element:Object Reference), Tag=Type)
   n4 = Find(TargetMap=Get ActiveBlockoutFilter(), Key=Type)
   > Branch(Condition=OR Boolean(A="Equal Exactly (String)"(A=@n3, B=@n4), B=IsEmpty(InString=@n4)))
   > n5 = GetMetadataTag(Object=Get BP Element(), Tag=Category)
   n6 = Find(TargetMap=Get ActiveBlockoutFilter(), Key=Category)
   > Branch(Condition=OR Boolean(A="Equal Exactly (String)"(A=@n5, B=@n6), B=IsEmpty(InString=@n6)))
   > n7 = GetMetadataTag(Object=Get BP Element(), Tag=Size)
   n8 = Find(TargetMap=Get ActiveBlockoutFilter(), Key=Size)
   > Branch(Condition=OR Boolean(A="Equal Exactly (String)"(A=@n7, B=@n8), B=IsEmpty(InString=@n8)))
   > n9 = GetMetadataTag(Object=Get BP Element(), Tag=Shape)
   n10 = Find(TargetMap=Get ActiveBlockoutFilter(), Key=Shape)
   > Branch(Condition=OR Boolean(A="Equal Exactly (String)"(A=@n9, B=@n10), B=IsEmpty(InString=@n10)))
   > n11 = GetMetadataTag(Object=Get BP Element(), Tag=Feature)
   n12 = Find(TargetMap=Get ActiveBlockoutFilter(), Key=Feature)
   > n13 = Branch(Condition=OR Boolean(A="Equal Exactly (String)"(A=@n11, B=@n12), B=IsEmpty(InString=@n12)))
   > Add(TargetArray:Array of Object References=Get BlockoutFiltered(), NewItem:"Object Reference (by ref)"=Get BP Element())
   Set MatchFound(execute=@n13.else, MatchFound:bool=true) -> (Output_Get:bool)
   > n14 = GetMetadataTag(Object=Get BP Element(), Tag=Feature)
   > Set ValuesToCheck(ValuesToCheck:Array of Strings=ParseIntoArray(SourceString=@n14, Delimiter=",")) -> (Output_Get:Array of Strings)
   n15 = Get ActiveBlockoutFilter()
   > Set FilterArray(FilterArray:Array of Strings=ParseIntoArray(SourceString=Find(TargetMap=@n15, Key=Feature), Delimiter=",")) -> (Output_Get:Array of Strings)
   > n16 = For Each Loop(Array:Array of Strings=Get FilterArray() -> (FilterArray:Array of Strings)) -> (Array Element:String)
   Branch(execute=@n16.Completed, Condition=Get MatchFound() -> (MatchFound:bool))
   > Add(TargetArray=Get BlockoutFiltered(), NewItem=Get BP Element())
   n17 = Branch(
    execute=@n16.LoopBody
    Condition=Contains Item(TargetArray:Array of Strings=Get ValuesToCheck() -> (ValuesToCheck:Array of Strings), ItemToFind:"String (by ref)"=@n16)
   )
   Set MatchFound(execute=@n17.else, MatchFound=false)
  ]}
 GenerateUICheckbox {nodes [
   n1 = GenerateUICheckbox!() -> (then, SelectionToGenerate:String, BlockoutToParse:Array of Object References) {is=title K2Node_FunctionEntry}
   > n2 = Set SelectionKeyInput(SelectionKeyInput:String=@n1) -> (Output_Get:String)
   > Set BlockoutsToProcess(BlockoutsToProcess:Array of Object References=@n1.BlockoutToParse) -> (Output_Get:Array of Object References)
   > n3 = Switch on String(Selection=@n2) {pinNames [Shape,Size,Feature]}
   Set CheckboxArrayToProcess(execute=@n3.Shape, CheckboxArrayToProcess:Array of EUW BPSetup Checkbox Object References=Get ShapeCheckboxes()) -> (Output_Get:Array of EUW BPSetup Checkbox Object References)
   > n4 = Set PanelParent(PanelParent:Panel Widget Object Reference=Get UI_HorizontalBox_Shape()) -> (Output_Get:Panel Widget Object Reference)
   Clear(
    execute=[@n4.then, @n15.then, @n14.then]
    TargetArray:Array of EUW BPSetup Checkbox Object References=Get CheckboxArrayToProcess() -> (CheckboxArrayToProcess:Array of EUW BPSetup Checkbox Object References)
   )
   n5 = Get SelectionKeyInput() -> (SelectionKeyInput:String)
   > n6 = GetMetadataValues(AvailibleBlockoutsIn=Get BlockoutsToProcess() -> (BlockoutsToProcess:Array of Object References), SelectionKey=String To Name(InString=@n5))
   > n7 = For Each Loop(Array=@n6)
   > n8 = Construct Object from Class(Class=&EUW_BPSetup_Checkbox_C, CheckboxKey=@n5, CheckboxValue=@n7)
   > n9 = Set UI Element(UI Element:EUW BPSetup Checkbox Object Reference=@n8) -> (Output_Get:EUW BPSetup Checkbox Object Reference)
   > n10 = Switch on String(Selection=Get CheckboxKey(self=@n9)) {pinNames [Shape,Size,Feature]}
   n11 = Add Unique(
    execute=@n10.Shape
    TargetArray:Array of EUW BPSetup Checkbox Object References=Get ShapeCheckboxes()
    NewItem:"EUW BPSetup Checkbox Object Reference (by ref)"=Get UI Element() -> (UI Element:EUW BPSetup Checkbox Object Reference)
   )
   Bind Event to On Checkbox Changed(execute=[@n11.then, @n13.then, @n12.then], self=Get UI Element(), Delegate=Create Event())
   > AddChild(self=Get PanelParent() -> (PanelParent:Panel Widget Object Reference), Content=Get UI Element())
   > UpdateData(self=Get UI Element())
   n12 = Add Unique(execute=@n10.Size, TargetArray=Get SizeCheckboxes(), NewItem=Get UI Element())
   n13 = Add Unique(execute=@n10.Feature, TargetArray=Get FeatureCheckboxes(), NewItem=Get UI Element())
   Set CheckboxArrayToProcess(execute=@n3.Feature, CheckboxArrayToProcess=Get FeatureCheckboxes())
   > n14 = Set PanelParent(PanelParent=Get UI_HorizontalBox_Feature())
   Set CheckboxArrayToProcess(execute=@n3.Size, CheckboxArrayToProcess=Get SizeCheckboxes())
   > n15 = Set PanelParent(PanelParent=Get UI_HorizontalBox_Size())
  ]}
 GetAvailibleBlockouts {nodes [
   GetAvailibleBlockouts() {is=title K2Node_FunctionEntry}
   > n1 = GetAssetsInCollection!(
    execute
    self:Collection Manager Subsystem Object Reference=Get CollectionManagerScriptingSubsystem()
    Collection_Container:Name=Game
    Collection_Name:Name=Get BlockoutsCollectionName() -> (BlockoutsCollectionName:Name=None)
    Collection_ShareType:ECollectionScriptingShareType Enum=Shared
   ) -> (then, OutAssets:Array of Asset Data Structures, ReturnValue:bool)
   > Branch(Condition=@n1.ReturnValue)
   > n2 = For Each Loop(Array:Array of Asset Data Structures=@n1) -> (Array Element:Asset Data Structure)
   Return Node!(execute=@n2.Completed, AvailibleBlockouts:Array of Object References=Get Blockouts() -> (Blockouts:Array of Object References))
   > n3 = Set LoopedCollectionAsset(LoopedCollectionAsset:Object Reference=GetAsset(InAssetData=@n2)) -> (Output_Get:Object Reference)
   > n4 = GetMetadataTag(Object=@n3, Tag=Type)
   > Branch(Condition="Equal Exactly (String)"(A=@n4, B=Blockout))
   > Add Unique(
    TargetArray:Array of Object References=Get Blockouts()
    NewItem:"Object Reference (by ref)"=Get LoopedCollectionAsset() -> (LoopedCollectionAsset:Object Reference)
   )
  ]}
 GetMetadataValues {nodes [
   n1 = GetMetadataValues!() -> (then, AvailibleBlockoutsIn:Array of Object References, SelectionKey:Name) {is=title K2Node_FunctionEntry}
   > Set SelectedKey(SelectedKey:Name=@n1.SelectionKey) -> (Output_Get:Name=None)
   > Clear(TargetArray:Array of Strings=Get AvailibleMetadataValues() -> (AvailibleMetadataValues:Array of Strings))
   > n2 = For Each Loop(Array:Array of Object References=@n1) -> (Array Element:Object Reference)
   Return Node!(execute=@n2.Completed, AvailibleValues:Array of Strings=Get AvailibleMetadataValues())
   > n3 = GetMetadataTag(Object=@n2, Tag=Get SelectedKey() -> (SelectedKey:Name=None))
   > Branch(Condition="Not Equal Exactly (String)"(A=@n3))
   > n4 = Branch(Condition=Contains(SearchIn=@n3, Substring=","))
   > n5 = For Each Loop(Array:Array of Strings=ParseIntoArray(SourceString=@n3, Delimiter=",")) -> (Array Element:String)
   > Add Unique(TargetArray:Array of Strings=Get AvailibleMetadataValues(), NewItem:"String (by ref)"=@n5)
   Add Unique(execute=@n4.else, TargetArray=Get AvailibleMetadataValues(), NewItem=@n3)
  ]}
 HandleExclusiveCheckboxSelection {nodes [
   n1 = HandleExclusiveCheckboxSelection!() -> (then, IsEnabled:bool=~, Sender:EUW BPSetup Checkbox Object Reference) {is=title K2Node_FunctionEntry}
   > Set CheckboxEnabled(CheckboxEnabled:bool=@n1) -> (Output_Get:bool)
   > Set SenderToHandle(SenderToHandle:EUW BPSetup Checkbox Object Reference=@n1.Sender) -> (Output_Get:EUW BPSetup Checkbox Object Reference)
   > n2 = Switch on String(Selection=Get CheckboxKey(self=@n1.Sender)) {pinNames [Shape,Size,Feature]}
   n3 = Set CheckboxArrayToProcess(execute=@n2.Feature, CheckboxArrayToProcess=Get FeatureCheckboxes())
   n4 = For Each Loop(Exec=[@n6.then, @n3.then, @n5.then], Array:Array of EUW BPSetup Checkbox Object References=Get CheckboxArrayToProcess()) -> (Array Element:EUW BPSetup Checkbox Object Reference)
   > Branch(Condition="Not Equal (Object)"(A=@n4, B=Get SenderToHandle() -> (SenderToHandle:EUW BPSetup Checkbox Object Reference)) -> (ReturnValue=~))
   > SetIsEnabled(self=@n4, bInIsEnabled=Get CheckboxEnabled() -> (CheckboxEnabled:bool))
   n5 = Set CheckboxArrayToProcess(execute=@n2.Size, CheckboxArrayToProcess=Get SizeCheckboxes())
   n6 = Set CheckboxArrayToProcess(execute=@n2.Shape, CheckboxArrayToProcess=Get ShapeCheckboxes())
  ]}
 MakeBPPath {nodes [
   n1 = "GetText (Text Box)"!(self:Text Box Object Reference=Get UI_EditableText_CustomSuffix()) -> (ReturnValue:Text)
   n2 = Inputs!() -> (BPFolderPath:String, SMName:Name)
   n3 = Replace(SourceString="To String (Name)"(InName=@n2.SMName), From=SM_, To=BP_, SearchCase=CaseSensitive)
   n4 = Append(A=@n2, B=/)
   Outputs!(BPPath:String=Append(A=Append(A=@n4, B=@n3), B="To String (Text)"(InText=@n1)))
  ]}
 MakeBlueprint {nodes [
   n1 = MakeBlueprint!() -> (then, StaticMesh:Static Mesh Object Reference, ParentClass:Object Reference) {is=title K2Node_FunctionEntry}
   > Set StaticMeshIn(StaticMeshIn:Static Mesh Object Reference=@n1) -> (Output_Get:Static Mesh Object Reference)
   > Set ParentClassIn(ParentClassIn:Object Reference=@n1.ParentClass) -> (Output_Get:Object Reference)
   > n2 = ValidateStaticMeshSelection!(execute, self) -> (then, IsNotValid:bool)
   > n3 = Branch(Condition=@n2)
   > Show Message Dialog(
    Title=Get StaticMeshSelectionWarning() -> (StaticMeshSelectionWarning:Text)
    Message=Get StaticMeshSelectionMessage() -> (StaticMeshSelectionMessage:Text)
    MessageType=Ok
    DefaultValue=No
    MessageCategory=Warning
   )
   n4 = ValidateResourceSelection!(execute=@n3.else, self) -> (then, NothingIsSelected:bool)
   > n5 = Branch(Condition=@n4)
   > Show Message Dialog(
    Title=Get ResourceSelectionWarningTitle() -> (ResourceSelectionWarningTitle:Text)
    Message=Get ResourceSelectionWarningMessage() -> (ResourceSelectionWarningMessage:Text)
    MessageType=Ok
    DefaultValue=No
    MessageCategory=Warning
   )
   n6 = Get Asset By Object Path!(
    execute=@n5.else
    self:Asset Registry Interface=GetAssetRegistry()
    ObjectPath:"Soft Object Path Structure (by ref)"=GetSoftObjectPath(Object=Get StaticMeshIn() -> (StaticMeshIn:Static Mesh Object Reference))
    bIncludeOnlyOnDiskAssets:bool=false
    bSkipARFilteredAssets:bool=true
   ) -> (
    then
    ReturnValue_PackageName:Name
    ReturnValue_PackagePath:Name
    ReturnValue_AssetName:Name
    ReturnValue_AssetClassPath:Top Level Asset Path Structure
   ) {is=title K2Node_CallFunction}
   > n7 = EnsureBPFolderExist!(execute, self, PackagePath:Name=@n6.ReturnValue_PackagePath) -> (then, BlueprintsFolderPath:String) {is=title K2Node_CallFunction}
   > n8 = GetPathNameForLoadedAsset(LoadedAsset=Get ParentClassIn() -> (ParentClassIn:Object Reference))
   > n9 = LoadBlueprintClass!(execute, AssetPath:String=@n8) -> (then, ReturnValue:Object Class Reference)
   n10 = Make BPPath!(BPFolderPath:String=@n7, SMName:Name=@n6.ReturnValue_AssetName) -> (BPPath:String) {is=title K2Node_MacroInstance}
   > n11 = DoesAssetExist(AssetPath=@n10)
   > n12 = Branch(Condition=@n11)
   > Show Message Dialog(
    Title=Get BPCreationWarning() -> (BPCreationWarning:Text)
    Message=Get BPCreationWarningMessage() -> (BPCreationWarningMessage:Text)
    MessageType=Ok
    DefaultValue=No
    MessageCategory=Warning
   )
   n13 = CreateBlueprintAssetWithParent!(execute=@n12.else, AssetPath:String=@n10, ParentClass:Object Class Reference=@n9) -> (then, ReturnValue:Blueprint Object Reference)
   > n14 = GeneratedClass!(execute, BlueprintObj:Blueprint Object Reference=@n13) -> (then, ReturnValue:Object Class Reference)
   > n15 = GetClassDefaultObject!(execute, Class:Object Class Reference=@n14) -> (then, ReturnValue:Object Reference)
   > n16 = Set BP_CDO(BP_CDO:Object Reference=@n15) -> (Output_Get:Object Reference)
   > SetEditorProperty(Object=@n16, PropertyName=StaticMesh, PropertyValue:Static Mesh Object Reference=Get StaticMeshIn())
   > ClearTextureData!(execute, self, CDO:Object Reference=Get BP_CDO() -> (BP_CDO:Object Reference)) -> (then) {is=title K2Node_CallFunction}
   > SetBlueprintResource!(execute, self, BP:Object Reference=Get BP_CDO()) -> (then)
   > SaveAsset!(execute, AssetToSave:String=@n10, bOnlyIfIsDirty:bool=true) -> (then, ReturnValue:bool)
   > Return Node!(execute)
  ]}
 PopulateAvailableBPsGrid {nodes [
   n1 = PopulateAvailableBPsGrid!() -> (then, FilteredBlockouts:Array of Object References, MaxColumn:int=7) {is=title K2Node_FunctionEntry}
   > Clear(TargetArray:Array of EUW BPSetup Blockout Icon Toggle Object References=Get AvailableBlockoutCheckboxesFiltered())
   > Set MaxColumnIn(MaxColumnIn:int=@n1.MaxColumn) -> (Output_Get:int=0)
   > Set CurrentColumn(CurrentColumn:int=0) -> (Output_Get:int=0)
   > Set CurrentRow(CurrentRow:int=0) -> (Output_Get:int=0)
   > ClearChildren(self=Get UI_GridPanel_AvailableBPs() -> (UI_GridPanel_AvailableBPs:Uniform Grid Panel Object Reference))
   > n2 = For Each Loop(Array:Array of Object References=@n1) -> (Array Element:Object Reference)
   > Set Element(Element:Object Reference=@n2) -> (Output_Get:Object Reference)
   > n3 = Create Widget!(
    execute:Exec
    WorldContextObject:Object Reference
    Class:User Widget Class Reference=&/EUW_BPSetup_BlockoutIconToggle.EUW_BPSetup_BlockoutIconToggle_C
    OwningPlayer:Player Controller Object Reference
    Asset:Object Reference=Get Element() -> (Element:Object Reference)
   ) -> (then:Exec, ReturnValue:EUW BPSetup Blockout Icon Toggle Object Reference)
   > n4 = Set NewFilterCheckbox(NewFilterCheckbox:EUW BPSetup Blockout Icon Toggle Object Reference=@n3) -> (Output_Get:EUW BPSetup Blockout Icon Toggle Object Reference)
   > Bind Event to On BPCheckbox Change(self:EUW BPSetup Blockout Icon Toggle Object Reference=@n4, Delegate=Create Event()) {is=title K2Node_AddDelegate}
   > Add Unique(
    TargetArray:Array of EUW BPSetup Blockout Icon Toggle Object References=Get AvailableBlockoutCheckboxesFiltered()
    NewItem:"EUW BPSetup Blockout Icon Toggle Object Reference (by ref)"=Get NewFilterCheckbox() -> (NewFilterCheckbox:EUW BPSetup Blockout Icon Toggle Object Reference)
   )
   > AddChildToUniformGrid!(
    execute
    self:Uniform Grid Panel Object Reference=Get UI_GridPanel_AvailableBPs()
    Content:Widget Object Reference=Get NewFilterCheckbox()
    InRow:int=Get CurrentRow() -> (CurrentRow:int=0)
    InColumn:int=Get CurrentColumn() -> (CurrentColumn:int=0)
   ) -> (then, ReturnValue:Uniform Grid Slot Object Reference)
   n5 = int + int(A=Get CurrentColumn() -> (CurrentColumn=0))
   > Set CurrentColumn(CurrentColumn=@n5) -> (Output_Get=0)
   > Branch(Condition="integer > integer"(A=@n5, B=Get MaxColumnIn() -> (MaxColumnIn:int=0)))
   > Set CurrentColumn(CurrentColumn=0) -> (Output_Get=0)
   > Set CurrentRow(CurrentRow=int + int(A=Get CurrentRow() -> (CurrentRow=0))) -> (Output_Get=0)
  ]}
 SetBlueprintResource {nodes [
   n1 = SetBlueprintResource!() -> (then, BP:Object Reference) {is=title K2Node_FunctionEntry}
   > Set CDO(CDO:Object Reference=@n1) -> (Output_Get:Object Reference)
   > n2 = For Each Loop(Array:Array of Editor Utility Check Box Object References=Get ResourceCheckboxes()) -> (Array Element:Editor Utility Check Box Object Reference)
   > Branch(Condition=IsChecked(self=@n2))
   > Set SelectedCheckbox(SelectedCheckbox:Editor Utility Check Box Object Reference=@n2) -> (Output_Get:Editor Utility Check Box Object Reference)
   n3 = Split(
    SourceString=GetDisplayName(Object=Get SelectedCheckbox() -> (SelectedCheckbox:Editor Utility Check Box Object Reference))
    InStr=_
    SearchDir=FromEnd
   )
   Set SelectedResource(execute=@n2.Completed, SelectedResource:String=@n3.RightS) -> (Output_Get:String)
   > PrintString(InString=@n3.RightS)
   > Execute Python Script!(
    execute
    PythonScript="import unreal\r\n\r\n##--- Set ResourceType ---##\r\nresourceTypeEnum = bp_cdo.get_editor_property(\"ResourceType\")\r\nkey = selected_resource.upper()\r\nenum_value = getattr(resourceTypeEnum, key)\r\nbp_cdo.set_editor_property(\"ResourceType\", enum_value)\r\n\r\n##--- Set ResourceAtt ---##\r\nattInitKeys = bp_cdo.get_editor_property(\"AttributeInitKeys\")\r\ncurrentCategoryValue = str(attInitKeys[0].get_editor_property(\"attribute_init_category\"))\r\nPREFIXES = (\"Building\", \"Resource\", \"Prop\")\r\nprefix = next((p for p in PREFIXES if currentCategoryValue.startswith(p)), currentCategoryValue)\r\nNewCategoryValue = prefix + selected_resource\r\ncurrentSubCategoryValue = attInitKeys[0].get_editor_property(\"attribute_init_sub_category\")\r\nattInitKeys[0] = unreal.FortAttributeInitializationKey(NewCategoryValue, currentSubCategoryValue)\r\nattInitKeys[1] = unreal.FortAttributeInitializationKey(NewCategoryValue, currentSubCategoryValue)\r\n"
    bp_cdo:Object Reference=Get CDO() -> (CDO:Object Reference)
    selected_resource:String=Get SelectedResource() -> (SelectedResource:String)
   ) -> (then, ReturnValue) {inputs [bp_cdo, selected_resource]}
  ]}
 SetSMPicker {nodes [
   n1 = SetSMPicker!() -> (then, PickerRef:Single Property View Object Reference) {is=title K2Node_FunctionEntry}
   > SetPropertyName!(execute, self:Single Property View Object Reference=@n1, NewPropertyName:Name=StaticMeshProperty) -> (then)
   > SetObject(self=@n1, NewObject=Self-Reference())
  ]}
 UpdateFilterFromSelection {nodes [
   n1 = UpdateFilterFromSelection!() -> (then, Sender:EUW BPSetup Checkbox Object Reference, RemoveValue:bool=~) {is=title K2Node_FunctionEntry}
   > Set Sender Key(Sender Key:String=Get CheckboxKey(self=@n1)) -> (Output_Get:String)
   > Set Sender Value(Sender Value:String=Get CheckboxValue!(self:EUW BPSetup Checkbox Object Reference=@n1) -> (CheckboxValue:String)) -> (Output_Get:String)
   > n2 = Branch(Condition=@n1.RemoveValue)
   n3 = Get ActiveBlockoutFilter()
   n4 = Get Sender Key() -> (Sender Key:String)
   > n5 = Set NewValue(NewValue:Array of Strings=ParseIntoArray(SourceString=Find(TargetMap=@n3, Key=@n4), Delimiter=",")) -> (Output_Get:Array of Strings)
   > n6 = Branch(Condition="integer > integer"(A=Length(TargetArray:Array of Strings=@n5), B=1))
   > Remove Item!(execute, TargetArray=Get NewValue() -> (NewValue:Array of Strings), Item:"String (by ref)"=Get Sender Value() -> (Sender Value:String)) -> (then, ReturnValue:bool)
   > Add(TargetMap=Get ActiveBlockoutFilter(), Key=Get Sender Key(), Value=JoinStringArray(SourceArray=Get NewValue(), Separator=",")) {is=title K2Node_CallFunction}
   Add(execute=@n6.else, TargetMap=Get ActiveBlockoutFilter(), Key=Get Sender Key())
   n7 = Branch(execute=@n2.else, Condition="Equal Exactly (String)"(A=Get Sender Key(), B=Feature))
   n8 = ParseIntoArray(SourceString=Find(TargetMap=Get ActiveBlockoutFilter(), Key=Get Sender Key()), Delimiter=",")
   > n9 = Branch(Condition="integer >= integer"(A=Length(TargetArray=@n8), B=1))
   n10 = Get ActiveBlockoutFilter()
   n11 = Get Sender Key()
   > Add(
    TargetMap=Get ActiveBlockoutFilter()
    Key=Get Sender Key()
    Value=Append!(A=Find(TargetMap=@n10, Key=@n11), B=",", C:String=Get Sender Value()) -> (ReturnValue)
   )
   Add(execute=@n9.else, TargetMap=Get ActiveBlockoutFilter(), Key=Get Sender Key(), Value=Get Sender Value())
   Add(execute=@n7.else, TargetMap=Get ActiveBlockoutFilter(), Key=Get Sender Key(), Value=Get Sender Value())
  ]}
 ValidateResourceSelection {nodes [
   ValidateResourceSelection() {is=title K2Node_FunctionEntry}
   > Set NothingSelected(NothingSelected:bool=true) -> (Output_Get:bool)
   > n1 = For Each Loop(Array=Get ResourceCheckboxes())
   Return Node!(execute=@n1.Completed, NothingIsSelected:bool=Get NothingSelected() -> (NothingSelected:bool))
   > Branch(Condition=IsChecked(self=@n1))
   > Set NothingSelected(NothingSelected=false)
  ]}
 ValidateStaticMeshSelection {nodes [
   ValidateStaticMeshSelection() {is=title K2Node_FunctionEntry}
   n1 = Get StaticMeshProperty()
   > Return Node!(execute, IsNotValid:bool="Not Equal (Boolean)"(A=IsValid(Object=@n1), B=true) -> (ReturnValue=~))
  ]}
]
variables [
 ActiveBlockoutFilter {type="map<string>", default {Category="", Feature="", Shape="", Size="", Type=Blockout}}
 AvailableBlockoutCheckboxesFiltered {type="array<EUW_BPSetup_BlockoutIconToggle_C>", default []}
 AvailableBlockouts {type="array<Object>", default []}
 AvailableBlockoutsFiltered {type="array<Object>", default []}
 AvailibleFeatureBlockouts {type="array<Object>", default []}
 BlockoutsCollectionName {type=string, default=ENV_Blockouts}
 Checkbox Array {type=EUW_BPSetup_Checkbox_C, default=null}
 FeatureCheckboxes {type="array<EUW_BPSetup_Checkbox_C>", default [null]}
 NewVar {type=boolean}
 ResourceCheckboxes {type="array<EditorUtilityCheckBox>", default []}
 Selection Type {type=ESelectInfo, default=OnKeyPress}
 SenderKey {type=string, default=""}
 ShapeCheckboxes {type="array<EUW_BPSetup_Checkbox_C>", default []}
 SizeCheckboxes {type="array<EUW_BPSetup_Checkbox_C>", default [null]}
 StaticMeshProperty {type=StaticMesh, default=null}
 Title {type=string, default=""}
 UI Checkbox Resource Wood {type=EditorUtilityCheckBox, default=null}
]
widgets {
 info {rootWidgetClass=&VerticalBox, widgetCount=53}
 widgets [
  VerticalBox {widgetClassPath=&VerticalBox, widget=VerticalBox, properties {bOverride_Cursor=true}}
  SizeBox_SM_Picker {
   widgetClassPath=&SizeBox
   parent=VerticalBox
   widget=SizeBox_SM_Picker
   properties {bOverride_WidthOverride=true, heightOverride=200, widthOverride=500}
   slot {class=VerticalBoxSlot, properties {horizontalAlignment=HAlign_Center}}
  }
  Border_SM_Picker {
   widgetClassPath=&Border
   parent=SizeBox_SM_Picker
   widget=Border_SM_Picker
   properties {
    background {
     bIsDynamicallyLoaded=false
     drawAs=RoundedBox
     imageType=NoImage
     mirroring=NoMirror
     resourceName=None
     resourceObject=null
     tiling=NoTile
     imageSize=<32 32>
     margin=<ltrb 0 0 0 0>
     outlineSettings {
      bUseBrushTransparency=false
      roundingType=FixedRadius
      width=1
      color {colorUseRule=UseColor_Specified, specifiedColor=<rgba 0 0.7126 1 1>}
      cornerRadii=<5 5 5 5>
     }
     tintColor {colorUseRule=UseColor_Specified, specifiedColor=<rgba 0.0194 0.0194 0.0194 1>}
     uVRegion {bIsValid=false, max=<0 0>, min=<0 0>}
    }
    padding=<ltrb 5 5 5 5>
   }
   slot {class=SizeBoxSlot}
  }
  UI_SinglePropertyView_SM_Picker {
   widgetClassPath=&SinglePropertyView
   parent=Border_SM_Picker
   widget=UI_SinglePropertyView_SM_Picker
   properties {nameOverride=" ", propertyName=StaticMeshProperty}
   slot {class=BorderSlot, properties {padding=<ltrb 5 5 5 5>}}
  }
  SizeBox_Category {
   widgetClassPath=&SizeBox
   parent=VerticalBox
   widget=SizeBox_Category
   properties {bOverride_WidthOverride=true, heightOverride=200, widthOverride=500}
   slot {class=VerticalBoxSlot, properties {horizontalAlignment=HAlign_Center, padding=<ltrb 5 5 5 5>}}
  }
  Border_Category {
   widgetClassPath=&Border
   parent=SizeBox_Category
   widget=Border_Category
   properties {
    horizontalAlignment=HAlign_Center
    background {
     bIsDynamicallyLoaded=false
     drawAs=RoundedBox
     imageType=NoImage
     mirroring=NoMirror
     resourceName=None
     resourceObject=null
     tiling=NoTile
     imageSize=<32 32>
     margin=<ltrb 0 0 0 0>
     outlineSettings {
      bUseBrushTransparency=false
      roundingType=FixedRadius
      width=1
      color {colorUseRule=UseColor_Specified, specifiedColor=<rgba 0 0.7126 1 1>}
      cornerRadii=<5 5 5 5>
     }
     tintColor {colorUseRule=UseColor_Specified, specifiedColor=<rgba 0.0194 0.0194 0.0194 1>}
     uVRegion {bIsValid=false, max=<0 0>, min=<0 0>}
    }
    padding=<ltrb 5 5 5 5>
   }
   slot {class=SizeBoxSlot}
  }
  HorizontalBox_Category {
   widgetClassPath=&HorizontalBox
   parent=Border_Category
   widget=HorizontalBox_Category
   slot {class=BorderSlot, properties {horizontalAlignment=HAlign_Center, padding=<ltrb 5 5 5 5>}}
  }
  Text_CBox_Category {
   widgetClassPath=&TextBlock
   parent=HorizontalBox_Category
   widget=Text_CBox_Category
   properties {
    text=Category :
    font {
     bForceMonospaced=false
     bMaterialIsStencil=false
     fontMaterial=null
     fontObject=&Roboto
     letterSpacing=0
     monospacedWidth=1
     size=13.5
     skewAmount=0
     typefaceFontName=Bold
     outlineSettings {
      bApplyOutlineToDropShadows=false
      bMiteredCorners=false
      bSeparateFillAlpha=false
      outlineMaterial=null
      outlineSize=0
      outlineColor=<rgba 0 0 0 1>
     }
    }
   }
   slot {class=HorizontalBoxSlot, properties {horizontalAlignment=HAlign_Center, verticalAlignment=VAlign_Center, padding=<ltrb 10 0 0 0>}}
  }
  SizeBox_CBox_Category {
   widgetClassPath=&SizeBox
   parent=HorizontalBox_Category
   widget=SizeBox_CBox_Category
   properties {bOverride_WidthOverride=true, widthOverride=249.9472}
   slot {class=HorizontalBoxSlot, properties {padding=<ltrb 10 0 0 0>, size {sizeRule=Fill, value=1}}}
  }
  UI_CBox_Category {
   widgetClassPath=/Script/FortniteEditor.SearchableStringComboBox
   bIsVariable=true
   parent=SizeBox_CBox_Category
   widget=UI_CBox_Category
   properties {
    font {
     bForceMonospaced=false
     bMaterialIsStencil=false
     fontMaterial=null
     fontObject=&Roboto
     letterSpacing=0
     monospacedWidth=1
     size=13.5
     skewAmount=0
     typefaceFontName=Bold
     outlineSettings {
      bApplyOutlineToDropShadows=false
      bMiteredCorners=false
      bSeparateFillAlpha=false
      outlineMaterial=null
      outlineSize=0
      outlineColor=<rgba 0 0 0 1>
     }
    }
    foregroundColor {colorUseRule=UseColor_Specified, specifiedColor=<rgba 0.5271 0.5271 0.5271 1>}
    itemStyle {
     bUseParentRowBrush=false
     activeBrush {
      bIsDynamicallyLoaded=false
      drawAs=Image
      imageType=NoImage
      mirroring=NoMirror
      resourceName=None
      resourceObject=null
      tiling=NoTile
      imageSize=<0 0>
      margin=<ltrb 0 0 0 0>
      outlineSettings {
       bUseBrushTransparency=false
       roundingType=HalfHeightRadius
       width=0
       color {colorUseRule=UseColor_Specified, specifiedColor=<rgba 0 0 0 0>}
       cornerRadii=<0 0 0 0>
      }
      tintColor {colorUseRule=UseColor_Specified, specifiedColor=<rgba 0 0.162 0.7454 1>}
      uVRegion {bIsValid=false, max=<0 0>, min=<0 0>}
     }
     activeHighlightedBrush {
      bIsDynamicallyLoaded=false
      drawAs=Image
      imageType=NoImage
      mirroring=NoMirror
      resourceName=None
      resourceObject=null
      tiling=NoTile
      imageSize=<0 0>
      margin=<ltrb 0 0 0 0>
      outlineSettings {
       bUseBrushTransparency=false
       roundingType=HalfHeightRadius
       width=0
       color {colorUseRule=UseColor_Specified, specifiedColor=<rgba 0 0 0 0>}
       cornerRadii=<0 0 0 0>
      }
      tintColor {colorUseRule=UseColor_Specified, specifiedColor=<rgba 0.0252 0.0319 0.0423 1>}
      uVRegion {bIsValid=false, max=<0 0>, min=<0 0>}
     }
     activeHoveredBrush {
      bIsDynamicallyLoaded=false
      drawAs=Image
      imageType=NoImage
      mirroring=NoMirror
      resourceName=None
      resourceObject=null
      tiling=NoTile
      imageSize=<0 0>
      margin=<ltrb 0 0 0 0>
      outlineSettings {
       bUseBrushTransparency=false
       roundingType=HalfHeightRadius
       width=0
       color {colorUseRule=UseColor_Specified, specifiedColor=<rgba 0 0 0 0>}
       cornerRadii=<0 0 0 0>
      }
      tintColor {colorUseRule=UseColor_Specified, specifiedColor=<rgba 0 0.162 0.7454 1>}
      uVRegion {bIsValid=false, max=<0 0>, min=<0 0>}
     }
     dropIndicator_Above {
      bIsDynamicallyLoaded=false
      drawAs=Box
      imageType=FullColor
      mirroring=NoMirror
      resourceName=../../../Engine/Content/Slate/Common/DropZoneIndicator_Above.png
      resourceObject=null
      tiling=NoTile
      imageSize=<0 0>
      margin=<ltrb 0.625 0.625 0 0>
      outlineSettings {
       bUseBrushTransparency=false
       roundingType=HalfHeightRadius
       width=0
       color {colorUseRule=UseColor_Specified, specifiedColor=<rgba 0 0 0 0>}
       cornerRadii=<0 0 0 0>
      }
      tintColor {colorUseRule=UseColor_Specified, specifiedColor=<rgba 0.0091 0.3515 0.9647 1>}
      uVRegion {bIsValid=false, max=<0 0>, min=<0 0>}
     }
     dropIndicator_Below {
      bIsDynamicallyLoaded=false
      drawAs=Box
      imageType=FullColor
      mirroring=NoMirror
      resourceName=../../../Engine/Content/Slate/Common/DropZoneIndicator_Below.png
      resourceObject=null
      tiling=NoTile
      imageSize=<0 0>
      margin=<ltrb 0.625 0 0 0.625>
      outlineSettings {
       bUseBrushTransparency=false
       roundingType=HalfHeightRadius
       width=0
       color {colorUseRule=UseColor_Specified, specifiedColor=<rgba 0 0 0 0>}
       cornerRadii=<0 0 0 0>
      }
      tintColor {colorUseRule=UseColor_Specified, specifiedColor=<rgba 0.0091 0.3515 0.9647 1>}
      uVRegion {bIsValid=false, max=<0 0>, min=<0 0>}
     }
     dropIndicator_Onto {
      bIsDynamicallyLoaded=false
      drawAs=Box
      imageType=FullColor
      mirroring=NoMirror
      resourceName=../../../Engine/Content/Slate/Common/DropZoneIndicator_Onto.png
      resourceObject=null
      tiling=NoTile
      imageSize=<0 0>
      margin=<ltrb 0.25 0.25 0.25 0.25>
      outlineSettings {
       bUseBrushTransparency=false
       roundingType=HalfHeightRadius
       width=0
       color {colorUseRule=UseColor_Specified, specifiedColor=<rgba 0 0 0 0>}
       cornerRadii=<0 0 0 0>
      }
      tintColor {colorUseRule=UseColor_Specified, specifiedColor=<rgba 0.0091 0.3515 0.9647 1>}
      uVRegion {bIsValid=false, max=<0 0>, min=<0 0>}
     }
     evenRowBackgroundBrush {
      bIsDynamicallyLoaded=false
      drawAs=Image
      imageType=NoImage
      mirroring=NoMirror
      resourceName=None
      resourceObject=null
      tiling=NoTile
      imageSize=<0 0>
      margin=<ltrb 0 0 0 0>
      outlineSettings {
       bUseBrushTransparency=false
       roundingType=HalfHeightRadius
       width=0
       color {colorUseRule=UseColor_Specified, specifiedColor=<rgba 0 0 0 0>}
       cornerRadii=<0 0 0 0>
      }
      tintColor {colorUseRule=UseColor_Specified, specifiedColor=<rgba 0.0103 0.0103 0.0103 1>}
      uVRegion {bIsValid=false, max=<0 0>, min=<0 0>}
     }
     evenRowBackgroundHoveredBrush {
      bIsDynamicallyLoaded=false
      drawAs=Image
      imageType=NoImage
      mirroring=NoMirror
      resourceName=None
      resourceObject=null
      tiling=NoTile
      imageSize=<0 0>
      margin=<ltrb 0 0 0 0>
      outlineSettings {
       bUseBrushTransparency=false
       roundingType=HalfHeightRadius
       width=0
       color {colorUseRule=UseColor_Specified, specifiedColor=<rgba 0 0 0 0>}
       cornerRadii=<0 0 0 0>
      }
      tintColor {colorUseRule=UseColor_Specified, specifiedColor=<rgba 0.0176 0.0176 0.0176 1>}
      uVRegion {bIsValid=false, max=<0 0>, min=<0 0>}
     }
     inactiveBrush {
      bIsDynamicallyLoaded=false
      drawAs=Image
      imageType=NoImage
      mirroring=NoMirror
      resourceName=None
      resourceObject=null
      tiling=NoTile
      imageSize=<0 0>
      margin=<ltrb 0 0 0 0>
      outlineSettings {
       bUseBrushTransparency=false
       roundingType=HalfHeightRadius
       width=0
       color {colorUseRule=UseColor_Specified, specifiedColor=<rgba 0 0 0 0>}
       cornerRadii=<0 0 0 0>
      }
      tintColor {colorUseRule=UseColor_Specified, specifiedColor=<rgba 0.0513 0.0953 0.159 1>}
      uVRegion {bIsValid=false, max=<0 0>, min=<0 0>}
     }
     inactiveHighlightedBrush {
      bIsDynamicallyLoaded=false
      drawAs=Image
      imageType=NoImage
      mirroring=NoMirror
      resourceName=None
      resourceObject=null
      tiling=NoTile
      imageSize=<0 0>
      margin=<ltrb 0 0 0 0>
      outlineSettings {
       bUseBrushTransparency=false
       roundingType=HalfHeightRadius
       width=0
       color {colorUseRule=UseColor_Specified, specifiedColor=<rgba 0 0 0 0>}
       cornerRadii=<0 0 0 0>
      }
      tintColor {colorUseRule=UseColor_Specified, specifiedColor=<rgba 0.0252 0.0319 0.0423 1>}
      uVRegion {bIsValid=false, max=<0 0>, min=<0 0>}
     }
     inactiveHoveredBrush {
      bIsDynamicallyLoaded=false
      drawAs=Image
      imageType=NoImage
      mirroring=NoMirror
      resourceName=None
      resourceObject=null
      tiling=NoTile
      imageSize=<0 0>
      margin=<ltrb 0 0 0 0>
      outlineSettings {
       bUseBrushTransparency=false
       roundingType=HalfHeightRadius
       width=0
       color {colorUseRule=UseColor_Specified, specifiedColor=<rgba 0 0 0 0>}
       cornerRadii=<0 0 0 0>
      }
      tintColor {colorUseRule=UseColor_Specified, specifiedColor=<rgba 0.0513 0.0953 0.159 1>}
      uVRegion {bIsValid=false, max=<0 0>, min=<0 0>}
     }
     oddRowBackgroundBrush {
      bIsDynamicallyLoaded=false
      drawAs=Image
      imageType=NoImage
      mirroring=NoMirror
      resourceName=None
      resourceObject=null
      tiling=NoTile
      imageSize=<0 0>
      margin=<ltrb 0 0 0 0>
      outlineSettings {
       bUseBrushTransparency=false
       roundingType=HalfHeightRadius
       width=0
       color {colorUseRule=UseColor_Specified, specifiedColor=<rgba 0 0 0 0>}
       cornerRadii=<0 0 0 0>
      }
      tintColor {colorUseRule=UseColor_Specified, specifiedColor=<rgba 0.0103 0.0103 0.0103 1>}
      uVRegion {bIsValid=false, max=<0 0>, min=<0 0>}
     }
     oddRowBackgroundHoveredBrush {
      bIsDynamicallyLoaded=false
      drawAs=Image
      imageType=NoImage
      mirroring=NoMirror
      resourceName=None
      resourceObject=null
      tiling=NoTile
      imageSize=<0 0>
      margin=<ltrb 0 0 0 0>
      outlineSettings {
       bUseBrushTransparency=false
       roundingType=HalfHeightRadius
       width=0
       color {colorUseRule=UseColor_Specified, specifiedColor=<rgba 0 0 0 0>}
       cornerRadii=<0 0 0 0>
      }
      tintColor {colorUseRule=UseColor_Specified, specifiedColor=<rgba 0.0176 0.0176 0.0176 1>}
      uVRegion {bIsValid=false, max=<0 0>, min=<0 0>}
     }
     parentRowBackgroundBrush {
      bIsDynamicallyLoaded=false
      drawAs=Image
      imageType=NoImage
      mirroring=NoMirror
      resourceName=None
      resourceObject=null
      tiling=NoTile
      imageSize=<32 32>
      margin=<ltrb 0 0 0 0>
      outlineSettings {
       bUseBrushTransparency=false
       roundingType=HalfHeightRadius
       width=0
       color {colorUseRule=UseColor_Specified, specifiedColor=<rgba 0 0 0 0>}
       cornerRadii=<0 0 0 0>
      }
      tintColor {colorUseRule=UseColor_Specified, specifiedColor=<rgba 1 1 1 1>}
      uVRegion {bIsValid=false, max=<0 0>, min=<0 0>}
     }
     parentRowBackgroundHoveredBrush {
      bIsDynamicallyLoaded=false
      drawAs=Image
      imageType=NoImage
      mirroring=NoMirror
      resourceName=None
      resourceObject=null
      tiling=NoTile
      imageSize=<32 32>
      margin=<ltrb 0 0 0 0>
      outlineSettings {
       bUseBrushTransparency=false
       roundingType=HalfHeightRadius
       width=0
       color {colorUseRule=UseColor_Specified, specifiedColor=<rgba 0 0 0 0>}
       cornerRadii=<0 0 0 0>
      }
      tintColor {colorUseRule=UseColor_Specified, specifiedColor=<rgba 1 1 1 1>}
      uVRegion {bIsValid=false, max=<0 0>, min=<0 0>}
     }
     selectedTextColor {colorUseRule=UseColor_Specified, specifiedColor=<rgba 1 1 1 1>}
     selectorFocusedBrush {
      bIsDynamicallyLoaded=false
      drawAs=Border
      imageType=FullColor
      mirroring=NoMirror
      resourceName=../../../Engine/Content/Slate/Common/Selector.png
      resourceObject=null
      tiling=Both
      imageSize=<0 0>
      margin=<ltrb 0.25 0.25 0.25 0.25>
      outlineSettings {
       bUseBrushTransparency=false
       roundingType=HalfHeightRadius
       width=0
       color {colorUseRule=UseColor_Specified, specifiedColor=<rgba 0 0 0 0>}
       cornerRadii=<0 0 0 0>
      }
      tintColor {colorUseRule=UseColor_Specified, specifiedColor=<rgba 0.701 0.225 0.003 1>}
      uVRegion {bIsValid=false, max=<0 0>, min=<0 0>}
     }
     textColor {colorUseRule=UseColor_Specified, specifiedColor=<rgba 0.5271 0.5271 0.5271 1>}
    }
    widgetStyle {
     comboButtonStyle {
      downArrowAlign=VAlign_Center
      buttonStyle {
       clickedSlateSound {resourceObject=null}
       disabled {
        bIsDynamicallyLoaded=false
        drawAs=RoundedBox
        imageType=NoImage
        mirroring=NoMirror
        resourceName=None
        resourceObject=null
        tiling=NoTile
        imageSize=<0 0>
        margin=<ltrb 0 0 0 0>
        outlineSettings {
         bUseBrushTransparency=false
         roundingType=FixedRadius
         width=1
         color {colorUseRule=UseColor_ColorTable, specifiedColor=<rgba 0 0 0 0>}
         cornerRadii=<4 4 4 4>
        }
        tintColor {colorUseRule=UseColor_Specified, specifiedColor=<rgba 0.0395 0.0395 0.0395 1>}
        uVRegion {bIsValid=false, max=<0 0>, min=<0 0>}
       }
       disabledForeground {colorUseRule=UseColor_ColorTable, specifiedColor=<rgba 0 0 0 0>}
       hovered {
        bIsDynamicallyLoaded=false
        drawAs=RoundedBox
        imageType=NoImage
        mirroring=NoMirror
        resourceName=None
        resourceObject=null
        tiling=NoTile
        imageSize=<0 0>
        margin=<ltrb 0 0 0 0>
        outlineSettings {
         bUseBrushTransparency=false
         roundingType=FixedRadius
         width=1
         color {colorUseRule=UseColor_ColorTable, specifiedColor=<rgba 0 0 0 0>}
         cornerRadii=<4 4 4 4>
        }
        tintColor {colorUseRule=UseColor_Specified, specifiedColor=<rgba 0.0048 0.0048 0.0048 1>}
        uVRegion {bIsValid=false, max=<0 0>, min=<0 0>}
       }
       hoveredForeground {colorUseRule=UseColor_ColorTable, specifiedColor=<rgba 0 0 0 0>}
       hoveredSlateSound {resourceObject=null}
       normal {
        bIsDynamicallyLoaded=false
        drawAs=RoundedBox
        imageType=NoImage
        mirroring=NoMirror
        resourceName=None
        resourceObject=null
        tiling=NoTile
        imageSize=<0 0>
        margin=<ltrb 0 0 0 0>
        outlineSettings {
         bUseBrushTransparency=false
         roundingType=FixedRadius
         width=1
         color {colorUseRule=UseColor_ColorTable, specifiedColor=<rgba 0 0 0 0>}
         cornerRadii=<4 4 4 4>
        }
        tintColor {colorUseRule=UseColor_Specified, specifiedColor=<rgba 0.0048 0.0048 0.0048 1>}
        uVRegion {bIsValid=false, max=<0 0>, min=<0 0>}
       }
       normalForeground {colorUseRule=UseColor_ColorTable, specifiedColor=<rgba 0 0 0 0>}
       normalPadding=<ltrb 8 1 8 1>
       pressed {
        bIsDynamicallyLoaded=false
        drawAs=RoundedBox
        imageType=NoImage
        mirroring=NoMirror
        resourceName=None
        resourceObject=null
        tiling=NoTile
        imageSize=<0 0>
        margin=<ltrb 0 0 0 0>
        outlineSettings {
         bUseBrushTransparency=false
         roundingType=FixedRadius
         width=1
         color {colorUseRule=UseColor_ColorTable, specifiedColor=<rgba 0 0 0 0>}
         cornerRadii=<4 4 4 4>
        }
        tintColor {colorUseRule=UseColor_Specified, specifiedColor=<rgba 0.0048 0.0048 0.0048 1>}
        uVRegion {bIsValid=false, max=<0 0>, min=<0 0>}
       }
       pressedForeground {colorUseRule=UseColor_ColorTable, specifiedColor=<rgba 0 0 0 0>}
       pressedPadding=<ltrb 8 2 8 0>
       pressedSlateSound {resourceObject=null}
      }
      contentPadding=<ltrb 0 0 0 0>
      downArrowImage {
       bIsDynamicallyLoaded=false
       drawAs=Image
       imageType=Vector
       mirroring=NoMirror
       resourceName=../../../Engine/Content/Slate/Starship/CoreWidgets/ComboBox/wide-chevron-down.svg
       resourceObject=null
       tiling=NoTile
       imageSize=<16 16>
       margin=<ltrb 0 0 0 0>
       outlineSettings {
        bUseBrushTransparency=false
        roundingType=HalfHeightRadius
        width=0
        color {colorUseRule=UseColor_Specified, specifiedColor=<rgba 0 0 0 0>}
        cornerRadii=<0 0 0 0>
       }
       tintColor {colorUseRule=UseColor_Specified, specifiedColor=<rgba 1 1 1 1>}
       uVRegion {bIsValid=false, max=<0 0>, min=<0 0>}
      }
      downArrowPadding=<ltrb 2 2 2 2>
      menuBorderBrush {
       bIsDynamicallyLoaded=false
       drawAs=RoundedBox
       imageType=NoImage
       mirroring=NoMirror
       resourceName=None
       resourceObject=null
       tiling=NoTile
       imageSize=<0 0>
       margin=<ltrb 0 0 0 0>
       outlineSettings {
        bUseBrushTransparency=false
        roundingType=FixedRadius
        width=1
        color {colorUseRule=UseColor_ColorTable, specifiedColor=<rgba 0 0 0 0>}
        cornerRadii=<0 0 0 0>
       }
       tintColor {colorUseRule=UseColor_Specified, specifiedColor=<rgba 0.0395 0.0395 0.0395 1>}
       uVRegion {bIsValid=false, max=<0 0>, min=<0 0>}
      }
      menuBorderPadding=<ltrb 0 0 0 0>
      shadowColorAndOpacity=<rgba 0 0 0 1>
      shadowOffset=<0 0>
     }
     contentPadding=<ltrb 0 0 0 0>
     menuRowPadding=<ltrb 8 3 8 3>
     pressedSlateSound {resourceObject=null}
     selectionChangeSlateSound {resourceObject=null}
    }
   }
   slot {class=SizeBoxSlot}
  }
  SizeBox_Shape {
   widgetClassPath=&SizeBox
   parent=VerticalBox
   widget=SizeBox_Shape
   properties {bOverride_WidthOverride=true, heightOverride=200, visibility=Visible, widthOverride=500}
   slot {class=VerticalBoxSlot, properties {horizontalAlignment=HAlign_Center, padding=<ltrb 5 5 5 5>}}
  }
  Border_Shape {
   widgetClassPath=&Border
   parent=SizeBox_Shape
   widget=Border_Shape
   properties {
    horizontalAlignment=HAlign_Center
    background {
     bIsDynamicallyLoaded=false
     drawAs=RoundedBox
     imageType=NoImage
     mirroring=NoMirror
     resourceName=None
     resourceObject=null
     tiling=NoTile
     imageSize=<32 32>
     margin=<ltrb 0 0 0 0>
     outlineSettings {
      bUseBrushTransparency=false
      roundingType=FixedRadius
      width=1
      color {colorUseRule=UseColor_Specified, specifiedColor=<rgba 0 0.7126 1 1>}
      cornerRadii=<5 5 5 5>
     }
     tintColor {colorUseRule=UseColor_Specified, specifiedColor=<rgba 0.0194 0.0194 0.0194 1>}
     uVRegion {bIsValid=false, max=<0 0>, min=<0 0>}
    }
    padding=<ltrb 5 5 5 5>
   }
   slot {class=SizeBoxSlot}
  }
  UI_HorizontalBox_Shape {
   widgetClassPath=&HorizontalBox
   bIsVariable=true
   parent=Border_Shape
   widget=UI_HorizontalBox_Shape
   slot {class=BorderSlot, properties {horizontalAlignment=HAlign_Center, padding=<ltrb 5 5 5 5>}}
  }
  SizeBox_Size {
   widgetClassPath=&SizeBox
   parent=VerticalBox
   widget=SizeBox_Size
   properties {bOverride_WidthOverride=true, heightOverride=200, visibility=Visible, widthOverride=500}
   slot {class=VerticalBoxSlot, properties {horizontalAlignment=HAlign_Center, padding=<ltrb 5 5 5 5>}}
  }
  Border_Size {
   widgetClassPath=&Border
   parent=SizeBox_Size
   widget=Border_Size
   properties {
    horizontalAlignment=HAlign_Center
    background {
     bIsDynamicallyLoaded=false
     drawAs=RoundedBox
     imageType=NoImage
     mirroring=NoMirror
     resourceName=None
     resourceObject=null
     tiling=NoTile
     imageSize=<32 32>
     margin=<ltrb 0 0 0 0>
     outlineSettings {
      bUseBrushTransparency=false
      roundingType=FixedRadius
      width=1
      color {colorUseRule=UseColor_Specified, specifiedColor=<rgba 0 0.7126 1 1>}
      cornerRadii=<5 5 5 5>
     }
     tintColor {colorUseRule=UseColor_Specified, specifiedColor=<rgba 0.0194 0.0194 0.0194 1>}
     uVRegion {bIsValid=false, max=<0 0>, min=<0 0>}
    }
    padding=<ltrb 5 5 5 5>
   }
   slot {class=SizeBoxSlot}
  }
  UI_HorizontalBox_Size {
   widgetClassPath=&HorizontalBox
   bIsVariable=true
   parent=Border_Size
   widget=UI_HorizontalBox_Size
   slot {class=BorderSlot, properties {horizontalAlignment=HAlign_Center, padding=<ltrb 5 5 5 5>}}
  }
  SizeBox_Feature {
   widgetClassPath=&SizeBox
   parent=VerticalBox
   widget=SizeBox_Feature
   properties {bOverride_WidthOverride=true, heightOverride=200, visibility=Visible, widthOverride=500}
   slot {class=VerticalBoxSlot, properties {horizontalAlignment=HAlign_Center, padding=<ltrb 5 5 5 5>}}
  }
  Border_Feature {
   widgetClassPath=&Border
   parent=SizeBox_Feature
   widget=Border_Feature
   properties {
    horizontalAlignment=HAlign_Center
    background {
     bIsDynamicallyLoaded=false
     drawAs=RoundedBox
     imageType=NoImage
     mirroring=NoMirror
     resourceName=None
     resourceObject=null
     tiling=NoTile
     imageSize=<32 32>
     margin=<ltrb 0 0 0 0>
     outlineSettings {
      bUseBrushTransparency=false
      roundingType=FixedRadius
      width=1
      color {colorUseRule=UseColor_Specified, specifiedColor=<rgba 0 0.7126 1 1>}
      cornerRadii=<5 5 5 5>
     }
     tintColor {colorUseRule=UseColor_Specified, specifiedColor=<rgba 0.0194 0.0194 0.0194 1>}
     uVRegion {bIsValid=false, max=<0 0>, min=<0 0>}
    }
    padding=<ltrb 5 5 5 5>
   }
   slot {class=SizeBoxSlot}
  }
  UI_HorizontalBox_Feature {
   widgetClassPath=&HorizontalBox
   bIsVariable=true
   parent=Border_Feature
   widget=UI_HorizontalBox_Feature
   slot {class=BorderSlot, properties {horizontalAlignment=HAlign_Center, padding=<ltrb 5 5 5 5>}}
  }
  SizeBox_AvailableBPs {
   widgetClassPath=&SizeBox
   parent=VerticalBox
   widget=SizeBox_AvailableBPs
   properties {bOverride_MaxDesiredHeight=true, bOverride_WidthOverride=true, heightOverride=10, maxDesiredHeight=200, widthOverride=500}
   slot {class=VerticalBoxSlot, properties {horizontalAlignment=HAlign_Center, padding=<ltrb 5 5 5 5>}}
  }
  Border_AvailableBPs {
   widgetClassPath=&Border
   parent=SizeBox_AvailableBPs
   widget=Border_AvailableBPs
   properties {
    background {
     bIsDynamicallyLoaded=false
     drawAs=RoundedBox
     imageType=NoImage
     mirroring=NoMirror
     resourceName=None
     resourceObject=null
     tiling=NoTile
     imageSize=<32 32>
     margin=<ltrb 0 0 0 0>
     outlineSettings {
      bUseBrushTransparency=false
      roundingType=FixedRadius
      width=1
      color {colorUseRule=UseColor_Specified, specifiedColor=<rgba 0 0.7126 1 1>}
      cornerRadii=<5 5 5 5>
     }
     tintColor {colorUseRule=UseColor_Specified, specifiedColor=<rgba 0.0194 0.0194 0.0194 1>}
     uVRegion {bIsValid=false, max=<0 0>, min=<0 0>}
    }
    padding=<ltrb 0 0 0 0>
   }
   slot {class=SizeBoxSlot}
  }
  EditorUtilityScrollBox_117 {
   widgetClassPath=&EditorUtilityScrollBox
   parent=Border_AvailableBPs
   widget=EditorUtilityScrollBox_117
   slot {class=BorderSlot, properties {padding=<ltrb 0 0 0 0>}}
  }
  UI_GridPanel_AvailableBPs {
   widgetClassPath=/Script/UMG.UniformGridPanel
   bIsVariable=true
   parent=EditorUtilityScrollBox_117
   widget=UI_GridPanel_AvailableBPs
   slot {class=ScrollBoxSlot, properties {horizontalAlignment=HAlign_Center, verticalAlignment=VAlign_Center}}
  }
  SizeBox_MakeBP {
   widgetClassPath=&SizeBox
   parent=VerticalBox
   widget=SizeBox_MakeBP
   properties {bOverride_WidthOverride=true, heightOverride=200, widthOverride=500}
   slot {class=VerticalBoxSlot, properties {horizontalAlignment=HAlign_Center, padding=<ltrb 5 5 5 5>}}
  }
  Border_MakeBP {
   widgetClassPath=&Border
   parent=SizeBox_MakeBP
   widget=Border_MakeBP
   properties {
    background {
     bIsDynamicallyLoaded=false
     drawAs=RoundedBox
     imageType=NoImage
     mirroring=NoMirror
     resourceName=None
     resourceObject=null
     tiling=NoTile
     imageSize=<32 32>
     margin=<ltrb 0 0 0 0>
     outlineSettings {
      bUseBrushTransparency=false
      roundingType=FixedRadius
      width=1
      color {colorUseRule=UseColor_Specified, specifiedColor=<rgba 0 0.7126 1 1>}
      cornerRadii=<5 5 5 5>
     }
     tintColor {colorUseRule=UseColor_Specified, specifiedColor=<rgba 0.0194 0.0194 0.0194 1>}
     uVRegion {bIsValid=false, max=<0 0>, min=<0 0>}
    }
    padding=<ltrb 5 5 5 5>
   }
   slot {class=SizeBoxSlot}
  }
  VerticalBox_0 {
   widgetClassPath=&VerticalBox
   parent=Border_MakeBP
   widget=VerticalBox_0
   slot {class=BorderSlot, properties {padding=<ltrb 5 5 5 5>}}
  }
  HorizontalBox_CustomOptions {
   widgetClassPath=&HorizontalBox
   parent=VerticalBox_0
   widget=HorizontalBox_CustomOptions
   slot {class=VerticalBoxSlot, properties {horizontalAlignment=HAlign_Center, padding=<ltrb 0 0 0 5>}}
  }
  Border_KeepTD {
   widgetClassPath=&Border
   parent=HorizontalBox_CustomOptions
   widget=Border_KeepTD
   properties {
    background {
     bIsDynamicallyLoaded=false
     drawAs=RoundedBox
     imageType=NoImage
     mirroring=NoMirror
     resourceName=None
     resourceObject=null
     tiling=NoTile
     imageSize=<32 32>
     margin=<ltrb 0 0 0 0>
     outlineSettings {
      bUseBrushTransparency=false
      roundingType=FixedRadius
      width=1
      color {colorUseRule=UseColor_Specified, specifiedColor=<rgba 0 0.7126 1 1>}
      cornerRadii=<5 5 5 5>
     }
     tintColor {colorUseRule=UseColor_Specified, specifiedColor=<rgba 0.0194 0.0194 0.0194 1>}
     uVRegion {bIsValid=false, max=<0 0>, min=<0 0>}
    }
    padding=<ltrb 5 5 5 5>
   }
   slot {class=HorizontalBoxSlot, properties {horizontalAlignment=HAlign_Center, verticalAlignment=VAlign_Center, padding=<ltrb 0 0 5 0>}}
  }
  HBox_KeepTD {
   widgetClassPath=&HorizontalBox
   parent=Border_KeepTD
   widget=HBox_KeepTD
   slot {class=BorderSlot, properties {padding=<ltrb 5 5 5 5>}}
  }
  Text_AddCustomSuffix_1 {
   widgetClassPath=&TextBlock
   parent=HBox_KeepTD
   widget=Text_AddCustomSuffix_1
   properties {
    text="Keep Blockout TDs: "
    font {
     bForceMonospaced=false
     bMaterialIsStencil=false
     fontMaterial=null
     fontObject=&Roboto
     letterSpacing=0
     monospacedWidth=1
     size=13.5
     skewAmount=0
     typefaceFontName=Bold
     outlineSettings {
      bApplyOutlineToDropShadows=false
      bMiteredCorners=false
      bSeparateFillAlpha=false
      outlineMaterial=null
      outlineSize=0
      outlineColor=<rgba 0 0 0 1>
     }
    }
   }
   slot {class=HorizontalBoxSlot, properties {horizontalAlignment=HAlign_Center, verticalAlignment=VAlign_Center}}
  }
  UI_Checkbox_KeepTD {
   widgetClassPath=&EditorUtilityCheckBox
   bIsVariable=true
   parent=HBox_KeepTD
   widget=UI_Checkbox_KeepTD
   slot {class=HorizontalBoxSlot}
  }
  Border_CustomSuffix {
   widgetClassPath=&Border
   parent=HorizontalBox_CustomOptions
   widget=Border_CustomSuffix
   properties {
    background {
     bIsDynamicallyLoaded=false
     drawAs=RoundedBox
     imageType=NoImage
     mirroring=NoMirror
     resourceName=None
     resourceObject=null
     tiling=NoTile
     imageSize=<32 32>
     margin=<ltrb 0 0 0 0>
     outlineSettings {
      bUseBrushTransparency=false
      roundingType=FixedRadius
      width=1
      color {colorUseRule=UseColor_Specified, specifiedColor=<rgba 0 0.7126 1 1>}
      cornerRadii=<5 5 5 5>
     }
     tintColor {colorUseRule=UseColor_Specified, specifiedColor=<rgba 0.0194 0.0194 0.0194 1>}
     uVRegion {bIsValid=false, max=<0 0>, min=<0 0>}
    }
    padding=<ltrb 5 5 5 5>
   }
   slot {class=HorizontalBoxSlot, properties {horizontalAlignment=HAlign_Center, verticalAlignment=VAlign_Center}}
  }
  HorizontalBox_1 {
   widgetClassPath=&HorizontalBox
   parent=Border_CustomSuffix
   widget=HorizontalBox_1
   slot {class=BorderSlot, properties {padding=<ltrb 5 5 5 5>}}
  }
  Text_AddCustomSuffix {
   widgetClassPath=&TextBlock
   parent=HorizontalBox_1
   widget=Text_AddCustomSuffix
   properties {
    text="Custom Suffix: "
    font {
     bForceMonospaced=false
     bMaterialIsStencil=false
     fontMaterial=null
     fontObject=&Roboto
     letterSpacing=0
     monospacedWidth=1
     size=13.5
     skewAmount=0
     typefaceFontName=Bold
     outlineSettings {
      bApplyOutlineToDropShadows=false
      bMiteredCorners=false
      bSeparateFillAlpha=false
      outlineMaterial=null
      outlineSize=0
      outlineColor=<rgba 0 0 0 1>
     }
    }
   }
   slot {class=HorizontalBoxSlot}
  }
  UI_Checkbox_AddCustomSuffix {
   widgetClassPath=&EditorUtilityCheckBox
   bIsVariable=true
   parent=HorizontalBox_1
   widget=UI_Checkbox_AddCustomSuffix
   slot {class=HorizontalBoxSlot}
  }
  SizeBox_1 {
   widgetClassPath=&SizeBox
   parent=HorizontalBox_1
   widget=SizeBox_1
   properties {bOverride_WidthOverride=true, widthOverride=122.5299}
   slot {class=HorizontalBoxSlot}
  }
  UI_EditableText_CustomSuffix {
   widgetClassPath=&EditorUtilityEditableTextBox
   bIsVariable=true
   parent=SizeBox_1
   widget=UI_EditableText_CustomSuffix
   properties {bIsEnabled=false}
   slot {class=SizeBoxSlot}
  }
  HorizontalBox_MakeBP {
   widgetClassPath=&HorizontalBox
   parent=VerticalBox_0
   widget=HorizontalBox_MakeBP
   slot {class=VerticalBoxSlot, properties {horizontalAlignment=HAlign_Center, verticalAlignment=VAlign_Center}}
  }
  HorizontalBox_Resources {
   widgetClassPath=&HorizontalBox
   parent=HorizontalBox_MakeBP
   widget=HorizontalBox_Resources
   slot {class=HorizontalBoxSlot, properties {padding=<ltrb 0 0 5 0>}}
  }
  SizeBox_RessourcesWood {
   widgetClassPath=&SizeBox
   parent=HorizontalBox_Resources
   widget=SizeBox_RessourcesWood
   properties {bOverride_HeightOverride=true, bOverride_WidthOverride=true, heightOverride=40, widthOverride=40}
   slot {class=HorizontalBoxSlot}
  }
  UI_Checkbox_Resource_Wood {
   widgetClassPath=&EditorUtilityCheckBox
   bIsVariable=true
   parent=SizeBox_RessourcesWood
   widget=UI_Checkbox_Resource_Wood
   properties {
    horizontalAlignment=HAlign_Center
    widgetStyle {
     checkBoxType=CheckBox
     backgroundHoveredImage {
      bIsDynamicallyLoaded=false
      drawAs=RoundedBox
      imageType=FullColor
      mirroring=NoMirror
      resourceName=None
      resourceObject=/Game/ArtTools/Environment/Implementation/BPSetup/Textures/WoodM
      tiling=NoTile
      imageSize=<40 40>
      margin=<ltrb 0 0 0 0>
      outlineSettings {
       bUseBrushTransparency=false
       roundingType=FixedRadius
       width=1
       color {colorUseRule=UseColor_Specified, specifiedColor=<rgba 0 0.7126 1 1>}
       cornerRadii=<5 5 5 5>
      }
      tintColor {colorUseRule=UseColor_Specified, specifiedColor=<rgba 1 1 1 1>}
      uVRegion {bIsValid=false, max=<0 0>, min=<0 0>}
     }
     backgroundImage {
      bIsDynamicallyLoaded=false
      drawAs=RoundedBox
      imageType=FullColor
      mirroring=NoMirror
      resourceName=None
      resourceObject=&WoodM
      tiling=NoTile
      imageSize=<40 40>
      margin=<ltrb 0 0 0 0>
      outlineSettings {
       bUseBrushTransparency=false
       roundingType=FixedRadius
       width=1
       color {colorUseRule=UseColor_Specified, specifiedColor=<rgba 0 0.7126 1 1>}
       cornerRadii=<5 5 5 5>
      }
      tintColor {colorUseRule=UseColor_Specified, specifiedColor=<rgba 1 1 1 1>}
      uVRegion {bIsValid=false, max=<0 0>, min=<0 0>}
     }
     backgroundPressedImage {
      bIsDynamicallyLoaded=false
      drawAs=RoundedBox
      imageType=FullColor
      mirroring=NoMirror
      resourceName=None
      resourceObject=&WoodM
      tiling=NoTile
      imageSize=<40 40>
      margin=<ltrb 0 0 0 0>
      outlineSettings {
       bUseBrushTransparency=false
       roundingType=FixedRadius
       width=1
       color {colorUseRule=UseColor_Specified, specifiedColor=<rgba 0 0.7126 1 1>}
       cornerRadii=<5 5 5 5>
      }
      tintColor {colorUseRule=UseColor_Specified, specifiedColor=<rgba 1 1 1 1>}
      uVRegion {bIsValid=false, max=<0 0>, min=<0 0>}
     }
     borderBackgroundColor {colorUseRule=UseColor_Specified, specifiedColor=<rgba 1 1 1 1>}
     checkedForeground {colorUseRule=UseColor_Specified, specifiedColor=<rgba 1 1 1 1>}
     checkedHoveredForeground {colorUseRule=UseColor_Specified, specifiedColor=<rgba 1 1 1 1>}
     checkedHoveredImage {
      bIsDynamicallyLoaded=false
      drawAs=Image
      imageType=Vector
      mirroring=NoMirror
      resourceName=../../../Engine/Content/Slate/Starship/CoreWidgets/CheckBox/check.svg
      resourceObject=null
      tiling=NoTile
      imageSize=<30 30>
      margin=<ltrb 0 0 0 0>
      outlineSettings {
       bUseBrushTransparency=false
       roundingType=HalfHeightRadius
       width=0
       color {colorUseRule=UseColor_Specified, specifiedColor=<rgba 0 0 0 0>}
       cornerRadii=<0 0 0 0>
      }
      tintColor {colorUseRule=UseColor_Specified, specifiedColor=<rgba 0 0.0802 0.3515 1>}
      uVRegion {bIsValid=false, max=<0 0>, min=<0 0>}
     }
     checkedImage {
      bIsDynamicallyLoaded=false
      drawAs=Image
      imageType=Vector
      mirroring=NoMirror
      resourceName=../../../Engine/Content/Slate/Starship/CoreWidgets/CheckBox/check.svg
      resourceObject=null
      tiling=NoTile
      imageSize=<30 30>
      margin=<ltrb 0 0 0 0>
      outlineSettings {
       bUseBrushTransparency=false
       roundingType=HalfHeightRadius
       width=0
       color {colorUseRule=UseColor_Specified, specifiedColor=<rgba 0 0 0 0>}
       cornerRadii=<0 0 0 0>
      }
      tintColor {colorUseRule=UseColor_Specified, specifiedColor=<rgba 0 0.162 0.7454 1>}
      uVRegion {bIsValid=false, max=<0 0>, min=<0 0>}
     }
     checkedPressedForeground {colorUseRule=UseColor_Specified, specifiedColor=<rgba 1 1 1 1>}
     checkedPressedImage {
      bIsDynamicallyLoaded=false
      drawAs=Image
      imageType=Vector
      mirroring=NoMirror
      resourceName=../../../Engine/Content/Slate/Starship/CoreWidgets/CheckBox/check.svg
      resourceObject=null
      tiling=NoTile
      imageSize=<30 30>
      margin=<ltrb 0 0 0 0>
      outlineSettings {
       bUseBrushTransparency=false
       roundingType=HalfHeightRadius
       width=0
       color {colorUseRule=UseColor_Specified, specifiedColor=<rgba 0 0 0 0>}
       cornerRadii=<0 0 0 0>
      }
      tintColor {colorUseRule=UseColor_Specified, specifiedColor=<rgba 0 0.0802 0.3515 1>}
      uVRegion {bIsValid=false, max=<0 0>, min=<0 0>}
     }
     checkedSlateSound {resourceObject=null}
     foregroundColor {colorUseRule=UseColor_Specified, specifiedColor=<rgba 1 1 1 0>}
     hoveredForeground {colorUseRule=UseColor_Specified, specifiedColor=<rgba 1 1 1 1>}
     hoveredSlateSound {resourceObject=null}
     padding=<ltrb 0 0 0 0>
     pressedForeground {colorUseRule=UseColor_Specified, specifiedColor=<rgba 1 1 1 1>}
     uncheckedHoveredImage {
      bIsDynamicallyLoaded=false
      drawAs=Image
      imageType=Vector
      mirroring=NoMirror
      resourceName=../../../Engine/Content/Slate/Starship/CoreWidgets/CheckBox/check.svg
      resourceObject=null
      tiling=NoTile
      imageSize=<30 30>
      margin=<ltrb 0 0 0 0>
      outlineSettings {
       bUseBrushTransparency=false
       roundingType=HalfHeightRadius
       width=0
       color {colorUseRule=UseColor_Specified, specifiedColor=<rgba 0 0 0 0>}
       cornerRadii=<0 0 0 0>
      }
      tintColor {colorUseRule=UseColor_Specified, specifiedColor=<rgba 0 0.0802 0.3515 1>}
      uVRegion {bIsValid=false, max=<0 0>, min=<0 0>}
     }
     uncheckedImage {
      bIsDynamicallyLoaded=false
      drawAs=NoDrawType
      imageType=NoImage
      mirroring=NoMirror
      resourceName=None
      resourceObject=null
      tiling=NoTile
      imageSize=<0 0>
      margin=<ltrb 0 0 0 0>
      outlineSettings {
       bUseBrushTransparency=false
       roundingType=HalfHeightRadius
       width=0
       color {colorUseRule=UseColor_Specified, specifiedColor=<rgba 0 0 0 0>}
       cornerRadii=<0 0 0 0>
      }
      tintColor {colorUseRule=UseColor_Specified, specifiedColor=<rgba 1 1 1 1>}
      uVRegion {bIsValid=false, max=<0 0>, min=<0 0>}
     }
     uncheckedPressedImage {
      bIsDynamicallyLoaded=false
      drawAs=Image
      imageType=Vector
      mirroring=NoMirror
      resourceName=../../../Engine/Content/Slate/Starship/CoreWidgets/CheckBox/check.svg
      resourceObject=null
      tiling=NoTile
      imageSize=<30 30>
      margin=<ltrb 0 0 0 0>
      outlineSettings {
       bUseBrushTransparency=false
       roundingType=HalfHeightRadius
       width=0
       color {colorUseRule=UseColor_Specified, specifiedColor=<rgba 0 0 0 0>}
       cornerRadii=<0 0 0 0>
      }
      tintColor {colorUseRule=UseColor_Specified, specifiedColor=<rgba 0 0.162 0.7454 1>}
      uVRegion {bIsValid=false, max=<0 0>, min=<0 0>}
     }
     uncheckedSlateSound {resourceObject=null}
     undeterminedForeground {colorUseRule=UseColor_Specified, specifiedColor=<rgba 1 1 1 1>}
     undeterminedHoveredImage {
      bIsDynamicallyLoaded=false
      drawAs=Image
      imageType=Vector
      mirroring=NoMirror
      resourceName=../../../Engine/Content/Slate/Starship/CoreWidgets/CheckBox/indeterminate.svg
      resourceObject=null
      tiling=NoTile
      imageSize=<16 16>
      margin=<ltrb 0 0 0 0>
      outlineSettings {
       bUseBrushTransparency=false
       roundingType=HalfHeightRadius
       width=0
       color {colorUseRule=UseColor_Specified, specifiedColor=<rgba 0 0 0 0>}
       cornerRadii=<0 0 0 0>
      }
      tintColor {colorUseRule=UseColor_Specified, specifiedColor=<rgba 0.0044 0.2384 1 1>}
      uVRegion {bIsValid=false, max=<0 0>, min=<0 0>}
     }
     undeterminedImage {
      bIsDynamicallyLoaded=false
      drawAs=Image
      imageType=Vector
      mirroring=NoMirror
      resourceName=../../../Engine/Content/Slate/Starship/CoreWidgets/CheckBox/indeterminate.svg
      resourceObject=null
      tiling=NoTile
      imageSize=<16 16>
      margin=<ltrb 0 0 0 0>
      outlineSettings {
       bUseBrushTransparency=false
       roundingType=HalfHeightRadius
       width=0
       color {colorUseRule=UseColor_Specified, specifiedColor=<rgba 0 0 0 0>}
       cornerRadii=<0 0 0 0>
      }
      tintColor {colorUseRule=UseColor_Specified, specifiedColor=<rgba 0 0.162 0.7454 1>}
      uVRegion {bIsValid=false, max=<0 0>, min=<0 0>}
     }
     undeterminedPressedImage {
      bIsDynamicallyLoaded=false
      drawAs=Image
      imageType=Vector
      mirroring=NoMirror
      resourceName=../../../Engine/Content/Slate/Starship/CoreWidgets/CheckBox/indeterminate.svg
      resourceObject=null
      tiling=NoTile
      imageSize=<16 16>
      margin=<ltrb 0 0 0 0>
      outlineSettings {
       bUseBrushTransparency=false
       roundingType=HalfHeightRadius
       width=0
       color {colorUseRule=UseColor_Specified, specifiedColor=<rgba 0 0 0 0>}
       cornerRadii=<0 0 0 0>
      }
      tintColor {colorUseRule=UseColor_Specified, specifiedColor=<rgba 0 0.0802 0.3515 1>}
      uVRegion {bIsValid=false, max=<0 0>, min=<0 0>}
     }
    }
   }
   slot {class=SizeBoxSlot, properties {horizontalAlignment=HAlign_Center, verticalAlignment=VAlign_Center}}
  }
  SizeBox_Ressources_Stone {
   widgetClassPath=&SizeBox
   parent=HorizontalBox_Resources
   widget=SizeBox_Ressources_Stone
   properties {bOverride_HeightOverride=true, bOverride_WidthOverride=true, heightOverride=40, widthOverride=40}
   slot {class=HorizontalBoxSlot, properties {padding=<ltrb 5 0 0 0>}}
  }
  UI_Checkbox_Resource_Stone {
   widgetClassPath=&EditorUtilityCheckBox
   bIsVariable=true
   parent=SizeBox_Ressources_Stone
   widget=UI_Checkbox_Resource_Stone
   properties {
    horizontalAlignment=HAlign_Center
    widgetStyle {
     checkBoxType=CheckBox
     backgroundHoveredImage {
      bIsDynamicallyLoaded=false
      drawAs=RoundedBox
      imageType=FullColor
      mirroring=NoMirror
      resourceName=None
      resourceObject=/Game/ArtTools/Environment/Implementation/BPSetup/Textures/StoneM
      tiling=NoTile
      imageSize=<40 40>
      margin=<ltrb 0 0 0 0>
      outlineSettings {
       bUseBrushTransparency=false
       roundingType=FixedRadius
       width=1
       color {colorUseRule=UseColor_Specified, specifiedColor=<rgba 0 0.7126 1 1>}
       cornerRadii=<5 5 5 5>
      }
      tintColor {colorUseRule=UseColor_Specified, specifiedColor=<rgba 1 1 1 1>}
      uVRegion {bIsValid=false, max=<0 0>, min=<0 0>}
     }
     backgroundImage {
      bIsDynamicallyLoaded=false
      drawAs=RoundedBox
      imageType=FullColor
      mirroring=NoMirror
      resourceName=None
      resourceObject=&StoneM
      tiling=NoTile
      imageSize=<40 40>
      margin=<ltrb 0 0 0 0>
      outlineSettings {
       bUseBrushTransparency=false
       roundingType=FixedRadius
       width=1
       color {colorUseRule=UseColor_Specified, specifiedColor=<rgba 0 0.7126 1 1>}
       cornerRadii=<5 5 5 5>
      }
      tintColor {colorUseRule=UseColor_Specified, specifiedColor=<rgba 1 1 1 1>}
      uVRegion {bIsValid=false, max=<0 0>, min=<0 0>}
     }
     backgroundPressedImage {
      bIsDynamicallyLoaded=false
      drawAs=RoundedBox
      imageType=FullColor
      mirroring=NoMirror
      resourceName=None
      resourceObject=&StoneM
      tiling=NoTile
      imageSize=<40 40>
      margin=<ltrb 0 0 0 0>
      outlineSettings {
       bUseBrushTransparency=false
       roundingType=FixedRadius
       width=1
       color {colorUseRule=UseColor_Specified, specifiedColor=<rgba 0 0.7126 1 1>}
       cornerRadii=<5 5 5 5>
      }
      tintColor {colorUseRule=UseColor_Specified, specifiedColor=<rgba 1 1 1 1>}
      uVRegion {bIsValid=false, max=<0 0>, min=<0 0>}
     }
     borderBackgroundColor {colorUseRule=UseColor_Specified, specifiedColor=<rgba 1 1 1 1>}
     checkedForeground {colorUseRule=UseColor_Specified, specifiedColor=<rgba 1 1 1 1>}
     checkedHoveredForeground {colorUseRule=UseColor_Specified, specifiedColor=<rgba 1 1 1 1>}
     checkedHoveredImage {
      bIsDynamicallyLoaded=false
      drawAs=Image
      imageType=Vector
      mirroring=NoMirror
      resourceName=../../../Engine/Content/Slate/Starship/CoreWidgets/CheckBox/check.svg
      resourceObject=null
      tiling=NoTile
      imageSize=<30 30>
      margin=<ltrb 0 0 0 0>
      outlineSettings {
       bUseBrushTransparency=false
       roundingType=HalfHeightRadius
       width=0
       color {colorUseRule=UseColor_Specified, specifiedColor=<rgba 0 0 0 0>}
       cornerRadii=<0 0 0 0>
      }
      tintColor {colorUseRule=UseColor_Specified, specifiedColor=<rgba 0 0.0802 0.3515 1>}
      uVRegion {bIsValid=false, max=<0 0>, min=<0 0>}
     }
     checkedImage {
      bIsDynamicallyLoaded=false
      drawAs=Image
      imageType=Vector
      mirroring=NoMirror
      resourceName=../../../Engine/Content/Slate/Starship/CoreWidgets/CheckBox/check.svg
      resourceObject=null
      tiling=NoTile
      imageSize=<30 30>
      margin=<ltrb 0 0 0 0>
      outlineSettings {
       bUseBrushTransparency=false
       roundingType=HalfHeightRadius
       width=0
       color {colorUseRule=UseColor_Specified, specifiedColor=<rgba 0 0 0 0>}
       cornerRadii=<0 0 0 0>
      }
      tintColor {colorUseRule=UseColor_Specified, specifiedColor=<rgba 0 0.162 0.7454 1>}
      uVRegion {bIsValid=false, max=<0 0>, min=<0 0>}
     }
     checkedPressedForeground {colorUseRule=UseColor_Specified, specifiedColor=<rgba 1 1 1 1>}
     checkedPressedImage {
      bIsDynamicallyLoaded=false
      drawAs=Image
      imageType=Vector
      mirroring=NoMirror
      resourceName=../../../Engine/Content/Slate/Starship/CoreWidgets/CheckBox/check.svg
      resourceObject=null
      tiling=NoTile
      imageSize=<30 30>
      margin=<ltrb 0 0 0 0>
      outlineSettings {
       bUseBrushTransparency=false
       roundingType=HalfHeightRadius
       width=0
       color {colorUseRule=UseColor_Specified, specifiedColor=<rgba 0 0 0 0>}
       cornerRadii=<0 0 0 0>
      }
      tintColor {colorUseRule=UseColor_Specified, specifiedColor=<rgba 0 0.0802 0.3515 1>}
      uVRegion {bIsValid=false, max=<0 0>, min=<0 0>}
     }
     checkedSlateSound {resourceObject=null}
     foregroundColor {colorUseRule=UseColor_Specified, specifiedColor=<rgba 1 1 1 0>}
     hoveredForeground {colorUseRule=UseColor_Specified, specifiedColor=<rgba 1 1 1 1>}
     hoveredSlateSound {resourceObject=null}
     padding=<ltrb 0 0 0 0>
     pressedForeground {colorUseRule=UseColor_Specified, specifiedColor=<rgba 1 1 1 1>}
     uncheckedHoveredImage {
      bIsDynamicallyLoaded=false
      drawAs=Image
      imageType=Vector
      mirroring=NoMirror
      resourceName=../../../Engine/Content/Slate/Starship/CoreWidgets/CheckBox/check.svg
      resourceObject=null
      tiling=NoTile
      imageSize=<30 30>
      margin=<ltrb 0 0 0 0>
      outlineSettings {
       bUseBrushTransparency=false
       roundingType=HalfHeightRadius
       width=0
       color {colorUseRule=UseColor_Specified, specifiedColor=<rgba 0 0 0 0>}
       cornerRadii=<0 0 0 0>
      }
      tintColor {colorUseRule=UseColor_Specified, specifiedColor=<rgba 0 0.0802 0.3515 1>}
      uVRegion {bIsValid=false, max=<0 0>, min=<0 0>}
     }
     uncheckedImage {
      bIsDynamicallyLoaded=false
      drawAs=NoDrawType
      imageType=NoImage
      mirroring=NoMirror
      resourceName=None
      resourceObject=null
      tiling=NoTile
      imageSize=<0 0>
      margin=<ltrb 0 0 0 0>
      outlineSettings {
       bUseBrushTransparency=false
       roundingType=HalfHeightRadius
       width=0
       color {colorUseRule=UseColor_Specified, specifiedColor=<rgba 0 0 0 0>}
       cornerRadii=<0 0 0 0>
      }
      tintColor {colorUseRule=UseColor_Specified, specifiedColor=<rgba 1 1 1 1>}
      uVRegion {bIsValid=false, max=<0 0>, min=<0 0>}
     }
     uncheckedPressedImage {
      bIsDynamicallyLoaded=false
      drawAs=Image
      imageType=Vector
      mirroring=NoMirror
      resourceName=../../../Engine/Content/Slate/Starship/CoreWidgets/CheckBox/check.svg
      resourceObject=null
      tiling=NoTile
      imageSize=<30 30>
      margin=<ltrb 0 0 0 0>
      outlineSettings {
       bUseBrushTransparency=false
       roundingType=HalfHeightRadius
       width=0
       color {colorUseRule=UseColor_Specified, specifiedColor=<rgba 0 0 0 0>}
       cornerRadii=<0 0 0 0>
      }
      tintColor {colorUseRule=UseColor_Specified, specifiedColor=<rgba 0 0.162 0.7454 1>}
      uVRegion {bIsValid=false, max=<0 0>, min=<0 0>}
     }
     uncheckedSlateSound {resourceObject=null}
     undeterminedForeground {colorUseRule=UseColor_Specified, specifiedColor=<rgba 1 1 1 1>}
     undeterminedHoveredImage {
      bIsDynamicallyLoaded=false
      drawAs=Image
      imageType=Vector
      mirroring=NoMirror
      resourceName=../../../Engine/Content/Slate/Starship/CoreWidgets/CheckBox/indeterminate.svg
      resourceObject=null
      tiling=NoTile
      imageSize=<16 16>
      margin=<ltrb 0 0 0 0>
      outlineSettings {
       bUseBrushTransparency=false
       roundingType=HalfHeightRadius
       width=0
       color {colorUseRule=UseColor_Specified, specifiedColor=<rgba 0 0 0 0>}
       cornerRadii=<0 0 0 0>
      }
      tintColor {colorUseRule=UseColor_Specified, specifiedColor=<rgba 0.0044 0.2384 1 1>}
      uVRegion {bIsValid=false, max=<0 0>, min=<0 0>}
     }
     undeterminedImage {
      bIsDynamicallyLoaded=false
      drawAs=Image
      imageType=Vector
      mirroring=NoMirror
      resourceName=../../../Engine/Content/Slate/Starship/CoreWidgets/CheckBox/indeterminate.svg
      resourceObject=null
      tiling=NoTile
      imageSize=<16 16>
      margin=<ltrb 0 0 0 0>
      outlineSettings {
       bUseBrushTransparency=false
       roundingType=HalfHeightRadius
       width=0
       color {colorUseRule=UseColor_Specified, specifiedColor=<rgba 0 0 0 0>}
       cornerRadii=<0 0 0 0>
      }
      tintColor {colorUseRule=UseColor_Specified, specifiedColor=<rgba 0 0.162 0.7454 1>}
      uVRegion {bIsValid=false, max=<0 0>, min=<0 0>}
     }
     undeterminedPressedImage {
      bIsDynamicallyLoaded=false
      drawAs=Image
      imageType=Vector
      mirroring=NoMirror
      resourceName=../../../Engine/Content/Slate/Starship/CoreWidgets/CheckBox/indeterminate.svg
      resourceObject=null
      tiling=NoTile
      imageSize=<16 16>
      margin=<ltrb 0 0 0 0>
      outlineSettings {
       bUseBrushTransparency=false
       roundingType=HalfHeightRadius
       width=0
       color {colorUseRule=UseColor_Specified, specifiedColor=<rgba 0 0 0 0>}
       cornerRadii=<0 0 0 0>
      }
      tintColor {colorUseRule=UseColor_Specified, specifiedColor=<rgba 0 0.0802 0.3515 1>}
      uVRegion {bIsValid=false, max=<0 0>, min=<0 0>}
     }
    }
   }
   slot {class=SizeBoxSlot, properties {horizontalAlignment=HAlign_Center, verticalAlignment=VAlign_Center}}
  }
  SizeBox_Ressources_Metal {
   widgetClassPath=&SizeBox
   parent=HorizontalBox_Resources
   widget=SizeBox_Ressources_Metal
   properties {bOverride_HeightOverride=true, bOverride_WidthOverride=true, heightOverride=40, widthOverride=40}
   slot {class=HorizontalBoxSlot, properties {padding=<ltrb 5 0 0 0>}}
  }
  UI_Checkbox_Resource_Metal {
   widgetClassPath=&EditorUtilityCheckBox
   bIsVariable=true
   parent=SizeBox_Ressources_Metal
   widget=UI_Checkbox_Resource_Metal
   properties {
    horizontalAlignment=HAlign_Center
    widgetStyle {
     checkBoxType=CheckBox
     backgroundHoveredImage {
      bIsDynamicallyLoaded=false
      drawAs=RoundedBox
      imageType=FullColor
      mirroring=NoMirror
      resourceName=None
      resourceObject=/Game/ArtTools/Environment/Implementation/BPSetup/Textures/MetalM
      tiling=NoTile
      imageSize=<40 40>
      margin=<ltrb 0 0 0 0>
      outlineSettings {
       bUseBrushTransparency=false
       roundingType=FixedRadius
       width=1
       color {colorUseRule=UseColor_Specified, specifiedColor=<rgba 0 0.7126 1 1>}
       cornerRadii=<5 5 5 5>
      }
      tintColor {colorUseRule=UseColor_Specified, specifiedColor=<rgba 1 1 1 1>}
      uVRegion {bIsValid=false, max=<0 0>, min=<0 0>}
     }
     backgroundImage {
      bIsDynamicallyLoaded=false
      drawAs=RoundedBox
      imageType=FullColor
      mirroring=NoMirror
      resourceName=None
      resourceObject=&MetalM
      tiling=NoTile
      imageSize=<40 40>
      margin=<ltrb 0 0 0 0>
      outlineSettings {
       bUseBrushTransparency=false
       roundingType=FixedRadius
       width=1
       color {colorUseRule=UseColor_Specified, specifiedColor=<rgba 0 0.7126 1 1>}
       cornerRadii=<5 5 5 5>
      }
      tintColor {colorUseRule=UseColor_Specified, specifiedColor=<rgba 1 1 1 1>}
      uVRegion {bIsValid=false, max=<0 0>, min=<0 0>}
     }
     backgroundPressedImage {
      bIsDynamicallyLoaded=false
      drawAs=RoundedBox
      imageType=FullColor
      mirroring=NoMirror
      resourceName=None
      resourceObject=&MetalM
      tiling=NoTile
      imageSize=<40 40>
      margin=<ltrb 0 0 0 0>
      outlineSettings {
       bUseBrushTransparency=false
       roundingType=FixedRadius
       width=1
       color {colorUseRule=UseColor_Specified, specifiedColor=<rgba 0 0.7126 1 1>}
       cornerRadii=<5 5 5 5>
      }
      tintColor {colorUseRule=UseColor_Specified, specifiedColor=<rgba 1 1 1 1>}
      uVRegion {bIsValid=false, max=<0 0>, min=<0 0>}
     }
     borderBackgroundColor {colorUseRule=UseColor_Specified, specifiedColor=<rgba 1 1 1 1>}
     checkedForeground {colorUseRule=UseColor_Specified, specifiedColor=<rgba 1 1 1 1>}
     checkedHoveredForeground {colorUseRule=UseColor_Specified, specifiedColor=<rgba 1 1 1 1>}
     checkedHoveredImage {
      bIsDynamicallyLoaded=false
      drawAs=Image
      imageType=Vector
      mirroring=NoMirror
      resourceName=../../../Engine/Content/Slate/Starship/CoreWidgets/CheckBox/check.svg
      resourceObject=null
      tiling=NoTile
      imageSize=<30 30>
      margin=<ltrb 0 0 0 0>
      outlineSettings {
       bUseBrushTransparency=false
       roundingType=HalfHeightRadius
       width=0
       color {colorUseRule=UseColor_Specified, specifiedColor=<rgba 0 0 0 0>}
       cornerRadii=<0 0 0 0>
      }
      tintColor {colorUseRule=UseColor_Specified, specifiedColor=<rgba 0 0.0802 0.3515 1>}
      uVRegion {bIsValid=false, max=<0 0>, min=<0 0>}
     }
     checkedImage {
      bIsDynamicallyLoaded=false
      drawAs=Image
      imageType=Vector
      mirroring=NoMirror
      resourceName=../../../Engine/Content/Slate/Starship/CoreWidgets/CheckBox/check.svg
      resourceObject=null
      tiling=NoTile
      imageSize=<30 30>
      margin=<ltrb 0 0 0 0>
      outlineSettings {
       bUseBrushTransparency=false
       roundingType=HalfHeightRadius
       width=0
       color {colorUseRule=UseColor_Specified, specifiedColor=<rgba 0 0 0 0>}
       cornerRadii=<0 0 0 0>
      }
      tintColor {colorUseRule=UseColor_Specified, specifiedColor=<rgba 0 0.162 0.7454 1>}
      uVRegion {bIsValid=false, max=<0 0>, min=<0 0>}
     }
     checkedPressedForeground {colorUseRule=UseColor_Specified, specifiedColor=<rgba 1 1 1 1>}
     checkedPressedImage {
      bIsDynamicallyLoaded=false
      drawAs=Image
      imageType=Vector
      mirroring=NoMirror
      resourceName=../../../Engine/Content/Slate/Starship/CoreWidgets/CheckBox/check.svg
      resourceObject=null
      tiling=NoTile
      imageSize=<30 30>
      margin=<ltrb 0 0 0 0>
      outlineSettings {
       bUseBrushTransparency=false
       roundingType=HalfHeightRadius
       width=0
       color {colorUseRule=UseColor_Specified, specifiedColor=<rgba 0 0 0 0>}
       cornerRadii=<0 0 0 0>
      }
      tintColor {colorUseRule=UseColor_Specified, specifiedColor=<rgba 0 0.0802 0.3515 1>}
      uVRegion {bIsValid=false, max=<0 0>, min=<0 0>}
     }
     checkedSlateSound {resourceObject=null}
     foregroundColor {colorUseRule=UseColor_Specified, specifiedColor=<rgba 1 1 1 0>}
     hoveredForeground {colorUseRule=UseColor_Specified, specifiedColor=<rgba 1 1 1 1>}
     hoveredSlateSound {resourceObject=null}
     padding=<ltrb 0 0 0 0>
     pressedForeground {colorUseRule=UseColor_Specified, specifiedColor=<rgba 1 1 1 1>}
     uncheckedHoveredImage {
      bIsDynamicallyLoaded=false
      drawAs=Image
      imageType=Vector
      mirroring=NoMirror
      resourceName=../../../Engine/Content/Slate/Starship/CoreWidgets/CheckBox/check.svg
      resourceObject=null
      tiling=NoTile
      imageSize=<30 30>
      margin=<ltrb 0 0 0 0>
      outlineSettings {
       bUseBrushTransparency=false
       roundingType=HalfHeightRadius
       width=0
       color {colorUseRule=UseColor_Specified, specifiedColor=<rgba 0 0 0 0>}
       cornerRadii=<0 0 0 0>
      }
      tintColor {colorUseRule=UseColor_Specified, specifiedColor=<rgba 0 0.0802 0.3515 1>}
      uVRegion {bIsValid=false, max=<0 0>, min=<0 0>}
     }
     uncheckedImage {
      bIsDynamicallyLoaded=false
      drawAs=NoDrawType
      imageType=NoImage
      mirroring=NoMirror
      resourceName=None
      resourceObject=null
      tiling=NoTile
      imageSize=<0 0>
      margin=<ltrb 0 0 0 0>
      outlineSettings {
       bUseBrushTransparency=false
       roundingType=HalfHeightRadius
       width=0
       color {colorUseRule=UseColor_Specified, specifiedColor=<rgba 0 0 0 0>}
       cornerRadii=<0 0 0 0>
      }
      tintColor {colorUseRule=UseColor_Specified, specifiedColor=<rgba 1 1 1 1>}
      uVRegion {bIsValid=false, max=<0 0>, min=<0 0>}
     }
     uncheckedPressedImage {
      bIsDynamicallyLoaded=false
      drawAs=Image
      imageType=Vector
      mirroring=NoMirror
      resourceName=../../../Engine/Content/Slate/Starship/CoreWidgets/CheckBox/check.svg
      resourceObject=null
      tiling=NoTile
      imageSize=<30 30>
      margin=<ltrb 0 0 0 0>
      outlineSettings {
       bUseBrushTransparency=false
       roundingType=HalfHeightRadius
       width=0
       color {colorUseRule=UseColor_Specified, specifiedColor=<rgba 0 0 0 0>}
       cornerRadii=<0 0 0 0>
      }
      tintColor {colorUseRule=UseColor_Specified, specifiedColor=<rgba 0 0.162 0.7454 1>}
      uVRegion {bIsValid=false, max=<0 0>, min=<0 0>}
     }
     uncheckedSlateSound {resourceObject=null}
     undeterminedForeground {colorUseRule=UseColor_Specified, specifiedColor=<rgba 1 1 1 1>}
     undeterminedHoveredImage {
      bIsDynamicallyLoaded=false
      drawAs=Image
      imageType=Vector
      mirroring=NoMirror
      resourceName=../../../Engine/Content/Slate/Starship/CoreWidgets/CheckBox/indeterminate.svg
      resourceObject=null
      tiling=NoTile
      imageSize=<16 16>
      margin=<ltrb 0 0 0 0>
      outlineSettings {
       bUseBrushTransparency=false
       roundingType=HalfHeightRadius
       width=0
       color {colorUseRule=UseColor_Specified, specifiedColor=<rgba 0 0 0 0>}
       cornerRadii=<0 0 0 0>
      }
      tintColor {colorUseRule=UseColor_Specified, specifiedColor=<rgba 0.0044 0.2384 1 1>}
      uVRegion {bIsValid=false, max=<0 0>, min=<0 0>}
     }
     undeterminedImage {
      bIsDynamicallyLoaded=false
      drawAs=Image
      imageType=Vector
      mirroring=NoMirror
      resourceName=../../../Engine/Content/Slate/Starship/CoreWidgets/CheckBox/indeterminate.svg
      resourceObject=null
      tiling=NoTile
      imageSize=<16 16>
      margin=<ltrb 0 0 0 0>
      outlineSettings {
       bUseBrushTransparency=false
       roundingType=HalfHeightRadius
       width=0
       color {colorUseRule=UseColor_Specified, specifiedColor=<rgba 0 0 0 0>}
       cornerRadii=<0 0 0 0>
      }
      tintColor {colorUseRule=UseColor_Specified, specifiedColor=<rgba 0 0.162 0.7454 1>}
      uVRegion {bIsValid=false, max=<0 0>, min=<0 0>}
     }
     undeterminedPressedImage {
      bIsDynamicallyLoaded=false
      drawAs=Image
      imageType=Vector
      mirroring=NoMirror
      resourceName=../../../Engine/Content/Slate/Starship/CoreWidgets/CheckBox/indeterminate.svg
      resourceObject=null
      tiling=NoTile
      imageSize=<16 16>
      margin=<ltrb 0 0 0 0>
      outlineSettings {
       bUseBrushTransparency=false
       roundingType=HalfHeightRadius
       width=0
       color {colorUseRule=UseColor_Specified, specifiedColor=<rgba 0 0 0 0>}
       cornerRadii=<0 0 0 0>
      }
      tintColor {colorUseRule=UseColor_Specified, specifiedColor=<rgba 0 0.0802 0.3515 1>}
      uVRegion {bIsValid=false, max=<0 0>, min=<0 0>}
     }
    }
   }
   slot {class=SizeBoxSlot}
  }
  HorizontalBox_MakeBPButton {
   widgetClassPath=&HorizontalBox
   parent=HorizontalBox_MakeBP
   widget=HorizontalBox_MakeBPButton
   slot {class=HorizontalBoxSlot, properties {horizontalAlignment=HAlign_Center, verticalAlignment=VAlign_Center}}
  }
  SizeBox_CBox_MakeBPButton {
   widgetClassPath=&SizeBox
   parent=HorizontalBox_MakeBPButton
   widget=SizeBox_CBox_MakeBPButton
   properties {bOverride_HeightOverride=true, bOverride_WidthOverride=true, heightOverride=40, widthOverride=295}
   slot {class=HorizontalBoxSlot, properties {horizontalAlignment=HAlign_Center, verticalAlignment=VAlign_Center}}
  }
  UI_Button_MakeBPButton {
   widgetClassPath=&EditorUtilityButton
   parent=SizeBox_CBox_MakeBPButton
   widget=UI_Button_MakeBPButton
   properties {widgetStyle {
     clickedSlateSound {resourceObject=null}
     disabled {
      bIsDynamicallyLoaded=false
      drawAs=RoundedBox
      imageType=NoImage
      mirroring=NoMirror
      resourceName=None
      resourceObject=null
      tiling=NoTile
      imageSize=<0 0>
      margin=<ltrb 0 0 0 0>
      outlineSettings {
       bUseBrushTransparency=false
       roundingType=FixedRadius
       width=1
       color {colorUseRule=UseColor_ColorTable, specifiedColor=<rgba 0 0 0 0>}
       cornerRadii=<4 4 4 4>
      }
      tintColor {colorUseRule=UseColor_Specified, specifiedColor=<rgba 0.0395 0.0395 0.0395 1>}
      uVRegion {bIsValid=false, max=<0 0>, min=<0 0>}
     }
     disabledForeground {colorUseRule=UseColor_ColorTable, specifiedColor=<rgba 0 0 0 0>}
     hovered {
      bIsDynamicallyLoaded=false
      drawAs=RoundedBox
      imageType=NoImage
      mirroring=NoMirror
      resourceName=None
      resourceObject=null
      tiling=NoTile
      imageSize=<32 32>
      margin=<ltrb 0 0 0 0>
      outlineSettings {
       bUseBrushTransparency=false
       roundingType=FixedRadius
       width=1
       color {colorUseRule=UseColor_ColorTable, specifiedColor=<rgba 0 0 0 0>}
       cornerRadii=<4 4 4 4>
      }
      tintColor {colorUseRule=UseColor_Specified, specifiedColor=<rgba 0.0953 0.0953 0.0953 1>}
      uVRegion {bIsValid=false, max=<0 0>, min=<0 0>}
     }
     hoveredForeground {colorUseRule=UseColor_ColorTable, specifiedColor=<rgba 0 0 0 0>}
     hoveredSlateSound {resourceObject=null}
     normal {
      bIsDynamicallyLoaded=false
      drawAs=RoundedBox
      imageType=NoImage
      mirroring=NoMirror
      resourceName=None
      resourceObject=null
      tiling=NoTile
      imageSize=<32 32>
      margin=<ltrb 0 0 0 0>
      outlineSettings {
       bUseBrushTransparency=false
       roundingType=FixedRadius
       width=1
       color {colorUseRule=UseColor_Specified, specifiedColor=<rgba 0 0.7126 1 1>}
       cornerRadii=<4 4 4 4>
      }
      tintColor {colorUseRule=UseColor_Specified, specifiedColor=<rgba 0.0395 0.0395 0.0395 1>}
      uVRegion {bIsValid=false, max=<0 0>, min=<0 0>}
     }
     normalForeground {colorUseRule=UseColor_ColorTable, specifiedColor=<rgba 0 0 0 0>}
     normalPadding=<ltrb 12 1.5 12 1.5>
     pressed {
      bIsDynamicallyLoaded=false
      drawAs=RoundedBox
      imageType=NoImage
      mirroring=NoMirror
      resourceName=None
      resourceObject=null
      tiling=NoTile
      imageSize=<32 32>
      margin=<ltrb 0 0 0 0>
      outlineSettings {
       bUseBrushTransparency=false
       roundingType=FixedRadius
       width=1
       color {colorUseRule=UseColor_ColorTable, specifiedColor=<rgba 0 0 0 0>}
       cornerRadii=<4 4 4 4>
      }
      tintColor {colorUseRule=UseColor_Specified, specifiedColor=<rgba 0.0284 0.0284 0.0284 1>}
      uVRegion {bIsValid=false, max=<0 0>, min=<0 0>}
     }
     pressedForeground {colorUseRule=UseColor_ColorTable, specifiedColor=<rgba 0 0 0 0>}
     pressedPadding=<ltrb 12 2.5 12 0.5>
     pressedSlateSound {resourceObject=null}
    }}
   slot {class=SizeBoxSlot}
  }
  UI_Text_MakeBP {
   widgetClassPath=&TextBlock
   parent=UI_Button_MakeBPButton
   widget=UI_Text_MakeBP
   properties {
    text=Make BP
    font {
     bForceMonospaced=false
     bMaterialIsStencil=false
     fontMaterial=null
     fontObject=&Roboto
     letterSpacing=0
     monospacedWidth=1
     size=13.5
     skewAmount=0
     typefaceFontName=Bold
     outlineSettings {
      bApplyOutlineToDropShadows=false
      bMiteredCorners=false
      bSeparateFillAlpha=false
      outlineMaterial=null
      outlineSize=0
      outlineColor=<rgba 0 0 0 1>
     }
    }
   }
   slot {class=ButtonSlot}
  }
  HorizontalBox_ResetButton {
   widgetClassPath=&HorizontalBox
   parent=HorizontalBox_MakeBP
   widget=HorizontalBox_ResetButton
   slot {class=HorizontalBoxSlot, properties {padding=<ltrb 5 0 0 0>}}
  }
  SizeBox_CBox__ResetButton {
   widgetClassPath=&SizeBox
   parent=HorizontalBox_ResetButton
   widget=SizeBox_CBox__ResetButton
   properties {bOverride_HeightOverride=true, bOverride_WidthOverride=true, heightOverride=40, widthOverride=40}
   slot {class=HorizontalBoxSlot}
  }
  Border_ResetButton {
   widgetClassPath=&Border
   parent=SizeBox_CBox__ResetButton
   widget=Border_ResetButton
   properties {
    horizontalAlignment=HAlign_Center
    verticalAlignment=VAlign_Center
    background {
     bIsDynamicallyLoaded=false
     drawAs=RoundedBox
     imageType=NoImage
     mirroring=NoMirror
     resourceName=None
     resourceObject=null
     tiling=NoTile
     imageSize=<32 32>
     margin=<ltrb 0 0 0 0>
     outlineSettings {
      bUseBrushTransparency=false
      roundingType=FixedRadius
      width=1
      color {colorUseRule=UseColor_Specified, specifiedColor=<rgba 0 0.7126 1 1>}
      cornerRadii=<5 5 5 5>
     }
     tintColor {colorUseRule=UseColor_Specified, specifiedColor=<rgba 0.0395 0.0395 0.0395 1>}
     uVRegion {bIsValid=false, max=<0 0>, min=<0 0>}
    }
    padding=<ltrb 0 0 0 0>
   }
   slot {class=SizeBoxSlot}
  }
  UI_Button_ResetButton {
   widgetClassPath=&EditorUtilityButton
   parent=Border_ResetButton
   widget=UI_Button_ResetButton
   properties {widgetStyle {
     clickedSlateSound {resourceObject=null}
     disabled {
      bIsDynamicallyLoaded=false
      drawAs=RoundedBox
      imageType=NoImage
      mirroring=NoMirror
      resourceName=None
      resourceObject=null
      tiling=NoTile
      imageSize=<0 0>
      margin=<ltrb 0 0 0 0>
      outlineSettings {
       bUseBrushTransparency=false
       roundingType=FixedRadius
       width=1
       color {colorUseRule=UseColor_ColorTable, specifiedColor=<rgba 0 0 0 0>}
       cornerRadii=<4 4 4 4>
      }
      tintColor {colorUseRule=UseColor_Specified, specifiedColor=<rgba 0.0395 0.0395 0.0395 1>}
      uVRegion {bIsValid=false, max=<0 0>, min=<0 0>}
     }
     disabledForeground {colorUseRule=UseColor_ColorTable, specifiedColor=<rgba 0 0 0 0>}
     hovered {
      bIsDynamicallyLoaded=false
      drawAs=RoundedBox
      imageType=FullColor
      mirroring=NoMirror
      resourceName=None
      resourceObject=/Game/Athena/UI/InGame/Creative/PropertyEditor/Icon_Picker/Textures_V2/64x64/T_UI_IconLibrary_Reset_64
      tiling=NoTile
      imageSize=<34 34>
      margin=<ltrb 0 0 0 0>
      outlineSettings {
       bUseBrushTransparency=false
       roundingType=FixedRadius
       width=1
       color {colorUseRule=UseColor_ColorTable, specifiedColor=<rgba 0 0 0 0>}
       cornerRadii=<4 4 4 4>
      }
      tintColor {colorUseRule=UseColor_Specified, specifiedColor=<rgba 0.0953 0.0953 0.0953 1>}
      uVRegion {bIsValid=false, max=<0 0>, min=<0 0>}
     }
     hoveredForeground {colorUseRule=UseColor_ColorTable, specifiedColor=<rgba 0 0 0 0>}
     hoveredSlateSound {resourceObject=null}
     normal {
      bIsDynamicallyLoaded=false
      drawAs=Image
      imageType=FullColor
      mirroring=NoMirror
      resourceName=None
      resourceObject=&T_UI_IconLibrary_Reset_64
      tiling=NoTile
      imageSize=<34 34>
      margin=<ltrb 0 0 0 0>
      outlineSettings {
       bUseBrushTransparency=false
       roundingType=FixedRadius
       width=1
       color {colorUseRule=UseColor_Specified, specifiedColor=<rgba 0 0.7126 1 1>}
       cornerRadii=<4 4 4 4>
      }
      tintColor {colorUseRule=UseColor_Specified, specifiedColor=<rgba 0 0.4944 0.5913 1>}
      uVRegion {bIsValid=false, max=<0 0>, min=<0 0>}
     }
     normalForeground {colorUseRule=UseColor_ColorTable, specifiedColor=<rgba 0 0 0 0>}
     normalPadding=<ltrb 12 1.5 12 1.5>
     pressed {
      bIsDynamicallyLoaded=false
      drawAs=RoundedBox
      imageType=NoImage
      mirroring=NoMirror
      resourceName=None
      resourceObject=null
      tiling=NoTile
      imageSize=<32 32>
      margin=<ltrb 0 0 0 0>
      outlineSettings {
       bUseBrushTransparency=false
       roundingType=FixedRadius
       width=1
       color {colorUseRule=UseColor_ColorTable, specifiedColor=<rgba 0 0 0 0>}
       cornerRadii=<4 4 4 4>
      }
      tintColor {colorUseRule=UseColor_Specified, specifiedColor=<rgba 0.0284 0.0284 0.0284 1>}
      uVRegion {bIsValid=false, max=<0 0>, min=<0 0>}
     }
     pressedForeground {colorUseRule=UseColor_ColorTable, specifiedColor=<rgba 0 0 0 0>}
     pressedPadding=<ltrb 12 2.5 12 0.5>
     pressedSlateSound {resourceObject=null}
    }}
   slot {class=BorderSlot, properties {horizontalAlignment=HAlign_Center, verticalAlignment=VAlign_Center, padding=<ltrb 0 0 0 0>}}
  }
 ]
}
