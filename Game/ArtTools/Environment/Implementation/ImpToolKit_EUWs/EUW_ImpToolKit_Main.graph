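# Editor Utility Widget serving as an environment art implementation toolkit for setting up and checking static meshes and their prop Blueprints.
# Picking either asset finds its counterparts through Asset Registry references, then displays Blueprint info and pass/fail Nanite and LOD checks.
# Buttons import LODs, call other tools through their class default objects, or run inline Python that edits Blueprint defaults.
# GetAssetsToFix feeds most tools from Single Asset or Selection mode; build-placement scripts always use the browser selection.
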EditorUtilityWidgetBlueprint /Game/ArtTools/Environment/Implementation/ImpToolKit_EUWs/EUW_ImpToolKit_Main
graphs [
 ClearAssets {nodes [
   ClearAssets() {is=title K2Node_FunctionEntry}
   > n1 = Sequence()
   > Set Asset_Mesh(Asset_Mesh:Static Mesh Object Reference) -> (Output_Get:Static Mesh Object Reference)
   n2 = Get EUW_Asset_Mesh() -> (EUW_Asset_Mesh:EUW Imp Tool Kit Asset Object Reference)
   n3 = Get AssetThumbnail!(self:EUW Imp Tool Kit Asset Object Reference=@n2) -> (AssetThumbnail:Asset Thumbnail Widget Object Reference)
   > SetAssetByObject!(execute, self:Asset Thumbnail Widget Object Reference=@n3, Object:Object Reference) -> (then)
   > SetVisibility(self=@n3, InVisibility=Hidden)
   > ClearOptions(self:"Combo Box (String) Object Reference"=Get AssetOptions!(self:EUW Imp Tool Kit Asset Object Reference=@n2) -> (AssetOptions:Editor Utility Combo Box String Object Reference))
   Set Asset_BP(execute=@n1.then_1, Asset_BP:Object Reference) -> (Output_Get:Object Reference)
   n4 = Get EUW_Asset_BP() -> (EUW_Asset_BP:EUW Imp Tool Kit Asset Object Reference)
   n5 = Get AssetThumbnail(self=@n4)
   > SetAssetByObject(self=@n5)
   > SetVisibility(self=@n5, InVisibility=Hidden)
   > ClearOptions(self=Get AssetOptions(self=@n4))
  ]}
 EventGraph {nodes [
   "On Clicked (BP_Button_SetCentroid)"() {is=title K2Node_ComponentBoundEvent}
   > n1 = GetAssetsToFix!(execute, self, Type:ENUM_ImpToolKit_Type Enum=NewEnumerator1) -> (then, Assets:Array of Object References)
   > Execute Python Script!(
    execute
    PythonScript="import unreal\r\n\r\ndef run():\r\n\r\n    for asset in selected_assets:\r\n        asset_name = asset.get_name()\r\n\r\n        if not isinstance(asset, unreal.Blueprint):\r\n            unreal.log_warning(f\"Skipping '{asset_name}' - not a Blueprint.\")\r\n            continue\r\n\r\n        ## Build the CDO object path: \"{AssetPath}.Default__{AssetName}_C\"\r\n        asset_path = asset.get_path_name().split(\".\")[0]\r\n        cdo_path = f\"{asset_path}.Default__{asset_name}_C\"\r\n\r\n        cdo = unreal.load_object(None, cdo_path)\r\n        if cdo is None:\r\n            unreal.log_warning(f\"Skipping '{asset_name}' - could not load CDO at: {cdo_path}\")\r\n            continue\r\n\r\n        static_mesh = cdo.get_editor_property(\"StaticMesh\")\r\n        if static_mesh is None:\r\n            unreal.log_warning(f\"Skipping '{asset_name}' - no StaticMesh property found.\")\r\n            continue\r\n\r\n        ## get_bounds().origin is the bounding box center in local space (double-precision Vector)\r\n        ## CentroidOffset is Vector3f (single-precision), so we must cast explicitly\r\n        origin = static_mesh.get_bounds().origin\r\n        center = unreal.Vector3f(origin.x, origin.y, origin.z)\r\n\r\n        cdo.set_editor_property(\"CentroidOffset\", center)\r\n\r\n        unreal.EditorAssetLibrary.save_loaded_asset(asset)\r\n\r\n        unreal.log(\r\n            f\"'{asset_name}': CentroidOffset set to \"\r\n            f\"({center.x:.2f}, {center.y:.2f}, {center.z:.2f})\"\r\n        )\r\n\r\nrun()\r\n"
    selected_assets:Array of Object References=@n1
   ) -> (then, ReturnValue) {inputs [selected_assets]}
   n2 = "On Expansion Changed (ExpandableArea_Step_Perf)"!() -> (OutputDelegate, then, Area:Expandable Area Object Reference, bIsExpanded:bool) {is=title K2Node_ComponentBoundEvent}
   > ExpandStepInfoMenus!(execute, self, Condition:bool=@n2.bIsExpanded) -> (then)
   "On Clicked (SM_Button_ImportLODs)"() {is=title K2Node_ComponentBoundEvent}
   > ImportLODsTool()
   "On Clicked (SM_Button_ReduceNanite)"() {is=title K2Node_ComponentBoundEvent}
   > n3 = GetBlueprintAsset(Object=/Game/ArtTools/Environment/Implementation/ReduceNanite/ReduceNanite)
   > n4 = GeneratedClass!(execute, BlueprintObj:Blueprint Object Reference=@n3) -> (then, ReturnValue:Object Class Reference)
   > n5 = GetClassDefaultObject!(execute, Class:Object Class Reference=@n4) -> (then, ReturnValue:Object Reference)
   > n6 = Cast To ReduceNanite!(execute, Object=@n5) -> (then, CastFailed, AsReduce Nanite:Reduce Nanite Object Reference) {is=title K2Node_DynamicCast}
   > n7 = GetAssetsToFix(Type=NewEnumerator0)
   > ReduceNanite!(execute, self:Reduce Nanite Object Reference=@n6, Assets:Array of Object References=@n7) -> (then)
   "On Clicked (BP_Button_CreateBSGVersion)"() {is=title K2Node_ComponentBoundEvent}
   > n8 = GetBlueprintAsset(Object=/Game/ArtTools/Environment/Implementation/CreateBSGVersion/AAU_CreateBSGVersion)
   > n9 = GeneratedClass(BlueprintObj=@n8)
   > n10 = GetClassDefaultObject(Class=@n9)
   > n11 = Cast To AAU_CreateBSGVersion!(execute, Object=@n10) -> (then, CastFailed, AsAAU Create BSGVersion:AAU Create BSGVersion Object Reference) {is=title K2Node_DynamicCast}
   > n12 = GetAssetsToFix(Type=NewEnumerator1)
   > CreateBSGVersion_Function!(execute, self:AAU Create BSGVersion Object Reference=@n11, TargetArray:Array of Object References=@n12) -> (then)
   "On Clicked (Button_GetBP)"() {is=title K2Node_ComponentBoundEvent}
   > RetrieveBP()
   > GetBPInfo()
   > GetSMDependencies()
   > GetMeshInfo()
   "On Clicked (Button_GetMesh)"() {is=title K2Node_ComponentBoundEvent}
   > RetrieveSM()
   > GetMeshInfo()
   > GetBPReferences()
   > GetBPInfo()
   "On Clicked (BP_Button_AutoPropBlueprint)"() {is=title K2Node_ComponentBoundEvent}
   > n13 = GetBlueprintAsset(Object=/Game/ArtTools/Environment/Implementation/AutoPropSetup/AAU_AutoPropSetup)
   > n14 = GeneratedClass(BlueprintObj=@n13)
   > n15 = GetClassDefaultObject(Class=@n14)
   > n16 = Cast To AAU_AutoPropSetup!(execute, Object=@n15) -> (then, CastFailed, AsAAU Auto Prop Setup:AAU Auto Prop Setup Object Reference) {is=title K2Node_DynamicCast}
   > n17 = GetAssetsToFix(Type=NewEnumerator0)
   > AutoBlueprint_Function!(execute, self:AAU Auto Prop Setup Object Reference=@n16, Assets:Array of Object References=@n17) -> (then)
   "On Clicked (Button_ClearAssets)"() {is=title K2Node_ComponentBoundEvent}
   > ClearAssets() {is=title K2Node_CallFunction}
   > GetMeshInfo()
   > GetBPInfo()
   n18 = "On Expansion Changed (ExpandableArea_Step_LOD)"!() -> (OutputDelegate, then, Area:Expandable Area Object Reference, bIsExpanded:bool) {is=title K2Node_ComponentBoundEvent}
   > ExpandStepInfoMenus(Condition=@n18.bIsExpanded)
   n19 = "On Selection Changed (ComboBox_ToolMode)"!() -> (OutputDelegate, then, SelectedItem:String, SelectionType:ESelectInfo Enum) {is=title K2Node_ComponentBoundEvent}
   > n20 = Branch(Condition="Equal Exactly (String)"(A=@n19.SelectedItem, B=Single Asset))
   > n21 = Set ToolMode(ToolMode:ENUM_ImpToolKit_Modes Enum=NewEnumerator0) -> (Output_Get:ENUM_ImpToolKit_Modes Enum=NewEnumerator0)
   UpdateToolMode(execute=[@n22.then, @n21.then])
   Branch(execute=@n20.else, Condition="Equal Exactly (String)"(A=@n19.SelectedItem, B=Selection))
   > n22 = Set ToolMode(ToolMode=NewEnumerator1) -> (Output_Get=NewEnumerator0)
   n23 = "On Expansion Changed (ExpandableArea_Step_BPSetup)"!() -> (OutputDelegate, then, Area:Expandable Area Object Reference, bIsExpanded:bool) {is=title K2Node_ComponentBoundEvent}
   > ExpandStepInfoMenus(Condition=@n23.bIsExpanded)
   "On Clicked (SM_Button_AssignLODs)"() {is=title K2Node_ComponentBoundEvent}
   > n24 = GetBlueprintAsset(Object=/Game/ArtTools/Environment/Implementation/ApplyPerfSettings/AAU_ApplyPerfSettings)
   > n25 = GeneratedClass(BlueprintObj=@n24)
   > n26 = GetClassDefaultObject(Class=@n25)
   > n27 = Cast To AAU_ApplyPerfSettings!(execute, Object=@n26) -> (then, CastFailed, AsAAU Apply Perf Settings:AAU Apply Perf Settings Object Reference) {is=title K2Node_DynamicCast}
   > n28 = GetAssetsToFix(Type=NewEnumerator0)
   > ApplyPerfSettings!(
    execute
    self:AAU Apply Perf Settings Object Reference=@n27
    Enable Nanite:bool=true
    PreserveTangentSettings:bool=true
    Explicit Tangents:bool=false
    Fallback Target:ENaniteFallbackTarget Enum=RelativeError
    Fallback Relative Error:float=0
    Assets:Array of Object References=@n28
   ) -> (then)
   "On Clicked (BP_Button_BuildThrough)"() {is=title K2Node_ComponentBoundEvent}
   > Execute Python Script!(
    execute
    PythonScript="import unreal\r\n\r\nBLOCKED_FOLDER = \"/Game/Environments/Blockout\"\r\nBLOCKED_KEYWORD = \"Sets\"\r\n\r\ndef show_popup(message):\r\n    unreal.EditorDialog.show_message(\r\n        title=\"PropSetter - Assets Skipped\",\r\n        message=message,\r\n        message_type=unreal.AppMsgType.OK\r\n    )\r\n\r\ndef run():\r\n    selected_assets = unreal.EditorUtilityLibrary.get_selected_assets()\r\n    if not selected_assets:\r\n        unreal.log_warning(\"[PropSetter] No assets selected. Please select one or more Blueprints in the Content Browser.\")\r\n        return\r\n\r\n    skipped_assets = []\r\n\r\n    for asset in selected_assets:\r\n        asset_name = asset.get_name()\r\n        asset_path = asset.get_path_name().split(\".\")[0]\r\n\r\n        ## Guard: skip anything under the blockout folder\r\n        if asset_path.startswith(BLOCKED_FOLDER):\r\n            reason = f\"Located under {BLOCKED_FOLDER}\"\r\n            unreal.log_warning(f\"[PropSetter] Skipping '{asset_name}' - path is under protected folder: {BLOCKED_FOLDER}\")\r\n            skipped_assets.append((asset_name, reason))\r\n            continue\r\n\r\n        ## Guard: skip anything with \"Sets\" in the path\r\n        if BLOCKED_KEYWORD in asset_path:\r\n            reason = f\"Path contains blocked keyword '{BLOCKED_KEYWORD}'\"\r\n            unreal.log_warning(f\"[PropSetter] Skipping '{asset_name}' - path contains blocked keyword: '{BLOCKED_KEYWORD}'\")\r\n            skipped_assets.append((asset_name, reason))\r\n            continue\r\n\r\n        if not isinstance(asset, unreal.Blueprint):\r\n            unreal.log_warning(f\"[PropSetter] Skipping '{asset_name}' - not a Blueprint.\")\r\n            continue\r\n\r\n        ## Build the CDO object path: \"{AssetPath}.Default__{AssetName}_C\"\r\n        cdo_path = f\"{asset_path}.Default__{asset_name}_C\"\r\n        cdo = unreal.load_object(None, cdo_path)\r\n        if cdo is None:\r\n            unreal.log_warning(f\"[PropSetter] Skipping '{asset_name}' - could not load CDO at: {cdo_path}\")\r\n            continue\r\n\r\n        cdo.set_editor_property(\"bDoNotBlockBuildings\", True)\r\n        cdo.set_editor_property(\"bDestroyOnPlayerBuildingPlacement\", False)\r\n        unreal.EditorAssetLibrary.save_loaded_asset(asset)\r\n        unreal.log(f\"[PropSetter] Done: '{asset_name}' | bDoNotBlockBuildings=True, bDestroyOnPlayerBuildingPlacement=False\")\r\n\r\n    ## Show a single combined popup for all skipped protected assets\r\n    if skipped_assets:\r\n        lines = [\"The following assets are protected and were not modified:\\n\"]\r\n        for name, reason in skipped_assets:\r\n            lines.append(f\"  - {name}\\n    Reason: {reason}\")\r\n        show_popup(\"\\n\".join(lines))\r\n\r\nrun()\r\n"
   ) -> (then, ReturnValue)
   "On Clicked (BP_Button_DestroyOnBuild)"() {is=title K2Node_ComponentBoundEvent}
   > Execute Python Script(PythonScript="import unreal\r\n\r\nBLOCKED_FOLDER = \"&Blockout\"\r\nBLOCKED_KEYWORD = \"Sets\"\r\n\r\ndef show_popup(message):\r\n    unreal.EditorDialog.show_message(\r\n        title=\"PropSetter - Assets Skipped\",\r\n        message=message,\r\n        message_type=unreal.AppMsgType.OK\r\n    )\r\n\r\ndef run():\r\n    selected_assets = unreal.EditorUtilityLibrary.get_selected_assets()\r\n    if not selected_assets:\r\n        unreal.log_warning(\"[PropSetter] No assets selected. Please select one or more Blueprints in the Content Browser.\")\r\n        return\r\n\r\n    skipped_assets = []\r\n\r\n    for asset in selected_assets:\r\n        asset_name = asset.get_name()\r\n        asset_path = asset.get_path_name().split(\".\")[0]\r\n\r\n        ## Guard: skip anything under the blockout folder\r\n        if asset_path.startswith(BLOCKED_FOLDER):\r\n            reason = f\"Located under {BLOCKED_FOLDER}\"\r\n            unreal.log_warning(f\"[PropSetter] Skipping '{asset_name}' - path is under protected folder: {BLOCKED_FOLDER}\")\r\n            skipped_assets.append((asset_name, reason))\r\n            continue\r\n\r\n        ## Guard: skip anything with \"Sets\" in the path\r\n        if BLOCKED_KEYWORD in asset_path:\r\n            reason = f\"Path contains blocked keyword '{BLOCKED_KEYWORD}'\"\r\n            unreal.log_warning(f\"[PropSetter] Skipping '{asset_name}' - path contains blocked keyword: '{BLOCKED_KEYWORD}'\")\r\n            skipped_assets.append((asset_name, reason))\r\n            continue\r\n\r\n        if not isinstance(asset, unreal.Blueprint):\r\n            unreal.log_warning(f\"[PropSetter] Skipping '{asset_name}' - not a Blueprint.\")\r\n            continue\r\n\r\n        ## Build the CDO object path: \"{AssetPath}.Default__{AssetName}_C\"\r\n        cdo_path = f\"{asset_path}.Default__{asset_name}_C\"\r\n        cdo = unreal.load_object(None, cdo_path)\r\n        if cdo is None:\r\n            unreal.log_warning(f\"[PropSetter] Skipping '{asset_name}' - could not load CDO at: {cdo_path}\")\r\n            continue\r\n\r\n        cdo.set_editor_property(\"bDoNotBlockBuildings\", False)\r\n        cdo.set_editor_property(\"bDestroyOnPlayerBuildingPlacement\", True)\r\n        unreal.EditorAssetLibrary.save_loaded_asset(asset)\r\n        unreal.log(f\"[PropSetter] Done: '{asset_name}' | bDoNotBlockBuildings=False, bDestroyOnPlayerBuildingPlacement=True\")\r\n\r\n    ## Show a single combined popup for all skipped protected assets\r\n    if skipped_assets:\r\n        lines = [\"The following assets are protected and were not modified:\\n\"]\r\n        for name, reason in skipped_assets:\r\n            lines.append(f\"  - {name}\\n    Reason: {reason}\")\r\n        show_popup(\"\\n\".join(lines))\r\n\r\nrun()\r\n")
   n29 = "On Check State Changed (LOD1_CheckBox)"!() -> (OutputDelegate, then, bIsChecked:bool) {is=title K2Node_ComponentBoundEvent}
   LODCheckBoxes(execute=[@n29.then, @n30.then, @n31.then])
   n30 = "On Check State Changed (LOD2_CheckBox)"!() -> (OutputDelegate, then, bIsChecked:bool) {is=title K2Node_ComponentBoundEvent}
   n31 = "On Check State Changed (LOD3_CheckBox)"!() -> (OutputDelegate, then, bIsChecked:bool) {is=title K2Node_ComponentBoundEvent}
   n32 = SM Selection Change!() -> (OutputDelegate, then, SelectedItem:String) {is=title K2Node_CustomEvent}
   > PrintString(InString=@n32.SelectedItem)
   n33 = Find(
    TargetMap:Map of Strings to Static Mesh Object References=Get SM_Dependencies Map() -> (SM_Dependencies Map:Map of Strings to Static Mesh Object References)
    Key=@n32.SelectedItem
   ) -> (Value:Static Mesh Object Reference)
   > Branch(Condition=@n33.ReturnValue)
   > n34 = Set Asset_Mesh(Asset_Mesh=@n33)
   > GetMeshInfo()
   > SetAssetByObject(self=Get AssetThumbnail(self=Get EUW_Asset_Mesh()), Object=@n34)
   n35 = BP_ButtonEdit() {is=title K2Node_CustomEvent}
   > OpenEditorForAssets!(
    execute
    self:Asset Editor Subsystem Object Reference=Get AssetEditorSubsystem()
    Assets:Array of Object References=Make Array("[0]":Object Reference=Get Asset_BP() -> (Asset_BP:Object Reference)) -> (Array:Array of Object References)
    OpenedMethod:EAssetTypeActivationOpenedMethod Enum=Edit
   ) -> (then, ReturnValue:bool)
   n36 = BP Selection Change!() -> (OutputDelegate, then, SelectedItem:String) {is=title K2Node_CustomEvent}
   > PrintString(InString=@n36.SelectedItem)
   n37 = Find(
    TargetMap:Map of Strings to Blueprint Object References=Get BP_References Map() -> (BP_References Map:Map of Strings to Blueprint Object References)
    Key=@n36.SelectedItem
   ) -> (Value:Blueprint Object Reference)
   > Branch(Condition=@n37.ReturnValue)
   > n38 = Set Asset_BP(Asset_BP=@n37)
   > GetBPInfo()
   > SetAssetByObject(self=Get AssetThumbnail(self=Get EUW_Asset_BP()), Object=@n38)
   n39 = BP_ButtonBrowse() {is=title K2Node_CustomEvent}
   > n40 = GetPathNameForLoadedAsset(LoadedAsset=Get Asset_BP())
   > SyncBrowserToObjects(AssetPaths=Make Array("[0]":String=@n40) -> (Array:Array of Strings))
   n41 = SM_ButtonEdit() {is=title K2Node_CustomEvent}
   > OpenEditorForAssets(
    self=Get AssetEditorSubsystem()
    Assets=Make Array("[0]":Object Reference=Get Asset_Mesh() -> (Asset_Mesh:Static Mesh Object Reference)) -> (Array:Array of Object References)
    OpenedMethod=Edit
   )
   n42 = SM_ButtonBrowse() {is=title K2Node_CustomEvent}
   > n43 = GetPathNameForLoadedAsset(LoadedAsset=Get Asset_Mesh())
   > SyncBrowserToObjects(AssetPaths=Make Array("[0]":String=@n43) -> (Array:Array of Strings))
   Event PreConstruct()
   > SetObject(self=Get LOD1Path_Display() -> (LOD1Path_Display:Single Property View Object Reference), NewObject=Self-Reference())
   > SetObject(self=Get LOD2Path_Display() -> (LOD2Path_Display:Single Property View Object Reference), NewObject=Self-Reference())
   > SetObject(self=Get LOD3Path_Display() -> (LOD3Path_Display:Single Property View Object Reference), NewObject=Self-Reference())
   > n44 = Sequence!(execute) -> (then_0, then_1, then_2:Exec, then_3:Exec, then_4:Exec, then_5:Exec)
   > Bind Event to On Asset Selection Changed(self:EUW Imp Tool Kit Asset Object Reference=Get EUW_Asset_BP(), Delegate=@n36) {is=title K2Node_AddDelegate}
   Bind Event to On Asset Selection Changed(execute=@n44.then_1, self=Get EUW_Asset_Mesh(), Delegate=@n32)
   Bind Event to On Clicked Edit(execute=@n44.then_4, self:EUW Imp Tool Kit Asset Object Reference=Get EUW_Asset_BP(), Delegate=@n35) {is=title K2Node_AddDelegate}
   Bind Event to On Clicked Browse To(execute=@n44.then_5, self:EUW Imp Tool Kit Asset Object Reference=Get EUW_Asset_BP(), Delegate=@n39) {is=title K2Node_AddDelegate}
   Bind Event to On Clicked Edit(execute=@n44.then_2, self=Get EUW_Asset_Mesh(), Delegate=@n41)
   Bind Event to On Clicked Browse To(execute=@n44.then_3, self=Get EUW_Asset_Mesh(), Delegate=@n42)
   Event Construct()
   Event Tick()
  ]}
 ExpandStepInfoMenus {nodes [
   n1 = ExpandStepInfoMenus!() -> (then, Condition:bool=~) {is=title K2Node_FunctionEntry}
   > n2 = Branch(Condition=@n1)
   n3 = Make Array!(
    "[0]":Editor Utility Expandable Area Object Reference=Get ExpandableArea_Step_Perf() -> (ExpandableArea_Step_Perf:Editor Utility Expandable Area Object Reference)
    "[1]":Editor Utility Expandable Area Object Reference=Get ExpandableArea_Step_BPSetup() -> (ExpandableArea_Step_BPSetup:Editor Utility Expandable Area Object Reference)
    "[2]":Editor Utility Expandable Area Object Reference=Get ExpandableArea_Step_LOD() -> (ExpandableArea_Step_LOD:Editor Utility Expandable Area Object Reference)
   ) -> (Array:Array of Editor Utility Expandable Area Object References)
   > n4 = For Each Loop(Array:Array of Editor Utility Expandable Area Object References=@n3) -> (Array Element:Editor Utility Expandable Area Object Reference)
   > SetIsExpanded!(execute, self:Expandable Area Object Reference=@n4, IsExpanded:bool=true) -> (then)
   n5 = For Each Loop(Exec=@n2.else, Array=@n3)
   > SetIsExpanded(self=@n5, IsExpanded=false)
  ]}
 GetAssetsToFix {nodes [
   GetAssetsToFix!() -> (then, Type:ENUM_ImpToolKit_Type Enum=NewEnumerator0) {is=title K2Node_FunctionEntry}
   > Clear(TargetArray:Array of Object References=Get Assets To Fix() -> (Assets To Fix:Array of Object References))
   > n1 = Switch on ENUM_ImpToolKit_Modes!(execute, Selection:ENUM_ImpToolKit_Modes Enum=Get ToolMode() -> (ToolMode:ENUM_ImpToolKit_Modes Enum=NewEnumerator0)) -> (NewEnumerator0:Exec, NewEnumerator1:Exec) {is=title K2Node_SwitchEnum}
   n2 = GetSelectedAssets(execute=@n1.NewEnumerator1)
   > n3 = Set Assets To Fix(Assets To Fix:Array of Object References=@n2) -> (Output_Get:Array of Object References)
   Return Node!(execute=[@n5.then, @n3.then, @n6.then], Assets:Array of Object References=Get Assets To Fix())
   n4 = Switch on ENUM_ImpToolKit_Type!(execute=@n1.NewEnumerator0, Selection:ENUM_ImpToolKit_Type Enum=Get Type() -> (Type:ENUM_ImpToolKit_Type Enum=NewEnumerator0)) -> (NewEnumerator0:Exec, NewEnumerator1:Exec) {is=title K2Node_SwitchEnum}
   n5 = Set Assets To Fix(execute=@n4.NewEnumerator1, Assets To Fix=Make Array!("[0]":Object Reference=Get Asset_BP()) -> (Array:Array of Object References))
   n6 = Set Assets To Fix(execute=@n4.NewEnumerator0, Assets To Fix=Make Array("[0]"=Get Asset_Mesh()))
  ]}
 GetBPInfo {nodes [
   GetBPInfo() {is=title K2Node_FunctionEntry}
   > n1 = For Each Loop(Array:Array of Text Object References=Make Array!(
     "[0]":Text Object Reference
     "[1]":Text Object Reference=Get BP_Info_ParentBP() -> (BP_Info_ParentBP:Text Object Reference)
     "[2]":Text Object Reference=Get BP_Info_ResourceTable() -> (BP_Info_ResourceTable:Text Object Reference)
     "[3]":Text Object Reference=Get BP_Info_Resource() -> (BP_Info_Resource:Text Object Reference)
    ) -> (Array:Array of Text Object References)) -> (Array Element:Text Object Reference)
   Branch(execute=@n1.Completed, Condition=IsValid(Object=Get Asset_BP()))
   > n2 = Sequence!(execute) -> (then_0, then_1, then_2)
   n3 = GetBlueprintAsset(execute=@n2.then_1, Object=Get Asset_BP())
   > n4 = ParentClass!(execute, BlueprintObj:Blueprint Object Reference=@n3) -> (then, ReturnValue:Object Class Reference)
   > Set Text(Text="To Text (String)"(InString=Get Class Display Name(Class=@n4)), self=Get BP_Info_ParentBP())
   n5 = GetBlueprintAsset(execute=@n2.then_2, Object=Get Asset_BP())
   > n6 = GeneratedClass(BlueprintObj=@n5)
   > n7 = GetClassDefaultObject(Class=@n6)
   > n8 = Cast To BuildingActor!(execute, Object=@n7) -> (then, CastFailed, AsBuilding Actor:Building Actor Object Reference) {is=title K2Node_DynamicCast}
   > n9 = GetEditorProperty!(execute:Exec, Object:Object Reference=@n8, PropertyName:Name=resourceType) -> (then:Exec, PropertyValue:EFortResourceType Enum, ReturnValue:bool)
   > n10 = Set Resource(Resource:EFortResourceType Enum=@n9) -> (Output_Get:EFortResourceType Enum=Wood)
   > Set Text(
    Text="To Text (String)"(InString=Enum to String!(Enumerator:EFortResourceType Enum=@n10) -> (ReturnValue:String))
    self=Get BP_Info_Resource()
   )
   Set Text(execute=@n1.LoopBody, self=@n1)
  ]}
 GetBPReferences {nodes [
   GetBPReferences() {is=title K2Node_FunctionEntry}
   > Clear!(execute, TargetMap:Map of Strings to Blueprint Object References=Get BP_References Map()) -> (then) {is=title K2Node_CallFunction}
   > Clear(TargetArray=Get BP_References_Found() -> (BP_References_Found:Array of Object References)) {is=title K2Node_CallArrayFunction}
   > n1 = GetPathNameForLoadedAsset(LoadedAsset=Get Asset_Mesh())
   > n2 = Get Referencers!(
    execute
    self:Asset Registry Interface=GetAssetRegistry()
    PackageName:Name=String To Name(InString=Split(SourceString=@n1, InStr=., SearchDir=FromEnd))
    ReferenceOptions:"Asset Registry Dependency Options Structure (by ref)"=Get Reference Options() -> (Reference Options:Asset Registry Dependency Options Structure)
   ) -> (then, OutReferencers:Array of Names, ReturnValue:bool) {is=title K2Node_CallFunction}
   > n3 = For Each Loop(Array:Array of Names=@n2) -> (Array Element:Name)
   n4 = "To String (Name)"(InName=@n3)
   > n5 = Branch(Condition=Contains(SearchIn=@n4, Substring=FNEC_))
   n6 = LoadAsset(execute=@n5.else, AssetPath=@n4)
   > Add(TargetArray:Array of Object References=Get BP_References_Found(), NewItem:"Object Reference (by ref)"=@n6) {is=title K2Node_CallArrayFunction}
   n7 = Filter by Class!(
    execute=@n3.Completed
    TargetArray:Array of Object References=Get BP_References_Found()
    ObjectClass:Object Class Reference=/Script/Engine.Blueprint
    FilterType:EEditorScriptingFilterType Enum
   ) -> (then, ReturnValue:Array of Blueprint Object References)
   > ClearOptions(self=Get AssetOptions(self=Get EUW_Asset_BP()))
   > n8 = For Each Loop(Array:Array of Blueprint Object References=@n7) -> (Array Element:Blueprint Object Reference)
   SetVisibility(execute=@n8.Completed, self=Get AssetThumbnail(self=Get EUW_Asset_BP()), InVisibility=Visible)
   > n9 = Branch(Condition="integer > integer"(A=Length(TargetArray:Array of Blueprint Object References=@n7), B=1))
   n10 = Get AssetOptions(self=Get EUW_Asset_BP())
   > Set bIsEnabled!(execute, bIsEnabled:bool=true, self:Widget Object Reference=@n10) -> (then, Output_Get:bool)
   Set bIsEnabled(execute=@n9.else, bIsEnabled=false, self=@n10)
   n11 = GetDisplayName(Object=@n8)
   Add!(
    execute=@n8.LoopBody
    TargetMap:Map of Strings to Blueprint Object References=Get BP_References Map()
    Key:"String (by ref)"=@n11
    Value:"Blueprint Object Reference (by ref)"=@n8
   ) -> (then) {is=title K2Node_CallFunction}
   n12 = Get EUW_Asset_BP()
   n13 = Get AssetOptions(self=@n12)
   > AddOption!(execute, self:"Combo Box (String) Object Reference"=@n13, Option:String=@n11) -> (then)
   > Branch(Condition="Equal (Integer)"(A=@n8.Array Index))
   > SetSelectedOption!(execute, self:"Combo Box (String) Object Reference"=@n13, Option:String=@n11) -> (then)
   > n14 = Set Asset_BP(Asset_BP=@n8)
   > SetAssetByObject(self=Get AssetThumbnail(self=@n12), Object=@n14)
   Reroute Node(InputPin:Wildcard) -> (OutputPin:Wildcard)
  ]}
 GetMeshInfo {nodes [
   GetMeshInfo() {is=title K2Node_FunctionEntry}
   > Clear(TargetArray:Array of Strings=Get LODsGenerated() -> (LODsGenerated:Array of Strings))
   > Set LODsGeneratedText(LODsGeneratedText:String) -> (Output_Get:String)
   > Clear(TargetArray:Array of Booleans=Get LODGeneratedPassArray() -> (LODGeneratedPassArray:Array of Booleans))
   > n1 = For Each Loop(Array:Array of Text Object References=Make Array!(
     "[0]"=Get Mesh_Info_NaniteEnabled() -> (Mesh_Info_NaniteEnabled:Text Object Reference)
     "[1]"=Get Mesh_Info_NaniteFallback() -> (Mesh_Info_NaniteFallback:Text Object Reference)
     "[2]"=Get Mesh_Info_RelativeError() -> (Mesh_Info_RelativeError:Text Object Reference)
     "[3]"=Get Mesh_Info_LODs() -> (Mesh_Info_LODs:Text Object Reference)
     "[4]":Text Object Reference=Get Mesh_Info_MinLODs() -> (Mesh_Info_MinLODs:Text Object Reference)
     "[5]":Text Object Reference=Get Mesh_Info_LODForCollision() -> (Mesh_Info_LODForCollision:Text Object Reference)
     "[6]":Text Object Reference=Get TaskStatus_Perf() -> (TaskStatus_Perf:Text Object Reference)
     "[7]":Text Object Reference=Get Mesh_Info_LOD1() -> (Mesh_Info_LOD1:Text Object Reference)
     "[8]":Text Object Reference=Get Mesh_Info_LOD2() -> (Mesh_Info_LOD2:Text Object Reference)
     "[9]":Text Object Reference=Get Mesh_Info_LOD3() -> (Mesh_Info_LOD3:Text Object Reference)
     "[10]":Text Object Reference=Get TaskStatus_LOD() -> (TaskStatus_LOD:Text Object Reference)
    ) -> (Array)) -> (Array Element:Text Object Reference)
   Branch(execute=@n1.Completed, Condition=IsValid(Object=Get Asset_Mesh()))
   > n2 = Sequence!(execute) -> (then_0, then_1, then_2, then_3:Exec, then_4:Exec, then_5:Exec, then_6:Exec)
   n3 = For Loop(execute=@n2.then_5, FirstIndex=1, LastIndex=3)
   > n4 = GetLodReductionSettings!(
    execute
    self:Static Mesh Editor Subsystem Object Reference=Get StaticMeshEditorSubsystem()
    StaticMesh:Static Mesh Object Reference=Get Asset_Mesh()
    LodIndex:int=@n3
   ) -> (then, OutReductionOptions:Mesh Reduction Settings Structure)
   > n5 = Branch(Condition="float < float"(
     A=Break Mesh Reduction Settings!(MeshReductionSettings:"Mesh Reduction Settings Structure (by ref)"=@n4) -> (
      PercentTriangles:float=~
      PercentVertices:float=~
      MaxDeviation:float=~
      PixelError:float=~
      WeldingThreshold:float=~
      HardAngleThreshold:float=~
      BaseLODModel:int
      SilhouetteImportance:EMeshFeatureImportance Enum
      TextureImportance:EMeshFeatureImportance Enum
      ShadingImportance:EMeshFeatureImportance Enum
      bRecalculateNormals:bool=~
      bGenerateUniqueLightmapUVs:bool=~
      bKeepSymmetry:bool=~
      bVisibilityAided:bool=~
      bCullOccluded:bool=~
      TerminationCriterion:EStaticMeshReductionTerimationCriterion Enum
      VisibilityAggressiveness:EMeshFeatureImportance Enum
      VertexColorImportance:EMeshFeatureImportance Enum
     ) {is=title K2Node_BreakStruct}
     B=1
    ))
   > Add(TargetArray:Array of Strings=Get LODsGenerated(), NewItem:"String (by ref)"=Auto-Generated) {is=title K2Node_CallArrayFunction}
   Add(execute=@n5.else, TargetArray=Get LODsGenerated(), NewItem=Custom)
   n6 = For Each Loop(Exec=@n3.Completed, Array:Array of Strings=Get LODsGenerated()) -> (Array Element:String)
   n7 = "Get (a copy)"(
    Array:Array of Text Object References=Make Array!("[0]"=Get Mesh_Info_LOD1(), "[1]"=Get Mesh_Info_LOD2(), "[2]"=Get Mesh_Info_LOD3()) -> (Array)
    Dimension 1=@n6.Array Index
   ) -> (Output:Text Object Reference)
   > Set Text(Text="To Text (String)"(InString=@n6), self=@n7)
   n8 = "Equal Exactly (String)"(A=@n6, B=Custom)
   > Add(TargetArray:Array of Booleans=Get LODGeneratedPassArray(), NewItem:"Boolean (by ref)"=@n8)
   > SetPassFailColor!(execute, self, Condition:bool=@n8, Text Box:Text Object Reference=@n7) -> (then)
   n9 = Get Asset_Mesh()
   Set Text(execute=@n2.then_2, Text="To Text (Integer)"(Value=GetNumLODs(self=@n9)), self=Get Mesh_Info_LODs())
   n10 = Break Mesh Nanite Settings!(MeshNaniteSettings:"Mesh Nanite Settings Structure (by ref)"=GetNaniteSettings!(self:Static Mesh Editor Subsystem Object Reference=Get StaticMeshEditorSubsystem(), StaticMesh:Static Mesh Object Reference=Get Asset_Mesh()) -> (ReturnValue:Mesh Nanite Settings Structure)) -> (
    bEnabled:bool=~
    bExplicitTangents:bool=~
    bLerpUVs:bool=~
    bSeparable:bool=~
    bVoxelNDF:bool=~
    bVoxelOpacity:bool=~
    ShapePreservation:ENaniteShapePreservation Enum
    PositionPrecision:int
    NormalPrecision:int
    TangentPrecision:int
    BoneWeightPrecision:int
    KeepPercentTriangles:float=~
    TrimRelativeError:float=~
    GenerateFallback:ENaniteGenerateFallback Enum
    FallbackTarget:ENaniteFallbackTarget Enum
    FallbackPercentTriangles:float=~
    FallbackRelativeError:float=~
    MaxEdgeLengthFactor:float=~
    NumRays:int
    VoxelLevel:int
    RayBackUp:float=~
    DisplacementUVChannel:int
   ) {is=title K2Node_BreakStruct}
   n11 = Get Mesh_Info_NaniteEnabled()
   Set Text(execute=@n2.then_1, Text="To Text (Boolean)"!(InBool:bool=@n10) -> (ReturnValue:Text), self=@n11)
   n12 = "Equal (Boolean)"(A:bool=@n10, B:bool=true) -> (ReturnValue:bool=~) {is=title K2Node_PromotableOperator}
   > SetPassFailColor(Condition=@n12, Text Box=@n11)
   n13 = Get Mesh_Info_NaniteFallback()
   > Set Text(Text="To Text (String)"(InString=Enum to String(Enumerator:ENaniteFallbackTarget Enum=@n10.FallbackTarget)), self=@n13)
   n14 = "Equal (Enum)"(A:ENaniteFallbackTarget Enum=@n10.FallbackTarget, B:ENaniteFallbackTarget Enum=RelativeError)
   > SetPassFailColor(Condition=@n14, Text Box=@n13)
   n15 = Get Mesh_Info_RelativeError()
   > Set Text(
    Text="To Text (Float)"(Value=@n10.FallbackRelativeError, MinimumIntegralDigits=1, MaximumIntegralDigits=324, MinimumFractionalDigits=0, MaximumFractionalDigits=3)
    self=@n15
   )
   n16 = "Equal (Float)"(A=@n10.FallbackRelativeError)
   > SetPassFailColor(Condition=@n16, Text Box=@n15)
   > n17 = Branch(Condition=AND Boolean!(A=@n12, B=@n14, C:bool=@n16) -> (ReturnValue))
   > Set NanitePass(NanitePass:bool=true) -> (Output_Get:bool)
   Set NanitePass(execute=@n17.else, NanitePass=false)
   n18 = GetMinLODForQualityLevels!(self:Static Mesh Object Reference=Get Asset_Mesh()) -> (QualityLevelMinimumLODs:Map of EPerQualityLevels Enums to Integers=Low, Default:int=0)
   n19 = Find(TargetMap:Map of EPerQualityLevels Enums to Integers=@n18, Key:"EPerQualityLevels Enum (by ref)"=Medium) -> (Value:int)
   n20 = Append(A="   Medium: ", B="To String (Integer)"(InInt=@n19))
   n21 = Find(TargetMap=@n18, Key=Low)
   n22 = Append(A=Append(A="Default: ", B="To String (Integer)"(InInt=@n18.Default)), B=Append(A="   Low: ", B="To String (Integer)"(InInt=@n21)))
   n23 = Get Mesh_Info_MinLODs()
   Set Text(execute=@n2.then_4, Text="To Text (String)"(InString=Append(A=@n22, B=@n20)), self=@n23)
   n24 = AND Boolean!(A="integer >= integer"(A=@n19, B=2), B="integer >= integer"(A=@n21, B=2)) -> (ReturnValue)
   > SetPassFailColor(Condition=@n24, Text Box=@n23)
   > n25 = Branch(Condition=@n24)
   > Set MinLODPass(MinLODPass:bool=true) -> (Output_Get:bool)
   Set MinLODPass(execute=@n25.else, MinLODPass=false)
   n26 = Sequence!(execute=@n2.then_6) -> (then_0, then_1)
   > SetTaskStatus!(
    execute
    self
    Text Box:Text Object Reference=Get TaskStatus_Perf()
    Condition:bool=AND Boolean!(
     A=Get NanitePass() -> (NanitePass:bool)
     B=Get MinLODPass() -> (MinLODPass:bool)
     C:bool=Get LODForCollisionPass() -> (LODForCollisionPass:bool)
    ) -> (ReturnValue)
   ) -> (then)
   n27 = Branch(execute=@n26.then_1, Condition=Contains Item(TargetArray:Array of Booleans=Get LODGeneratedPassArray(), ItemToFind:"Boolean (by ref)"))
   > n28 = Set LODSGenerated Pass(LODSGenerated Pass:bool=false) -> (Output_Get:bool)
   SetTaskStatus(execute=[@n29.then, @n28.then], Text Box=Get TaskStatus_LOD(), Condition=Get LODSGenerated Pass() -> (LODSGenerated Pass:bool))
   n29 = Set LODSGenerated Pass(execute=@n27.else, LODSGenerated Pass=true)
   n30 = Get LODForCollision!(self:Static Mesh Object Reference=Get Asset_Mesh()) -> (LODForCollision:int=0)
   n31 = Get Mesh_Info_LODForCollision()
   Set Text(execute=@n2.then_3, Text="To Text (Integer)"(Value=@n30), self=@n31)
   n32 = "integer >= integer"(A=@n30, B=2)
   > SetPassFailColor(Condition=@n32, Text Box=@n31)
   > n33 = Branch(Condition=@n32)
   > Set LODForCollisionPass(LODForCollisionPass:bool=true) -> (Output_Get:bool)
   Set LODForCollisionPass(execute=@n33.else, LODForCollisionPass=false)
   Set Text(execute=@n1.LoopBody, self=@n1)
  ]}
 GetSMDependencies {nodes [
   GetSMDependencies() {is=title K2Node_FunctionEntry}
   > Clear(TargetMap:Map of Strings to Static Mesh Object References=Get SM_Dependencies Map()) {is=title K2Node_CallFunction}
   > Clear(TargetArray:Array of Object References=Get SM_Dependencies_Found() -> (SM_Dependencies_Found:Array of Object References)) {is=title K2Node_CallArrayFunction}
   n1 = Get Asset_BP()
   > n2 = GetPathNameForLoadedAsset(LoadedAsset=@n1)
   > PrintString(InString=Append(A="TEST : ", B=GetDisplayName(Object=@n1)))
   > n3 = Get Dependencies!(
    execute
    self:Asset Registry Interface=GetAssetRegistry()
    PackageName:Name=String To Name(InString=Split(SourceString=@n2, InStr=., SearchDir=FromEnd))
    DependencyOptions:"Asset Registry Dependency Options Structure (by ref)"=Get Dependencies Options() -> (Dependencies Options:Asset Registry Dependency Options Structure)
   ) -> (then, OutDependencies:Array of Names, ReturnValue:bool) {is=title K2Node_CallFunction}
   > n4 = For Each Loop(Array:Array of Names=@n3) -> (Array Element:Name)
   > n5 = LoadAsset(AssetPath="To String (Name)"(InName=@n4))
   > Add(TargetArray:Array of Object References=Get SM_Dependencies_Found(), NewItem:"Object Reference (by ref)"=@n5)
   n6 = Filter by Class(execute=@n4.Completed, TargetArray=Get SM_Dependencies_Found(), ObjectClass=&StaticMesh) -> (ReturnValue:Array of Static Mesh Object References)
   > ClearOptions(self=Get AssetOptions(self=Get EUW_Asset_Mesh()))
   > n7 = For Each Loop(Array:Array of Static Mesh Object References=@n6) -> (Array Element:Static Mesh Object Reference)
   SetVisibility(execute=@n7.Completed, self=Get AssetThumbnail(self=Get EUW_Asset_Mesh()), InVisibility=Visible)
   > n8 = Branch(Condition="integer > integer"(A=Length(TargetArray:Array of Static Mesh Object References=@n6), B=1))
   n9 = Get AssetOptions(self=Get EUW_Asset_Mesh())
   > Set bIsEnabled(bIsEnabled=true, self=@n9)
   Set bIsEnabled(execute=@n8.else, bIsEnabled=false, self=@n9)
   n10 = GetDisplayName(Object=@n7)
   Add(
    execute=@n7.LoopBody
    TargetMap:Map of Strings to Static Mesh Object References=Get SM_Dependencies Map()
    Key=@n10
    Value:"Static Mesh Object Reference (by ref)"=@n7
   ) {is=title K2Node_CallFunction}
   n11 = Get EUW_Asset_Mesh()
   n12 = Get AssetOptions(self=@n11)
   > AddOption(self=@n12, Option=@n10)
   > Branch(Condition="Equal (Integer)"(A=@n7.Array Index))
   > SetSelectedOption(self=@n12, Option=@n10)
   > n13 = Set Asset_Mesh(Asset_Mesh=@n7)
   > SetAssetByObject(self=Get AssetThumbnail(self=@n11), Object=@n13)
  ]}
 ImportLODsTool {nodes [
   ImportLODsTool() {is=title K2Node_FunctionEntry}
   > n1 = Branch(Condition=IsChecked(self=Get LOD1_CheckBox() -> (LOD1_CheckBox:Editor Utility Check Box Object Reference)))
   > n2 = ImportLOD!(
    execute
    self:Static Mesh Editor Subsystem Object Reference=Get StaticMeshEditorSubsystem()
    BaseStaticMesh:Static Mesh Object Reference=Get Asset_Mesh()
    LODIndex:int=1
    SourceFilename:String=Break File Path!(FilePath:"File Path Structure (by ref)"=Get PathLOD1() -> (PathLOD1:File Path Structure)) -> (FilePath:String) {is=title K2Node_BreakStruct}
   ) -> (then, ReturnValue:int=0)
   n3 = Branch(execute=[@n2.then, @n1.else], Condition=IsChecked(self=Get LOD2_CheckBox() -> (LOD2_CheckBox:Editor Utility Check Box Object Reference)))
   > n4 = ImportLOD(
    self=Get StaticMeshEditorSubsystem()
    BaseStaticMesh=Get Asset_Mesh()
    LODIndex=2
    SourceFilename=Break File Path(FilePath=Get PathLOD2() -> (PathLOD2:File Path Structure))
   ) -> (ReturnValue=0)
   Branch(execute=[@n4.then, @n3.else], Condition=IsChecked(self=Get LOD3_CheckBox() -> (LOD3_CheckBox:Editor Utility Check Box Object Reference)))
   > ImportLOD(
    self=Get StaticMeshEditorSubsystem()
    BaseStaticMesh=Get Asset_Mesh()
    LODIndex=3
    SourceFilename=Break File Path(FilePath=Get PathLOD3() -> (PathLOD3:File Path Structure))
   ) -> (ReturnValue=0)
  ]}
 LODCheckBoxes {nodes [
   LODCheckBoxes() {is=title K2Node_FunctionEntry}
   > n1 = "For Each Loop (Map)"!(
    execute:Exec
    MapPin:Map of Editor Utility Check Box Object References to Single Property View Object References=Make Map!(
     Key 0:Editor Utility Check Box Object Reference=Get LOD1_CheckBox()
     Value 0:Single Property View Object Reference=Get LOD1Path_Display()
     Key 1:Editor Utility Check Box Object Reference=Get LOD2_CheckBox()
     Value 1:Single Property View Object Reference=Get LOD2Path_Display()
     Key 2:Editor Utility Check Box Object Reference=Get LOD3_CheckBox()
     Value 2:Single Property View Object Reference=Get LOD3Path_Display()
    ) -> (Map:Map of Editor Utility Check Box Object References to Single Property View Object References) {is=title K2Node_MakeMap}
    BreakPin:Exec
   ) -> (then:Exec, KeyPin:Editor Utility Check Box Object Reference, ValuePin:Single Property View Object Reference, CompletedPin:Exec)
   > n2 = Branch(Condition=IsChecked(self=@n1))
   > SetIsEnabled(self=@n1.ValuePin, bInIsEnabled=true)
   SetIsEnabled(execute=@n2.else, self=@n1.ValuePin, bInIsEnabled=false)
  ]}
 RetrieveBP {nodes [
   RetrieveBP() {is=title K2Node_FunctionEntry}
   > n1 = GetSelectedAssets()
   > n2 = Filter by Class(TargetArray=@n1, ObjectClass=&Blueprint) -> (ReturnValue:Array of Blueprint Object References)
   > n3 = Set Asset_BP(Asset_BP="Get (a copy)"(Array:Array of Blueprint Object References=@n2) -> (Output:Blueprint Object Reference))
   n4 = Get EUW_Asset_BP()
   n5 = Get AssetThumbnail(self=@n4)
   > SetAssetByObject(self=@n5, Object=@n3)
   > SetVisibility(self=@n5, InVisibility=Visible)
   n6 = Get AssetOptions(self=@n4)
   > ClearOptions(self=@n6)
   n7 = GetDisplayName(Object=Get Asset_BP())
   > AddOption(self=@n6, Option=@n7)
   > SetSelectedOption(self=@n6, Option=@n7)
   > Branch(Condition="Equal (Integer)"(A=GetOptionCount!(self:"Combo Box (String) Object Reference"=@n6) -> (ReturnValue:int=0), B=1))
   > Set bIsEnabled(bIsEnabled=false, self=@n6)
   n8 = Get Asset_BP()
   > PrintString(InString=Append(A="End of Retrieve BP: ", B=GetDisplayName(Object=@n8)))
  ]}
 RetrieveSM {nodes [
   RetrieveSM() {is=title K2Node_FunctionEntry}
   > n1 = GetSelectedAssets()
   n2 = "Get (a copy)"(Array:Array of Object References=@n1) -> (Output:Object Reference)
   > n3 = Filter by Class(TargetArray=Make Array!("[0]":Object Reference=@n2) -> (Array:Array of Object References), ObjectClass=&StaticMesh) -> (ReturnValue:Array of Static Mesh Object References)
   > n4 = Branch(Condition="Equal (Integer)"(A=Length(TargetArray=@n3)))
   n5 = Cast To StaticMesh(execute=@n4.else, Object=@n2)
   > n6 = Set Asset_Mesh(Asset_Mesh=@n5)
   n7 = Get EUW_Asset_Mesh()
   n8 = Get AssetThumbnail(self=@n7)
   > SetAssetByObject(self=@n8, Object=@n6)
   > SetVisibility(self=@n8, InVisibility=Visible)
   n9 = Get AssetOptions(self=@n7)
   > ClearOptions(self=@n9)
   n10 = GetDisplayName(Object=Get Asset_Mesh())
   > AddOption(self=@n9, Option=@n10)
   > SetSelectedOption(self=@n9, Option=@n10)
   > Branch(Condition="Equal (Integer)"(A=GetOptionCount(self=@n9) -> (ReturnValue=0), B=1))
   > Set bIsEnabled(bIsEnabled=false, self=@n9)
  ]}
 SetPassFailColor {nodes [
   n1 = SetPassFailColor!() -> (then, Condition:bool=~, Text Box:Text Object Reference) {is=title K2Node_FunctionEntry}
   > n2 = Branch(Condition=@n1)
   > Set ColorAndOpacity!(
    execute
    ColorAndOpacity:Slate Color Structure=Get Pass Color() -> (Pass Color:Slate Color Structure)
    self:Text Object Reference=@n1.Text Box
   ) -> (then, Output_Get:Slate Color Structure)
   Set ColorAndOpacity(execute=@n2.else, ColorAndOpacity=Get Fail Color() -> (Fail Color:Slate Color Structure), self=@n1.Text Box)
  ]}
 SetTaskStatus {nodes [
   n1 = SetTaskStatus!() -> (then, Text Box:Text Object Reference, Condition:bool=~) {is=title K2Node_FunctionEntry}
   > n2 = Branch(Condition=@n1.Condition)
   > n3 = Set Text(Text="LOCTEXT(\"Pass\")", self=@n1)
   SetPassFailColor(execute=[@n4.then, @n3.then], Condition=@n1.Condition, Text Box=@n1) {is=title K2Node_CallFunction}
   n4 = Set Text(execute=@n2.else, Text="LOCTEXT(\"Fail\")", self=@n1)
  ]}
 UpdateToolMode {nodes [
   UpdateToolMode() {is=title K2Node_FunctionEntry}
   n1 = Get ToolMode() -> (ToolMode=NewEnumerator0)
   > n2 = Branch(Condition="Equal (Enum)"(A:ENUM_ImpToolKit_Modes Enum=@n1, B:ENUM_ImpToolKit_Modes Enum=NewEnumerator0))
   > SetVisibility(self=Get ExpandableArea_Asset() -> (ExpandableArea_Asset:Editor Utility Expandable Area Object Reference), InVisibility=Visible)
   > SetVisibility(self=Get ExpandableArea_Import() -> (ExpandableArea_Import:Editor Utility Expandable Area Object Reference), InVisibility=Visible)
   Branch(execute=@n2.else, Condition="Equal (Enum)"(A=@n1, B=NewEnumerator1))
   > SetVisibility(self=Get ExpandableArea_Asset(), InVisibility=Collapsed)
   > SetVisibility(self=Get ExpandableArea_Import(), InVisibility=Collapsed)
  ]}
]
variables [
 Asset_BP {type=Object, default=null}
 Asset_Mesh {type=StaticMesh, default=null}
 BP_References Map {type="map<Blueprint>", default {}}
 Map {default {}}
 Map_0 {default {}}
 PathLOD1 {type=FilePath, default {filePath=""}}
 PathLOD2 {type=FilePath, default {filePath=""}}
 PathLOD3 {type=FilePath, default {filePath=""}}
 SM_Dependencies Map {type="map<StaticMesh>", default {}}
 ToolMode {type=ENUM_ImpToolKit_Modes, default=Single Asset}
 ToolTip_CreateBSG {
  type=string
  default="Creates a Building Static Geo (BSG) version of an existing blueprint.\r\n\r\nNew BSG version will be created within a BSG folder in the location of the existing BP.\r\n\r\nUse this when needing to create an indestructible version of an existing destructible BP."
 }
]
widgets {
 info {namedSlotCount=22, rootWidgetClass=/Script/CommonUI.CommonHierarchicalScrollBox, widgetCount=140}
 widgets [
  CommonHierarchicalScrollBox_1 {
   widgetClassPath=&CommonHierarchicalScrollBox
   widget=CommonHierarchicalScrollBox_1
   properties {
    widgetBarStyle {
     thickness=8
     draggedThumbImage {
      bIsDynamicallyLoaded=false
      drawAs=RoundedBox
      imageType=NoImage
      mirroring=NoMirror
      resourceName=None
      resourceObject=null
      tiling=NoTile
      imageSize=<0 0>
      margin=<ltrb 0 0 0 0>
      outlineSettings {
       bUseBrushTransparency=false
       roundingType=FixedRadius
       width=0
       color {colorUseRule=UseColor_Specified, specifiedColor=<rgba 0 0 0 0>}
       cornerRadii=<4 4 4 4>
      }
      tintColor {colorUseRule=UseColor_Specified, specifiedColor=<rgba 0.2159 0.2159 0.2159 1>}
      uVRegion {bIsValid=false, max=<0 0>, min=<0 0>}
     }
     horizontalBackgroundImage {
      bIsDynamicallyLoaded=false
      drawAs=NoDrawType
      imageType=NoImage
      mirroring=NoMirror
      resourceName=None
      resourceObject=null
      tiling=NoTile
      imageSize=<0 0>
      margin=<ltrb 0 0 0 0>
      outlineSettings {
       bUseBrushTransparency=false
       roundingType=HalfHeightRadius
       width=0
       color {colorUseRule=UseColor_Specified, specifiedColor=<rgba 0 0 0 0>}
       cornerRadii=<0 0 0 0>
      }
      tintColor {colorUseRule=UseColor_Specified, specifiedColor=<rgba 1 1 1 1>}
      uVRegion {bIsValid=false, max=<0 0>, min=<0 0>}
     }
     horizontalBottomSlotImage {
      bIsDynamicallyLoaded=false
      drawAs=NoDrawType
      imageType=NoImage
      mirroring=NoMirror
      resourceName=None
      resourceObject=null
      tiling=NoTile
      imageSize=<0 0>
      margin=<ltrb 0 0 0 0>
      outlineSettings {
       bUseBrushTransparency=false
       roundingType=HalfHeightRadius
       width=0
       color {colorUseRule=UseColor_Specified, specifiedColor=<rgba 0 0 0 0>}
       cornerRadii=<0 0 0 0>
      }
      tintColor {colorUseRule=UseColor_Specified, specifiedColor=<rgba 1 1 1 1>}
      uVRegion {bIsValid=false, max=<0 0>, min=<0 0>}
     }
     horizontalTopSlotImage {
      bIsDynamicallyLoaded=false
      drawAs=NoDrawType
      imageType=NoImage
      mirroring=NoMirror
      resourceName=None
      resourceObject=null
      tiling=NoTile
      imageSize=<0 0>
      margin=<ltrb 0 0 0 0>
      outlineSettings {
       bUseBrushTransparency=false
       roundingType=HalfHeightRadius
       width=0
       color {colorUseRule=UseColor_Specified, specifiedColor=<rgba 0 0 0 0>}
       cornerRadii=<0 0 0 0>
      }
      tintColor {colorUseRule=UseColor_Specified, specifiedColor=<rgba 1 1 1 1>}
      uVRegion {bIsValid=false, max=<0 0>, min=<0 0>}
     }
     hoveredThumbImage {
      bIsDynamicallyLoaded=false
      drawAs=RoundedBox
      imageType=NoImage
      mirroring=NoMirror
      resourceName=None
      resourceObject=null
      tiling=NoTile
      imageSize=<0 0>
      margin=<ltrb 0 0 0 0>
      outlineSettings {
       bUseBrushTransparency=false
       roundingType=FixedRadius
       width=0
       color {colorUseRule=UseColor_Specified, specifiedColor=<rgba 0 0 0 0>}
       cornerRadii=<4 4 4 4>
      }
      tintColor {colorUseRule=UseColor_Specified, specifiedColor=<rgba 0.2159 0.2159 0.2159 1>}
      uVRegion {bIsValid=false, max=<0 0>, min=<0 0>}
     }
     normalThumbImage {
      bIsDynamicallyLoaded=false
      drawAs=RoundedBox
      imageType=NoImage
      mirroring=NoMirror
      resourceName=None
      resourceObject=null
      tiling=NoTile
      imageSize=<0 0>
      margin=<ltrb 0 0 0 0>
      outlineSettings {
       bUseBrushTransparency=false
       roundingType=FixedRadius
       width=0
       color {colorUseRule=UseColor_Specified, specifiedColor=<rgba 0 0 0 0>}
       cornerRadii=<4 4 4 4>
      }
      tintColor {colorUseRule=UseColor_Specified, specifiedColor=<rgba 0.0953 0.0953 0.0953 1>}
      uVRegion {bIsValid=false, max=<0 0>, min=<0 0>}
     }
     verticalBackgroundImage {
      bIsDynamicallyLoaded=false
      drawAs=NoDrawType
      imageType=NoImage
      mirroring=NoMirror
      resourceName=None
      resourceObject=null
      tiling=NoTile
      imageSize=<0 0>
      margin=<ltrb 0 0 0 0>
      outlineSettings {
       bUseBrushTransparency=false
       roundingType=HalfHeightRadius
       width=0
       color {colorUseRule=UseColor_Specified, specifiedColor=<rgba 0 0 0 0>}
       cornerRadii=<0 0 0 0>
      }
      tintColor {colorUseRule=UseColor_Specified, specifiedColor=<rgba 1 1 1 1>}
      uVRegion {bIsValid=false, max=<0 0>, min=<0 0>}
     }
     verticalBottomSlotImage {
      bIsDynamicallyLoaded=false
      drawAs=NoDrawType
      imageType=NoImage
      mirroring=NoMirror
      resourceName=None
      resourceObject=null
      tiling=NoTile
      imageSize=<0 0>
      margin=<ltrb 0 0 0 0>
      outlineSettings {
       bUseBrushTransparency=false
       roundingType=HalfHeightRadius
       width=0
       color {colorUseRule=UseColor_Specified, specifiedColor=<rgba 0 0 0 0>}
       cornerRadii=<0 0 0 0>
      }
      tintColor {colorUseRule=UseColor_Specified, specifiedColor=<rgba 1 1 1 1>}
      uVRegion {bIsValid=false, max=<0 0>, min=<0 0>}
     }
     verticalTopSlotImage {
      bIsDynamicallyLoaded=false
      drawAs=NoDrawType
      imageType=NoImage
      mirroring=NoMirror
      resourceName=None
      resourceObject=null
      tiling=NoTile
      imageSize=<0 0>
      margin=<ltrb 0 0 0 0>
      outlineSettings {
       bUseBrushTransparency=false
       roundingType=HalfHeightRadius
       width=0
       color {colorUseRule=UseColor_Specified, specifiedColor=<rgba 0 0 0 0>}
       cornerRadii=<0 0 0 0>
      }
      tintColor {colorUseRule=UseColor_Specified, specifiedColor=<rgba 1 1 1 1>}
      uVRegion {bIsValid=false, max=<0 0>, min=<0 0>}
     }
    }
    widgetStyle {
     barThickness=8
     bottomShadowBrush {
      bIsDynamicallyLoaded=false
      drawAs=Box
      imageType=FullColor
      mirroring=NoMirror
      resourceName=../../../Engine/Content/Slate/Common/ScrollBoxShadowBottom.png
      resourceObject=null
      tiling=NoTile
      imageSize=<16 8>
      margin=<ltrb 0.5 0 0.5 1>
      outlineSettings {
       bUseBrushTransparency=false
       roundingType=HalfHeightRadius
       width=0
       color {colorUseRule=UseColor_Specified, specifiedColor=<rgba 0 0 0 0>}
       cornerRadii=<0 0 0 0>
      }
      tintColor {colorUseRule=UseColor_Specified, specifiedColor=<rgba 1 1 1 1>}
      uVRegion {bIsValid=false, max=<0 0>, min=<0 0>}
     }
     horizontalScrolledContentPadding=<ltrb 0 0 1 0>
     leftShadowBrush {
      bIsDynamicallyLoaded=false
      drawAs=Box
      imageType=FullColor
      mirroring=NoMirror
      resourceName=../../../Engine/Content/Slate/Common/ScrollBoxShadowLeft.png
      resourceObject=null
      tiling=NoTile
      imageSize=<8 16>
      margin=<ltrb 1 0.5 0 0.5>
      outlineSettings {
       bUseBrushTransparency=false
       roundingType=HalfHeightRadius
       width=0
       color {colorUseRule=UseColor_Specified, specifiedColor=<rgba 0 0 0 0>}
       cornerRadii=<0 0 0 0>
      }
      tintColor {colorUseRule=UseColor_Specified, specifiedColor=<rgba 1 1 1 1>}
      uVRegion {bIsValid=false, max=<0 0>, min=<0 0>}
     }
     rightShadowBrush {
      bIsDynamicallyLoaded=false
      drawAs=Box
      imageType=FullColor
      mirroring=NoMirror
      resourceName=../../../Engine/Content/Slate/Common/ScrollBoxShadowRight.png
      resourceObject=null
      tiling=NoTile
      imageSize=<8 16>
      margin=<ltrb 0 0.5 1 0.5>
      outlineSettings {
       bUseBrushTransparency=false
       roundingType=HalfHeightRadius
       width=0
       color {colorUseRule=UseColor_Specified, specifiedColor=<rgba 0 0 0 0>}
       cornerRadii=<0 0 0 0>
      }
      tintColor {colorUseRule=UseColor_Specified, specifiedColor=<rgba 1 1 1 1>}
      uVRegion {bIsValid=false, max=<0 0>, min=<0 0>}
     }
     topShadowBrush {
      bIsDynamicallyLoaded=false
      drawAs=Box
      imageType=FullColor
      mirroring=NoMirror
      resourceName=../../../Engine/Content/Slate/Common/ScrollBoxShadowTop.png
      resourceObject=null
      tiling=NoTile
      imageSize=<16 8>
      margin=<ltrb 0.5 1 0.5 0>
      outlineSettings {
       bUseBrushTransparency=false
       roundingType=HalfHeightRadius
       width=0
       color {colorUseRule=UseColor_Specified, specifiedColor=<rgba 0 0 0 0>}
       cornerRadii=<0 0 0 0>
      }
      tintColor {colorUseRule=UseColor_Specified, specifiedColor=<rgba 1 1 1 1>}
      uVRegion {bIsValid=false, max=<0 0>, min=<0 0>}
     }
     verticalScrolledContentPadding=<ltrb 0 0 0 1>
    }
   }
  }
  Border_59 {
   widgetClassPath=&Border
   parent=CommonHierarchicalScrollBox_1
   widget=Border_59
   properties {background {
     bIsDynamicallyLoaded=false
     drawAs=Image
     imageType=NoImage
     mirroring=NoMirror
     resourceName=None
     resourceObject=null
     tiling=NoTile
     imageSize=<32 32>
     margin=<ltrb 0 0 0 0>
     outlineSettings {
      bUseBrushTransparency=false
      roundingType=HalfHeightRadius
      width=0
      color {colorUseRule=UseColor_Specified, specifiedColor=<rgba 0 0 0 0>}
      cornerRadii=<0 0 0 0>
     }
     tintColor {colorUseRule=UseColor_Specified, specifiedColor=<rgba 0.0395 0.0395 0.0395 1>}
     uVRegion {bIsValid=false, max=<0 0>, min=<0 0>}
    }}
   slot {class=ScrollBoxSlot}
  }
  VerticalBox_83 {widgetClassPath=&VerticalBox, parent=Border_59, widget=VerticalBox_83, slot {class=BorderSlot}}
  Spacer {widgetClassPath=&Spacer, parent=VerticalBox_83, widget=Spacer, properties {size=<1 10>}, slot {class=VerticalBoxSlot}}
  HorizontalBox_108 {widgetClassPath=&HorizontalBox, parent=VerticalBox_83, widget=HorizontalBox_108, slot {class=VerticalBoxSlot}}
  TextBlock_101 {
   widgetClassPath=&TextBlock
   parent=HorizontalBox_108
   widget=TextBlock_101
   properties {
    text=Mode :
    font {
     bForceMonospaced=false
     bMaterialIsStencil=false
     fontMaterial=null
     fontObject=&Roboto
     letterSpacing=0
     monospacedWidth=1
     size=10.5
     skewAmount=0
     typefaceFontName=Bold
     outlineSettings {
      bApplyOutlineToDropShadows=false
      bMiteredCorners=false
      bSeparateFillAlpha=false
      outlineMaterial=null
      outlineSize=0
      outlineColor=<rgba 0 0 0 1>
     }
    }
   }
   slot {class=HorizontalBoxSlot, properties {verticalAlignment=VAlign_Center}}
  }
  ComboBox_ToolMode {
   widgetClassPath=/Script/Blutility.EditorUtilityComboBoxString
   bIsVariable=true
   parent=HorizontalBox_108
   widget=ComboBox_ToolMode
   properties {
    selectedOption=Single Asset
    defaultOptions [Single Asset, Selection]
    font {
     bForceMonospaced=false
     bMaterialIsStencil=false
     fontMaterial=null
     fontObject=&Roboto
     letterSpacing=0
     monospacedWidth=1
     size=10.5
     skewAmount=0
     typefaceFontName=Bold
     outlineSettings {
      bApplyOutlineToDropShadows=false
      bMiteredCorners=false
      bSeparateFillAlpha=false
      outlineMaterial=null
      outlineSize=0
      outlineColor=<rgba 0 0 0 1>
     }
    }
    foregroundColor {colorUseRule=UseColor_Specified, specifiedColor=<rgba 0.5271 0.5271 0.5271 1>}
   }
   slot {class=HorizontalBoxSlot}
  }
  SM_Button_Documentation {
   widgetClassPath=/Script/UIFramework.UIFrameworkCustomButtonWidget
   bIsVariable=true
   parent=HorizontalBox_108
   widget=SM_Button_Documentation
   properties {
    visibility=Collapsed
    widgetStyle {
     clickedSlateSound {resourceObject=null}
     disabled {
      bIsDynamicallyLoaded=false
      drawAs=RoundedBox
      imageType=NoImage
      mirroring=NoMirror
      resourceName=None
      resourceObject=null
      tiling=NoTile
      imageSize=<0 0>
      margin=<ltrb 0 0 0 0>
      outlineSettings {
       bUseBrushTransparency=false
       roundingType=FixedRadius
       width=1
       color {colorUseRule=UseColor_ColorTable, specifiedColor=<rgba 0 0 0 0>}
       cornerRadii=<4 4 4 4>
      }
      tintColor {colorUseRule=UseColor_Specified, specifiedColor=<rgba 0.0395 0.0395 0.0395 1>}
      uVRegion {bIsValid=false, max=<0 0>, min=<0 0>}
     }
     disabledForeground {colorUseRule=UseColor_ColorTable, specifiedColor=<rgba 0 0 0 0>}
     hovered {
      bIsDynamicallyLoaded=false
      drawAs=RoundedBox
      imageType=NoImage
      mirroring=NoMirror
      resourceName=None
      resourceObject=null
      tiling=NoTile
      imageSize=<32 32>
      margin=<ltrb 0 0 0 0>
      outlineSettings {
       bUseBrushTransparency=false
       roundingType=FixedRadius
       width=1
       color {colorUseRule=UseColor_ColorTable, specifiedColor=<rgba 0 0 0 0>}
       cornerRadii=<4 4 4 4>
      }
      tintColor {colorUseRule=UseColor_Specified, specifiedColor=<rgba 0.0953 0.0953 0.0953 1>}
      uVRegion {bIsValid=false, max=<0 0>, min=<0 0>}
     }
     hoveredForeground {colorUseRule=UseColor_ColorTable, specifiedColor=<rgba 0 0 0 0>}
     hoveredSlateSound {resourceObject=null}
     normal {
      bIsDynamicallyLoaded=false
      drawAs=RoundedBox
      imageType=NoImage
      mirroring=NoMirror
      resourceName=None
      resourceObject=null
      tiling=NoTile
      imageSize=<32 32>
      margin=<ltrb 0 0 0 0>
      outlineSettings {
       bUseBrushTransparency=false
       roundingType=FixedRadius
       width=1
       color {colorUseRule=UseColor_ColorTable, specifiedColor=<rgba 0 0 0 0>}
       cornerRadii=<4 4 4 4>
      }
      tintColor {colorUseRule=UseColor_Specified, specifiedColor=<rgba 0.0395 0.0395 0.0395 1>}
      uVRegion {bIsValid=false, max=<0 0>, min=<0 0>}
     }
     normalForeground {colorUseRule=UseColor_ColorTable, specifiedColor=<rgba 0 0 0 0>}
     normalPadding=<ltrb 12 1.5 12 1.5>
     pressed {
      bIsDynamicallyLoaded=false
      drawAs=RoundedBox
      imageType=NoImage
      mirroring=NoMirror
      resourceName=None
      resourceObject=null
      tiling=NoTile
      imageSize=<32 32>
      margin=<ltrb 0 0 0 0>
      outlineSettings {
       bUseBrushTransparency=false
       roundingType=FixedRadius
       width=1
       color {colorUseRule=UseColor_ColorTable, specifiedColor=<rgba 0 0 0 0>}
       cornerRadii=<4 4 4 4>
      }
      tintColor {colorUseRule=UseColor_Specified, specifiedColor=<rgba 0.0284 0.0284 0.0284 1>}
      uVRegion {bIsValid=false, max=<0 0>, min=<0 0>}
     }
     pressedForeground {colorUseRule=UseColor_ColorTable, specifiedColor=<rgba 0 0 0 0>}
     pressedPadding=<ltrb 12 2.5 12 0.5>
     pressedSlateSound {resourceObject=null}
    }
   }
   slot {class=HorizontalBoxSlot, properties {horizontalAlignment=HAlign_Right, padding=<ltrb 5 5 5 5>, size {sizeRule=Fill, value=1}}}
  }
  TextBlock_7 {
   widgetClassPath=&TextBlock
   parent=SM_Button_Documentation
   widget=TextBlock_7
   properties {
    text=Documentation
    font {
     bForceMonospaced=false
     bMaterialIsStencil=false
     fontMaterial=null
     fontObject=&Roboto
     letterSpacing=0
     monospacedWidth=1
     size=10.5
     skewAmount=0
     typefaceFontName=Bold
     outlineSettings {
      bApplyOutlineToDropShadows=false
      bMiteredCorners=false
      bSeparateFillAlpha=false
      outlineMaterial=null
      outlineSize=0
      outlineColor=<rgba 0 0 0 1>
     }
    }
   }
   slot {class=ButtonSlot}
  }
  Spacer_9 {widgetClassPath=&Spacer, parent=VerticalBox_83, widget=Spacer_9, properties {size=<1 10>}, slot {class=VerticalBoxSlot}}
  ExpandableArea_Asset {
   widgetClassPath=/Script/Blutility.EditorUtilityExpandableArea
   bIsVariable=true
   parent=VerticalBox_83
   widget=ExpandableArea_Asset
   properties {bIsExpanded=true}
   slot {class=VerticalBoxSlot}
  }
  Text_Asset {
   widgetClassPath=&TextBlock
   namedSlotHost=ExpandableArea_Asset
   widget=Text_Asset
   properties {
    text=Asset
    font {
     bForceMonospaced=false
     bMaterialIsStencil=false
     fontMaterial=null
     fontObject=&Roboto
     letterSpacing=0
     monospacedWidth=1
     size=13.5
     skewAmount=0
     typefaceFontName=Bold
     outlineSettings {
      bApplyOutlineToDropShadows=false
      bMiteredCorners=false
      bSeparateFillAlpha=false
      outlineMaterial=null
      outlineSize=0
      outlineColor=<rgba 0 0 0 1>
     }
    }
   }
  }
  Border_0 {
   widgetClassPath=&Border
   namedSlotHost=ExpandableArea_Asset
   widget=Border_0
   properties {brushColor=<rgba 1 1 1 0>, padding=<ltrb 10 10 10 10>}
  }
  VerticalBox_110 {
   widgetClassPath=&VerticalBox
   parent=Border_0
   widget=VerticalBox_110
   slot {class=BorderSlot, properties {padding=<ltrb 10 10 10 10>}}
  }
  HorizontalBox_2 {
   widgetClassPath=&HorizontalBox
   parent=VerticalBox_110
   widget=HorizontalBox_2
   slot {class=VerticalBoxSlot, properties {size {sizeRule=Fill, value=1}}}
  }
  HorizontalBox_89 {
   widgetClassPath=&HorizontalBox
   parent=HorizontalBox_2
   widget=HorizontalBox_89
   slot {class=HorizontalBoxSlot, properties {size {sizeRule=Fill, value=1}}}
  }
  TextBlock_94 {
   widgetClassPath=&TextBlock
   parent=HorizontalBox_89
   widget=TextBlock_94
   properties {
    text=Mesh
    font {
     bForceMonospaced=false
     bMaterialIsStencil=false
     fontMaterial=null
     fontObject=&Roboto
     letterSpacing=0
     monospacedWidth=1
     size=10.5
     skewAmount=0
     typefaceFontName=Bold
     outlineSettings {
      bApplyOutlineToDropShadows=false
      bMiteredCorners=false
      bSeparateFillAlpha=false
      outlineMaterial=null
      outlineSize=0
      outlineColor=<rgba 0 0 0 1>
     }
    }
   }
   slot {class=HorizontalBoxSlot, properties {horizontalAlignment=HAlign_Center, verticalAlignment=VAlign_Center}}
  }
  Spacer_102 {widgetClassPath=&Spacer, parent=HorizontalBox_89, widget=Spacer_102, properties {size=<10 1>}, slot {class=HorizontalBoxSlot}}
  EUW_Asset_Mesh {
   widgetClassPath=&/EUW_ImpToolKit_Asset
   bIsVariable=true
   parent=HorizontalBox_89
   widget=EUW_Asset_Mesh
   slot {class=HorizontalBoxSlot, properties {size {sizeRule=Fill, value=1}}}
  }
  Spacer_162 {widgetClassPath=&Spacer, parent=HorizontalBox_2, widget=Spacer_162, properties {size=<25 1>}, slot {class=HorizontalBoxSlot}}
  HorizontalBox_3 {
   widgetClassPath=&HorizontalBox
   parent=HorizontalBox_2
   widget=HorizontalBox_3
   slot {class=HorizontalBoxSlot, properties {size {sizeRule=Fill, value=1}}}
  }
  TextBlock_1 {
   widgetClassPath=&TextBlock
   parent=HorizontalBox_3
   widget=TextBlock_1
   properties {
    text=BP
    font {
     bForceMonospaced=false
     bMaterialIsStencil=false
     fontMaterial=null
     fontObject=&Roboto
     letterSpacing=0
     monospacedWidth=1
     size=10.5
     skewAmount=0
     typefaceFontName=Bold
     outlineSettings {
      bApplyOutlineToDropShadows=false
      bMiteredCorners=false
      bSeparateFillAlpha=false
      outlineMaterial=null
      outlineSize=0
      outlineColor=<rgba 0 0 0 1>
     }
    }
   }
   slot {class=HorizontalBoxSlot, properties {horizontalAlignment=HAlign_Center, verticalAlignment=VAlign_Center}}
  }
  Spacer_3 {widgetClassPath=&Spacer, parent=HorizontalBox_3, widget=Spacer_3, properties {size=<10 1>}, slot {class=HorizontalBoxSlot}}
  EUW_Asset_BP {
   widgetClassPath=&EUW_ImpToolKit_Asset
   bIsVariable=true
   parent=HorizontalBox_3
   widget=EUW_Asset_BP
   slot {class=HorizontalBoxSlot, properties {size {sizeRule=Fill, value=1}}}
  }
  Spacer_2 {widgetClassPath=&Spacer, parent=VerticalBox_110, widget=Spacer_2, properties {size=<1 15>}, slot {class=VerticalBoxSlot}}
  HorizontalBox_326 {widgetClassPath=&HorizontalBox, parent=VerticalBox_110, widget=HorizontalBox_326, slot {class=VerticalBoxSlot}}
  Button_GetMesh {
   widgetClassPath=&UIFrameworkCustomButtonWidget
   bIsVariable=true
   parent=HorizontalBox_326
   widget=Button_GetMesh
   properties {widgetStyle {
     clickedSlateSound {resourceObject=null}
     disabled {
      bIsDynamicallyLoaded=false
      drawAs=RoundedBox
      imageType=NoImage
      mirroring=NoMirror
      resourceName=None
      resourceObject=null
      tiling=NoTile
      imageSize=<0 0>
      margin=<ltrb 0 0 0 0>
      outlineSettings {
       bUseBrushTransparency=false
       roundingType=FixedRadius
       width=1
       color {colorUseRule=UseColor_ColorTable, specifiedColor=<rgba 0 0 0 0>}
       cornerRadii=<4 4 4 4>
      }
      tintColor {colorUseRule=UseColor_Specified, specifiedColor=<rgba 0.0395 0.0395 0.0395 1>}
      uVRegion {bIsValid=false, max=<0 0>, min=<0 0>}
     }
     disabledForeground {colorUseRule=UseColor_ColorTable, specifiedColor=<rgba 0 0 0 0>}
     hovered {
      bIsDynamicallyLoaded=false
      drawAs=RoundedBox
      imageType=NoImage
      mirroring=NoMirror
      resourceName=None
      resourceObject=null
      tiling=NoTile
      imageSize=<32 32>
      margin=<ltrb 0 0 0 0>
      outlineSettings {
       bUseBrushTransparency=false
       roundingType=FixedRadius
       width=1
       color {colorUseRule=UseColor_ColorTable, specifiedColor=<rgba 0 0 0 0>}
       cornerRadii=<4 4 4 4>
      }
      tintColor {colorUseRule=UseColor_Specified, specifiedColor=<rgba 0.0953 0.0953 0.0953 1>}
      uVRegion {bIsValid=false, max=<0 0>, min=<0 0>}
     }
     hoveredForeground {colorUseRule=UseColor_ColorTable, specifiedColor=<rgba 0 0 0 0>}
     hoveredSlateSound {resourceObject=null}
     normal {
      bIsDynamicallyLoaded=false
      drawAs=RoundedBox
      imageType=NoImage
      mirroring=NoMirror
      resourceName=None
      resourceObject=null
      tiling=NoTile
      imageSize=<32 32>
      margin=<ltrb 0 0 0 0>
      outlineSettings {
       bUseBrushTransparency=false
       roundingType=FixedRadius
       width=1
       color {colorUseRule=UseColor_ColorTable, specifiedColor=<rgba 0 0 0 0>}
       cornerRadii=<4 4 4 4>
      }
      tintColor {colorUseRule=UseColor_Specified, specifiedColor=<rgba 0.0395 0.0395 0.0395 1>}
      uVRegion {bIsValid=false, max=<0 0>, min=<0 0>}
     }
     normalForeground {colorUseRule=UseColor_ColorTable, specifiedColor=<rgba 0 0 0 0>}
     normalPadding=<ltrb 12 1.5 12 1.5>
     pressed {
      bIsDynamicallyLoaded=false
      drawAs=RoundedBox
      imageType=NoImage
      mirroring=NoMirror
      resourceName=None
      resourceObject=null
      tiling=NoTile
      imageSize=<32 32>
      margin=<ltrb 0 0 0 0>
      outlineSettings {
       bUseBrushTransparency=false
       roundingType=FixedRadius
       width=1
       color {colorUseRule=UseColor_ColorTable, specifiedColor=<rgba 0 0 0 0>}
       cornerRadii=<4 4 4 4>
      }
      tintColor {colorUseRule=UseColor_Specified, specifiedColor=<rgba 0.0284 0.0284 0.0284 1>}
      uVRegion {bIsValid=false, max=<0 0>, min=<0 0>}
     }
     pressedForeground {colorUseRule=UseColor_ColorTable, specifiedColor=<rgba 0 0 0 0>}
     pressedPadding=<ltrb 12 2.5 12 0.5>
     pressedSlateSound {resourceObject=null}
    }}
   slot {class=HorizontalBoxSlot, properties {horizontalAlignment=HAlign_Center, padding=<ltrb 5 5 5 5>, size {sizeRule=Fill, value=1}}}
  }
  TextBlock_8 {
   widgetClassPath=&TextBlock
   parent=Button_GetMesh
   widget=TextBlock_8
   properties {
    text=Get Mesh
    font {
     bForceMonospaced=false
     bMaterialIsStencil=false
     fontMaterial=null
     fontObject=&Roboto
     letterSpacing=0
     monospacedWidth=1
     size=10.5
     skewAmount=0
     typefaceFontName=Bold
     outlineSettings {
      bApplyOutlineToDropShadows=false
      bMiteredCorners=false
      bSeparateFillAlpha=false
      outlineMaterial=null
      outlineSize=0
      outlineColor=<rgba 0 0 0 1>
     }
    }
   }
   slot {class=ButtonSlot}
  }
  Button_ClearAssets {
   widgetClassPath=&UIFrameworkCustomButtonWidget
   bIsVariable=true
   parent=HorizontalBox_326
   widget=Button_ClearAssets
   properties {widgetStyle {
     clickedSlateSound {resourceObject=null}
     disabled {
      bIsDynamicallyLoaded=false
      drawAs=RoundedBox
      imageType=NoImage
      mirroring=NoMirror
      resourceName=None
      resourceObject=null
      tiling=NoTile
      imageSize=<0 0>
      margin=<ltrb 0 0 0 0>
      outlineSettings {
       bUseBrushTransparency=false
       roundingType=FixedRadius
       width=1
       color {colorUseRule=UseColor_ColorTable, specifiedColor=<rgba 0 0 0 0>}
       cornerRadii=<4 4 4 4>
      }
      tintColor {colorUseRule=UseColor_Specified, specifiedColor=<rgba 0.0395 0.0395 0.0395 1>}
      uVRegion {bIsValid=false, max=<0 0>, min=<0 0>}
     }
     disabledForeground {colorUseRule=UseColor_ColorTable, specifiedColor=<rgba 0 0 0 0>}
     hovered {
      bIsDynamicallyLoaded=false
      drawAs=RoundedBox
      imageType=NoImage
      mirroring=NoMirror
      resourceName=None
      resourceObject=null
      tiling=NoTile
      imageSize=<32 32>
      margin=<ltrb 0 0 0 0>
      outlineSettings {
       bUseBrushTransparency=false
       roundingType=FixedRadius
       width=1
       color {colorUseRule=UseColor_ColorTable, specifiedColor=<rgba 0 0 0 0>}
       cornerRadii=<4 4 4 4>
      }
      tintColor {colorUseRule=UseColor_Specified, specifiedColor=<rgba 0.0953 0.0953 0.0953 1>}
      uVRegion {bIsValid=false, max=<0 0>, min=<0 0>}
     }
     hoveredForeground {colorUseRule=UseColor_ColorTable, specifiedColor=<rgba 0 0 0 0>}
     hoveredSlateSound {resourceObject=null}
     normal {
      bIsDynamicallyLoaded=false
      drawAs=RoundedBox
      imageType=NoImage
      mirroring=NoMirror
      resourceName=None
      resourceObject=null
      tiling=NoTile
      imageSize=<32 32>
      margin=<ltrb 0 0 0 0>
      outlineSettings {
       bUseBrushTransparency=false
       roundingType=FixedRadius
       width=1
       color {colorUseRule=UseColor_ColorTable, specifiedColor=<rgba 0 0 0 0>}
       cornerRadii=<4 4 4 4>
      }
      tintColor {colorUseRule=UseColor_Specified, specifiedColor=<rgba 0.0395 0.0395 0.0395 1>}
      uVRegion {bIsValid=false, max=<0 0>, min=<0 0>}
     }
     normalForeground {colorUseRule=UseColor_ColorTable, specifiedColor=<rgba 0 0 0 0>}
     normalPadding=<ltrb 12 1.5 12 1.5>
     pressed {
      bIsDynamicallyLoaded=false
      drawAs=RoundedBox
      imageType=NoImage
      mirroring=NoMirror
      resourceName=None
      resourceObject=null
      tiling=NoTile
      imageSize=<32 32>
      margin=<ltrb 0 0 0 0>
      outlineSettings {
       bUseBrushTransparency=false
       roundingType=FixedRadius
       width=1
       color {colorUseRule=UseColor_ColorTable, specifiedColor=<rgba 0 0 0 0>}
       cornerRadii=<4 4 4 4>
      }
      tintColor {colorUseRule=UseColor_Specified, specifiedColor=<rgba 0.0284 0.0284 0.0284 1>}
      uVRegion {bIsValid=false, max=<0 0>, min=<0 0>}
     }
     pressedForeground {colorUseRule=UseColor_ColorTable, specifiedColor=<rgba 0 0 0 0>}
     pressedPadding=<ltrb 12 2.5 12 0.5>
     pressedSlateSound {resourceObject=null}
    }}
   slot {class=HorizontalBoxSlot, properties {horizontalAlignment=HAlign_Center, padding=<ltrb 5 5 5 5>, size {sizeRule=Fill, value=1}}}
  }
  TextBlock_22 {
   widgetClassPath=&TextBlock
   parent=Button_ClearAssets
   widget=TextBlock_22
   properties {
    text=Clear Assets
    font {
     bForceMonospaced=false
     bMaterialIsStencil=false
     fontMaterial=null
     fontObject=&Roboto
     letterSpacing=0
     monospacedWidth=1
     size=10.5
     skewAmount=0
     typefaceFontName=Bold
     outlineSettings {
      bApplyOutlineToDropShadows=false
      bMiteredCorners=false
      bSeparateFillAlpha=false
      outlineMaterial=null
      outlineSize=0
      outlineColor=<rgba 0 0 0 1>
     }
    }
   }
   slot {class=ButtonSlot}
  }
  Button_GetBP {
   widgetClassPath=&UIFrameworkCustomButtonWidget
   bIsVariable=true
   parent=HorizontalBox_326
   widget=Button_GetBP
   properties {widgetStyle {
     clickedSlateSound {resourceObject=null}
     disabled {
      bIsDynamicallyLoaded=false
      drawAs=RoundedBox
      imageType=NoImage
      mirroring=NoMirror
      resourceName=None
      resourceObject=null
      tiling=NoTile
      imageSize=<0 0>
      margin=<ltrb 0 0 0 0>
      outlineSettings {
       bUseBrushTransparency=false
       roundingType=FixedRadius
       width=1
       color {colorUseRule=UseColor_ColorTable, specifiedColor=<rgba 0 0 0 0>}
       cornerRadii=<4 4 4 4>
      }
      tintColor {colorUseRule=UseColor_Specified, specifiedColor=<rgba 0.0395 0.0395 0.0395 1>}
      uVRegion {bIsValid=false, max=<0 0>, min=<0 0>}
     }
     disabledForeground {colorUseRule=UseColor_ColorTable, specifiedColor=<rgba 0 0 0 0>}
     hovered {
      bIsDynamicallyLoaded=false
      drawAs=RoundedBox
      imageType=NoImage
      mirroring=NoMirror
      resourceName=None
      resourceObject=null
      tiling=NoTile
      imageSize=<32 32>
      margin=<ltrb 0 0 0 0>
      outlineSettings {
       bUseBrushTransparency=false
       roundingType=FixedRadius
       width=1
       color {colorUseRule=UseColor_ColorTable, specifiedColor=<rgba 0 0 0 0>}
       cornerRadii=<4 4 4 4>
      }
      tintColor {colorUseRule=UseColor_Specified, specifiedColor=<rgba 0.0953 0.0953 0.0953 1>}
      uVRegion {bIsValid=false, max=<0 0>, min=<0 0>}
     }
     hoveredForeground {colorUseRule=UseColor_ColorTable, specifiedColor=<rgba 0 0 0 0>}
     hoveredSlateSound {resourceObject=null}
     normal {
      bIsDynamicallyLoaded=false
      drawAs=RoundedBox
      imageType=NoImage
      mirroring=NoMirror
      resourceName=None
      resourceObject=null
      tiling=NoTile
      imageSize=<32 32>
      margin=<ltrb 0 0 0 0>
      outlineSettings {
       bUseBrushTransparency=false
       roundingType=FixedRadius
       width=1
       color {colorUseRule=UseColor_ColorTable, specifiedColor=<rgba 0 0 0 0>}
       cornerRadii=<4 4 4 4>
      }
      tintColor {colorUseRule=UseColor_Specified, specifiedColor=<rgba 0.0395 0.0395 0.0395 1>}
      uVRegion {bIsValid=false, max=<0 0>, min=<0 0>}
     }
     normalForeground {colorUseRule=UseColor_ColorTable, specifiedColor=<rgba 0 0 0 0>}
     normalPadding=<ltrb 12 1.5 12 1.5>
     pressed {
      bIsDynamicallyLoaded=false
      drawAs=RoundedBox
      imageType=NoImage
      mirroring=NoMirror
      resourceName=None
      resourceObject=null
      tiling=NoTile
      imageSize=<32 32>
      margin=<ltrb 0 0 0 0>
      outlineSettings {
       bUseBrushTransparency=false
       roundingType=FixedRadius
       width=1
       color {colorUseRule=UseColor_ColorTable, specifiedColor=<rgba 0 0 0 0>}
       cornerRadii=<4 4 4 4>
      }
      tintColor {colorUseRule=UseColor_Specified, specifiedColor=<rgba 0.0284 0.0284 0.0284 1>}
      uVRegion {bIsValid=false, max=<0 0>, min=<0 0>}
     }
     pressedForeground {colorUseRule=UseColor_ColorTable, specifiedColor=<rgba 0 0 0 0>}
     pressedPadding=<ltrb 12 2.5 12 0.5>
     pressedSlateSound {resourceObject=null}
    }}
   slot {class=HorizontalBoxSlot, properties {horizontalAlignment=HAlign_Center, padding=<ltrb 5 5 5 5>, size {sizeRule=Fill, value=1}}}
  }
  TextBlock_21 {
   widgetClassPath=&TextBlock
   parent=Button_GetBP
   widget=TextBlock_21
   properties {
    text=Get BP
    font {
     bForceMonospaced=false
     bMaterialIsStencil=false
     fontMaterial=null
     fontObject=&Roboto
     letterSpacing=0
     monospacedWidth=1
     size=10.5
     skewAmount=0
     typefaceFontName=Bold
     outlineSettings {
      bApplyOutlineToDropShadows=false
      bMiteredCorners=false
      bSeparateFillAlpha=false
      outlineMaterial=null
      outlineSize=0
      outlineColor=<rgba 0 0 0 1>
     }
    }
   }
   slot {class=ButtonSlot}
  }
  HorizontalBox_92 {widgetClassPath=&HorizontalBox, parent=VerticalBox_110, widget=HorizontalBox_92, slot {class=VerticalBoxSlot}}
  ExpandableArea_Step_Perf {
   widgetClassPath=&EditorUtilityExpandableArea
   bIsVariable=true
   parent=HorizontalBox_92
   widget=ExpandableArea_Step_Perf
   slot {class=HorizontalBoxSlot, properties {padding=<ltrb 5 5 5 5>, size {sizeRule=Fill, value=1}}}
  }
  HorizontalBox_95 {widgetClassPath=&HorizontalBox, namedSlotHost=ExpandableArea_Step_Perf, widget=HorizontalBox_95}
  TextBlock_20 {
   widgetClassPath=&TextBlock
   parent=HorizontalBox_95
   widget=TextBlock_20
   properties {
    text="Perf : "
    font {
     bForceMonospaced=false
     bMaterialIsStencil=false
     fontMaterial=null
     fontObject=&Roboto
     letterSpacing=0
     monospacedWidth=1
     size=10.5
     skewAmount=0
     typefaceFontName=Bold
     outlineSettings {
      bApplyOutlineToDropShadows=false
      bMiteredCorners=false
      bSeparateFillAlpha=false
      outlineMaterial=null
      outlineSize=0
      outlineColor=<rgba 0 0 0 1>
     }
    }
   }
   slot {class=HorizontalBoxSlot}
  }
  TaskStatus_Perf {
   widgetClassPath=&TextBlock
   bIsVariable=true
   parent=HorizontalBox_95
   widget=TaskStatus_Perf
   properties {font {
     bForceMonospaced=false
     bMaterialIsStencil=false
     fontMaterial=null
     fontObject=&Roboto
     letterSpacing=0
     monospacedWidth=1
     size=10.5
     skewAmount=0
     typefaceFontName=Bold
     outlineSettings {
      bApplyOutlineToDropShadows=false
      bMiteredCorners=false
      bSeparateFillAlpha=false
      outlineMaterial=null
      outlineSize=0
      outlineColor=<rgba 0 0 0 1>
     }
    }}
   slot {class=HorizontalBoxSlot}
  }
  HorizontalBox_94 {widgetClassPath=&HorizontalBox, namedSlotHost=ExpandableArea_Step_Perf, widget=HorizontalBox_94}
  GridPanel_MeshInfo {
   widgetClassPath=&GridPanel
   parent=HorizontalBox_94
   widget=GridPanel_MeshInfo
   slot {class=HorizontalBoxSlot, properties {size {sizeRule=Fill, value=1}}}
  }
  TextBlock_2 {
   widgetClassPath=&TextBlock
   parent=GridPanel_MeshInfo
   widget=TextBlock_2
   properties {
    text=Nanite Enabled :
    font {
     bForceMonospaced=false
     bMaterialIsStencil=false
     fontMaterial=null
     fontObject=&Roboto
     letterSpacing=0
     monospacedWidth=1
     size=9
     skewAmount=0
     typefaceFontName=Bold
     outlineSettings {
      bApplyOutlineToDropShadows=false
      bMiteredCorners=false
      bSeparateFillAlpha=false
      outlineMaterial=null
      outlineSize=0
      outlineColor=<rgba 0 0 0 1>
     }
    }
   }
   slot {class=GridSlot, properties {row=2}}
  }
  Mesh_Info_NaniteEnabled {
   widgetClassPath=&TextBlock
   bIsVariable=true
   parent=GridPanel_MeshInfo
   widget=Mesh_Info_NaniteEnabled
   properties {font {
     bForceMonospaced=false
     bMaterialIsStencil=false
     fontMaterial=null
     fontObject=&Roboto
     letterSpacing=0
     monospacedWidth=1
     size=9
     skewAmount=0
     typefaceFontName=Bold
     outlineSettings {
      bApplyOutlineToDropShadows=false
      bMiteredCorners=false
      bSeparateFillAlpha=false
      outlineMaterial=null
      outlineSize=0
      outlineColor=<rgba 0 0 0 1>
     }
    }}
   slot {class=GridSlot, properties {column=1, row=2}}
  }
  TextBlock_3 {
   widgetClassPath=&TextBlock
   parent=GridPanel_MeshInfo
   widget=TextBlock_3
   properties {
    text=Nanite Fallback :
    font {
     bForceMonospaced=false
     bMaterialIsStencil=false
     fontMaterial=null
     fontObject=&Roboto
     letterSpacing=0
     monospacedWidth=1
     size=9
     skewAmount=0
     typefaceFontName=Bold
     outlineSettings {
      bApplyOutlineToDropShadows=false
      bMiteredCorners=false
      bSeparateFillAlpha=false
      outlineMaterial=null
      outlineSize=0
      outlineColor=<rgba 0 0 0 1>
     }
    }
   }
   slot {class=GridSlot, properties {row=3}}
  }
  Mesh_Info_NaniteFallback {
   widgetClassPath=&TextBlock
   bIsVariable=true
   parent=GridPanel_MeshInfo
   widget=Mesh_Info_NaniteFallback
   properties {font {
     bForceMonospaced=false
     bMaterialIsStencil=false
     fontMaterial=null
     fontObject=&Roboto
     letterSpacing=0
     monospacedWidth=1
     size=9
     skewAmount=0
     typefaceFontName=Bold
     outlineSettings {
      bApplyOutlineToDropShadows=false
      bMiteredCorners=false
      bSeparateFillAlpha=false
      outlineMaterial=null
      outlineSize=0
      outlineColor=<rgba 0 0 0 1>
     }
    }}
   slot {class=GridSlot, properties {column=1, row=3}}
  }
  TextBlock_4 {
   widgetClassPath=&TextBlock
   parent=GridPanel_MeshInfo
   widget=TextBlock_4
   properties {
    text=Realtive Error :
    font {
     bForceMonospaced=false
     bMaterialIsStencil=false
     fontMaterial=null
     fontObject=&Roboto
     letterSpacing=0
     monospacedWidth=1
     size=9
     skewAmount=0
     typefaceFontName=Bold
     outlineSettings {
      bApplyOutlineToDropShadows=false
      bMiteredCorners=false
      bSeparateFillAlpha=false
      outlineMaterial=null
      outlineSize=0
      outlineColor=<rgba 0 0 0 1>
     }
    }
   }
   slot {class=GridSlot, properties {row=4}}
  }
  Mesh_Info_RelativeError {
   widgetClassPath=&TextBlock
   bIsVariable=true
   parent=GridPanel_MeshInfo
   widget=Mesh_Info_RelativeError
   properties {font {
     bForceMonospaced=false
     bMaterialIsStencil=false
     fontMaterial=null
     fontObject=&Roboto
     letterSpacing=0
     monospacedWidth=1
     size=9
     skewAmount=0
     typefaceFontName=Bold
     outlineSettings {
      bApplyOutlineToDropShadows=false
      bMiteredCorners=false
      bSeparateFillAlpha=false
      outlineMaterial=null
      outlineSize=0
      outlineColor=<rgba 0 0 0 1>
     }
    }}
   slot {class=GridSlot, properties {column=1, row=4}}
  }
  TextBlock_6 {
   widgetClassPath=&TextBlock
   parent=GridPanel_MeshInfo
   widget=TextBlock_6
   properties {
    text="Min LODs : "
    font {
     bForceMonospaced=false
     bMaterialIsStencil=false
     fontMaterial=null
     fontObject=&Roboto
     letterSpacing=0
     monospacedWidth=1
     size=9
     skewAmount=0
     typefaceFontName=Bold
     outlineSettings {
      bApplyOutlineToDropShadows=false
      bMiteredCorners=false
      bSeparateFillAlpha=false
      outlineMaterial=null
      outlineSize=0
      outlineColor=<rgba 0 0 0 1>
     }
    }
   }
   slot {class=GridSlot, properties {row=7}}
  }
  Mesh_Info_MinLODs {
   widgetClassPath=&TextBlock
   bIsVariable=true
   parent=GridPanel_MeshInfo
   widget=Mesh_Info_MinLODs
   properties {font {
     bForceMonospaced=false
     bMaterialIsStencil=false
     fontMaterial=null
     fontObject=&Roboto
     letterSpacing=0
     monospacedWidth=1
     size=9
     skewAmount=0
     typefaceFontName=Bold
     outlineSettings {
      bApplyOutlineToDropShadows=false
      bMiteredCorners=false
      bSeparateFillAlpha=false
      outlineMaterial=null
      outlineSize=0
      outlineColor=<rgba 0 0 0 1>
     }
    }}
   slot {class=GridSlot, properties {column=1, row=7}}
  }
  TextBlock_5 {
   widgetClassPath=&TextBlock
   parent=GridPanel_MeshInfo
   widget=TextBlock_5
   properties {
    text="LOD for Collision : "
    font {
     bForceMonospaced=false
     bMaterialIsStencil=false
     fontMaterial=null
     fontObject=&Roboto
     letterSpacing=0
     monospacedWidth=1
     size=9
     skewAmount=0
     typefaceFontName=Bold
     outlineSettings {
      bApplyOutlineToDropShadows=false
      bMiteredCorners=false
      bSeparateFillAlpha=false
      outlineMaterial=null
      outlineSize=0
      outlineColor=<rgba 0 0 0 1>
     }
    }
   }
   slot {class=GridSlot, properties {row=8}}
  }
  Mesh_Info_LODForCollision {
   widgetClassPath=&TextBlock
   bIsVariable=true
   parent=GridPanel_MeshInfo
   widget=Mesh_Info_LODForCollision
   properties {font {
     bForceMonospaced=false
     bMaterialIsStencil=false
     fontMaterial=null
     fontObject=&Roboto
     letterSpacing=0
     monospacedWidth=1
     size=9
     skewAmount=0
     typefaceFontName=Bold
     outlineSettings {
      bApplyOutlineToDropShadows=false
      bMiteredCorners=false
      bSeparateFillAlpha=false
      outlineMaterial=null
      outlineSize=0
      outlineColor=<rgba 0 0 0 1>
     }
    }}
   slot {class=GridSlot, properties {column=1, row=8}}
  }
  ExpandableArea_Step_BPSetup {
   widgetClassPath=&EditorUtilityExpandableArea
   bIsVariable=true
   parent=HorizontalBox_92
   widget=ExpandableArea_Step_BPSetup
   slot {class=HorizontalBoxSlot, properties {padding=<ltrb 5 5 5 5>, size {sizeRule=Fill, value=1}}}
  }
  Text_BPSetup {
   widgetClassPath=&TextBlock
   namedSlotHost=ExpandableArea_Step_BPSetup
   widget=Text_BPSetup
   properties {
    text=BP Setup:
    font {
     bForceMonospaced=false
     bMaterialIsStencil=false
     fontMaterial=null
     fontObject=&Roboto
     letterSpacing=0
     monospacedWidth=1
     size=10.5
     skewAmount=0
     typefaceFontName=Bold
     outlineSettings {
      bApplyOutlineToDropShadows=false
      bMiteredCorners=false
      bSeparateFillAlpha=false
      outlineMaterial=null
      outlineSize=0
      outlineColor=<rgba 0 0 0 1>
     }
    }
   }
  }
  GridPanel_BPInfo {widgetClassPath=&GridPanel, namedSlotHost=ExpandableArea_Step_BPSetup, widget=GridPanel_BPInfo}
  TextBlock_9 {
   widgetClassPath=&TextBlock
   parent=GridPanel_BPInfo
   widget=TextBlock_9
   properties {
    text=Parent BP :
    font {
     bForceMonospaced=false
     bMaterialIsStencil=false
     fontMaterial=null
     fontObject=&Roboto
     letterSpacing=0
     monospacedWidth=1
     size=9
     skewAmount=0
     typefaceFontName=Bold
     outlineSettings {
      bApplyOutlineToDropShadows=false
      bMiteredCorners=false
      bSeparateFillAlpha=false
      outlineMaterial=null
      outlineSize=0
      outlineColor=<rgba 0 0 0 1>
     }
    }
   }
   slot {class=GridSlot, properties {row=2}}
  }
  BP_Info_ParentBP {
   widgetClassPath=&TextBlock
   bIsVariable=true
   parent=GridPanel_BPInfo
   widget=BP_Info_ParentBP
   properties {font {
     bForceMonospaced=false
     bMaterialIsStencil=false
     fontMaterial=null
     fontObject=&Roboto
     letterSpacing=0
     monospacedWidth=1
     size=9
     skewAmount=0
     typefaceFontName=Bold
     outlineSettings {
      bApplyOutlineToDropShadows=false
      bMiteredCorners=false
      bSeparateFillAlpha=false
      outlineMaterial=null
      outlineSize=0
      outlineColor=<rgba 0 0 0 1>
     }
    }}
   slot {class=GridSlot, properties {column=1, row=2}}
  }
  TextBlock_10 {
   widgetClassPath=&TextBlock
   parent=GridPanel_BPInfo
   widget=TextBlock_10
   properties {
    text=Resource Table :
    visibility=Collapsed
    font {
     bForceMonospaced=false
     bMaterialIsStencil=false
     fontMaterial=null
     fontObject=&Roboto
     letterSpacing=0
     monospacedWidth=1
     size=9
     skewAmount=0
     typefaceFontName=Bold
     outlineSettings {
      bApplyOutlineToDropShadows=false
      bMiteredCorners=false
      bSeparateFillAlpha=false
      outlineMaterial=null
      outlineSize=0
      outlineColor=<rgba 0 0 0 1>
     }
    }
   }
   slot {class=GridSlot, properties {row=3}}
  }
  BP_Info_ResourceTable {
   widgetClassPath=&TextBlock
   bIsVariable=true
   parent=GridPanel_BPInfo
   widget=BP_Info_ResourceTable
   properties {
    visibility=Collapsed
    font {
     bForceMonospaced=false
     bMaterialIsStencil=false
     fontMaterial=null
     fontObject=&Roboto
     letterSpacing=0
     monospacedWidth=1
     size=9
     skewAmount=0
     typefaceFontName=Bold
     outlineSettings {
      bApplyOutlineToDropShadows=false
      bMiteredCorners=false
      bSeparateFillAlpha=false
      outlineMaterial=null
      outlineSize=0
      outlineColor=<rgba 0 0 0 1>
     }
    }
   }
   slot {class=GridSlot, properties {column=1, row=3}}
  }
  TextBlock_11 {
   widgetClassPath=&TextBlock
   parent=GridPanel_BPInfo
   widget=TextBlock_11
   properties {
    text=Resource :
    font {
     bForceMonospaced=false
     bMaterialIsStencil=false
     fontMaterial=null
     fontObject=&Roboto
     letterSpacing=0
     monospacedWidth=1
     size=9
     skewAmount=0
     typefaceFontName=Bold
     outlineSettings {
      bApplyOutlineToDropShadows=false
      bMiteredCorners=false
      bSeparateFillAlpha=false
      outlineMaterial=null
      outlineSize=0
      outlineColor=<rgba 0 0 0 1>
     }
    }
   }
   slot {class=GridSlot, properties {row=4}}
  }
  BP_Info_Resource {
   widgetClassPath=&TextBlock
   bIsVariable=true
   parent=GridPanel_BPInfo
   widget=BP_Info_Resource
   properties {font {
     bForceMonospaced=false
     bMaterialIsStencil=false
     fontMaterial=null
     fontObject=&Roboto
     letterSpacing=0
     monospacedWidth=1
     size=9
     skewAmount=0
     typefaceFontName=Bold
     outlineSettings {
      bApplyOutlineToDropShadows=false
      bMiteredCorners=false
      bSeparateFillAlpha=false
      outlineMaterial=null
      outlineSize=0
      outlineColor=<rgba 0 0 0 1>
     }
    }}
   slot {class=GridSlot, properties {column=1, row=4}}
  }
  TextBlock_12 {
   widgetClassPath=&TextBlock
   parent=GridPanel_BPInfo
   widget=TextBlock_12
   properties {font {
     bForceMonospaced=false
     bMaterialIsStencil=false
     fontMaterial=null
     fontObject=&Roboto
     letterSpacing=0
     monospacedWidth=1
     size=9
     skewAmount=0
     typefaceFontName=Bold
     outlineSettings {
      bApplyOutlineToDropShadows=false
      bMiteredCorners=false
      bSeparateFillAlpha=false
      outlineMaterial=null
      outlineSize=0
      outlineColor=<rgba 0 0 0 1>
     }
    }}
   slot {class=GridSlot, properties {row=5}}
  }
  ExpandableArea_Step_LOD {
   widgetClassPath=&EditorUtilityExpandableArea
   bIsVariable=true
   parent=HorizontalBox_92
   widget=ExpandableArea_Step_LOD
   slot {class=HorizontalBoxSlot, properties {padding=<ltrb 5 5 5 5>, size {sizeRule=Fill, value=1}}}
  }
  HorizontalBox_0 {widgetClassPath=&HorizontalBox, namedSlotHost=ExpandableArea_Step_LOD, widget=HorizontalBox_0}
  Text_LOD {
   widgetClassPath=&TextBlock
   parent=HorizontalBox_0
   widget=Text_LOD
   properties {
    text="LOD : "
    font {
     bForceMonospaced=false
     bMaterialIsStencil=false
     fontMaterial=null
     fontObject=&Roboto
     letterSpacing=0
     monospacedWidth=1
     size=10.5
     skewAmount=0
     typefaceFontName=Bold
     outlineSettings {
      bApplyOutlineToDropShadows=false
      bMiteredCorners=false
      bSeparateFillAlpha=false
      outlineMaterial=null
      outlineSize=0
      outlineColor=<rgba 0 0 0 1>
     }
    }
   }
   slot {class=HorizontalBoxSlot}
  }
  TaskStatus_LOD {
   widgetClassPath=&TextBlock
   bIsVariable=true
   parent=HorizontalBox_0
   widget=TaskStatus_LOD
   properties {font {
     bForceMonospaced=false
     bMaterialIsStencil=false
     fontMaterial=null
     fontObject=&Roboto
     letterSpacing=0
     monospacedWidth=1
     size=10.5
     skewAmount=0
     typefaceFontName=Bold
     outlineSettings {
      bApplyOutlineToDropShadows=false
      bMiteredCorners=false
      bSeparateFillAlpha=false
      outlineMaterial=null
      outlineSize=0
      outlineColor=<rgba 0 0 0 1>
     }
    }}
   slot {class=HorizontalBoxSlot}
  }
  GridPanel_BPInfo_1 {widgetClassPath=&GridPanel, namedSlotHost=ExpandableArea_Step_LOD, widget=GridPanel_BPInfo_1}
  TextBlock {
   widgetClassPath=&TextBlock
   parent=GridPanel_BPInfo_1
   widget=TextBlock
   properties {
    text=Num of LODs :
    font {
     bForceMonospaced=false
     bMaterialIsStencil=false
     fontMaterial=null
     fontObject=&Roboto
     letterSpacing=0
     monospacedWidth=1
     size=9
     skewAmount=0
     typefaceFontName=Bold
     outlineSettings {
      bApplyOutlineToDropShadows=false
      bMiteredCorners=false
      bSeparateFillAlpha=false
      outlineMaterial=null
      outlineSize=0
      outlineColor=<rgba 0 0 0 1>
     }
    }
   }
   slot {class=GridSlot}
  }
  Mesh_Info_LODs {
   widgetClassPath=&TextBlock
   bIsVariable=true
   parent=GridPanel_BPInfo_1
   widget=Mesh_Info_LODs
   properties {font {
     bForceMonospaced=false
     bMaterialIsStencil=false
     fontMaterial=null
     fontObject=&Roboto
     letterSpacing=0
     monospacedWidth=1
     size=9
     skewAmount=0
     typefaceFontName=Bold
     outlineSettings {
      bApplyOutlineToDropShadows=false
      bMiteredCorners=false
      bSeparateFillAlpha=false
      outlineMaterial=null
      outlineSize=0
      outlineColor=<rgba 0 0 0 1>
     }
    }}
   slot {class=GridSlot, properties {column=1}}
  }
  Text_LOD1 {
   widgetClassPath=&TextBlock
   parent=GridPanel_BPInfo_1
   widget=Text_LOD1
   properties {
    text="LOD1 : "
    font {
     bForceMonospaced=false
     bMaterialIsStencil=false
     fontMaterial=null
     fontObject=&Roboto
     letterSpacing=0
     monospacedWidth=1
     size=9
     skewAmount=0
     typefaceFontName=Bold
     outlineSettings {
      bApplyOutlineToDropShadows=false
      bMiteredCorners=false
      bSeparateFillAlpha=false
      outlineMaterial=null
      outlineSize=0
      outlineColor=<rgba 0 0 0 1>
     }
    }
   }
   slot {class=GridSlot, properties {row=1}}
  }
  Mesh_Info_LOD1 {
   widgetClassPath=&TextBlock
   bIsVariable=true
   parent=GridPanel_BPInfo_1
   widget=Mesh_Info_LOD1
   properties {font {
     bForceMonospaced=false
     bMaterialIsStencil=false
     fontMaterial=null
     fontObject=&Roboto
     letterSpacing=0
     monospacedWidth=1
     size=9
     skewAmount=0
     typefaceFontName=Bold
     outlineSettings {
      bApplyOutlineToDropShadows=false
      bMiteredCorners=false
      bSeparateFillAlpha=false
      outlineMaterial=null
      outlineSize=0
      outlineColor=<rgba 0 0 0 1>
     }
    }}
   slot {class=GridSlot, properties {column=1, row=1}}
  }
  Text_LOD2 {
   widgetClassPath=&TextBlock
   parent=GridPanel_BPInfo_1
   widget=Text_LOD2
   properties {
    text="LOD2 : "
    font {
     bForceMonospaced=false
     bMaterialIsStencil=false
     fontMaterial=null
     fontObject=&Roboto
     letterSpacing=0
     monospacedWidth=1
     size=9
     skewAmount=0
     typefaceFontName=Bold
     outlineSettings {
      bApplyOutlineToDropShadows=false
      bMiteredCorners=false
      bSeparateFillAlpha=false
      outlineMaterial=null
      outlineSize=0
      outlineColor=<rgba 0 0 0 1>
     }
    }
   }
   slot {class=GridSlot, properties {row=2}}
  }
  Mesh_Info_LOD2 {
   widgetClassPath=&TextBlock
   bIsVariable=true
   parent=GridPanel_BPInfo_1
   widget=Mesh_Info_LOD2
   properties {font {
     bForceMonospaced=false
     bMaterialIsStencil=false
     fontMaterial=null
     fontObject=&Roboto
     letterSpacing=0
     monospacedWidth=1
     size=9
     skewAmount=0
     typefaceFontName=Bold
     outlineSettings {
      bApplyOutlineToDropShadows=false
      bMiteredCorners=false
      bSeparateFillAlpha=false
      outlineMaterial=null
      outlineSize=0
      outlineColor=<rgba 0 0 0 1>
     }
    }}
   slot {class=GridSlot, properties {column=1, row=2}}
  }
  Text_LOD3 {
   widgetClassPath=&TextBlock
   parent=GridPanel_BPInfo_1
   widget=Text_LOD3
   properties {
    text="LOD3 : "
    font {
     bForceMonospaced=false
     bMaterialIsStencil=false
     fontMaterial=null
     fontObject=&Roboto
     letterSpacing=0
     monospacedWidth=1
     size=9
     skewAmount=0
     typefaceFontName=Bold
     outlineSettings {
      bApplyOutlineToDropShadows=false
      bMiteredCorners=false
      bSeparateFillAlpha=false
      outlineMaterial=null
      outlineSize=0
      outlineColor=<rgba 0 0 0 1>
     }
    }
   }
   slot {class=GridSlot, properties {row=3}}
  }
  Mesh_Info_LOD3 {
   widgetClassPath=&TextBlock
   bIsVariable=true
   parent=GridPanel_BPInfo_1
   widget=Mesh_Info_LOD3
   properties {font {
     bForceMonospaced=false
     bMaterialIsStencil=false
     fontMaterial=null
     fontObject=&Roboto
     letterSpacing=0
     monospacedWidth=1
     size=9
     skewAmount=0
     typefaceFontName=Bold
     outlineSettings {
      bApplyOutlineToDropShadows=false
      bMiteredCorners=false
      bSeparateFillAlpha=false
      outlineMaterial=null
      outlineSize=0
      outlineColor=<rgba 0 0 0 1>
     }
    }}
   slot {class=GridSlot, properties {column=1, row=3}}
  }
  Spacer_1 {widgetClassPath=&Spacer, parent=VerticalBox_83, widget=Spacer_1, properties {size=<1 10>}, slot {class=VerticalBoxSlot}}
  ExpandableArea_Mesh {
   widgetClassPath=&EditorUtilityExpandableArea
   bIsVariable=true
   parent=VerticalBox_83
   widget=ExpandableArea_Mesh
   properties {bIsExpanded=true}
   slot {class=VerticalBoxSlot}
  }
  TextBlock_66 {
   widgetClassPath=&TextBlock
   namedSlotHost=ExpandableArea_Mesh
   widget=TextBlock_66
   properties {
    text=Mesh
    font {
     bForceMonospaced=false
     bMaterialIsStencil=false
     fontMaterial=null
     fontObject=&Roboto
     letterSpacing=0
     monospacedWidth=1
     size=13.5
     skewAmount=0
     typefaceFontName=Bold
     outlineSettings {
      bApplyOutlineToDropShadows=false
      bMiteredCorners=false
      bSeparateFillAlpha=false
      outlineMaterial=null
      outlineSize=0
      outlineColor=<rgba 0 0 0 1>
     }
    }
   }
  }
  Border {
   widgetClassPath=&Border
   namedSlotHost=ExpandableArea_Mesh
   widget=Border
   properties {
    background {
     bIsDynamicallyLoaded=false
     drawAs=Image
     imageType=NoImage
     mirroring=NoMirror
     resourceName=None
     resourceObject=null
     tiling=NoTile
     imageSize=<32 32>
     margin=<ltrb 0 0 0 0>
     outlineSettings {
      bUseBrushTransparency=false
      roundingType=HalfHeightRadius
      width=0
      color {colorUseRule=UseColor_Specified, specifiedColor=<rgba 0 0 0 0>}
      cornerRadii=<0 0 0 0>
     }
     tintColor {colorUseRule=UseColor_Specified, specifiedColor=<rgba 0.08 0.08 0.08 1>}
     uVRegion {bIsValid=false, max=<0 0>, min=<0 0>}
    }
    padding=<ltrb 10 10 10 10>
   }
  }
  VerticalBox_1 {
   widgetClassPath=&VerticalBox
   parent=Border
   widget=VerticalBox_1
   slot {class=BorderSlot, properties {padding=<ltrb 10 10 10 10>}}
  }
  ExpandableArea_Import {
   widgetClassPath=&EditorUtilityExpandableArea
   bIsVariable=true
   parent=VerticalBox_1
   widget=ExpandableArea_Import
   properties {bIsExpanded=true, visibility=Collapsed}
   slot {class=VerticalBoxSlot}
  }
  TextBlock_342 {
   widgetClassPath=&TextBlock
   namedSlotHost=ExpandableArea_Import
   widget=TextBlock_342
   properties {
    text=Import
    font {
     bForceMonospaced=false
     bMaterialIsStencil=false
     fontMaterial=null
     fontObject=&Roboto
     letterSpacing=0
     monospacedWidth=1
     size=10.5
     skewAmount=0
     typefaceFontName=Bold
     outlineSettings {
      bApplyOutlineToDropShadows=false
      bMiteredCorners=false
      bSeparateFillAlpha=false
      outlineMaterial=null
      outlineSize=0
      outlineColor=<rgba 0 0 0 1>
     }
    }
   }
  }
  Border_38 {
   widgetClassPath=&Border
   namedSlotHost=ExpandableArea_Import
   widget=Border_38
   properties {brushColor=<rgba 0.15 0.15 0.15 1>, padding=<ltrb 5 5 5 5>}
  }
  VerticalBox_220 {
   widgetClassPath=&VerticalBox
   parent=Border_38
   widget=VerticalBox_220
   slot {class=BorderSlot, properties {padding=<ltrb 5 5 5 5>}}
  }
  Spacer_5 {widgetClassPath=&Spacer, parent=VerticalBox_220, widget=Spacer_5, properties {size=<1 10>}, slot {class=VerticalBoxSlot}}
  HorizontalBox_01 {widgetClassPath=&HorizontalBox, parent=VerticalBox_220, widget=HorizontalBox_01, slot {class=VerticalBoxSlot}}
  LOD1_CheckBox {
   widgetClassPath=&EditorUtilityCheckBox
   bIsVariable=true
   parent=HorizontalBox_01
   widget=LOD1_CheckBox
   slot {class=HorizontalBoxSlot}
  }
  LOD1Path_Display {
   widgetClassPath=&SinglePropertyView
   parent=HorizontalBox_01
   widget=LOD1Path_Display
   properties {bIsEnabled=false, nameOverride=LOD1, propertyName=PathLOD1, toolTipText=Path for LOD1 Mesh}
   slot {class=HorizontalBoxSlot, properties {size {sizeRule=Fill, value=1}}}
  }
  Spacer_4 {widgetClassPath=&Spacer, parent=VerticalBox_220, widget=Spacer_4, properties {size=<1 10>}, slot {class=VerticalBoxSlot}}
  HorizontalBox_4 {widgetClassPath=&HorizontalBox, parent=VerticalBox_220, widget=HorizontalBox_4, slot {class=VerticalBoxSlot}}
  LOD2_CheckBox {
   widgetClassPath=&EditorUtilityCheckBox
   bIsVariable=true
   parent=HorizontalBox_4
   widget=LOD2_CheckBox
   slot {class=HorizontalBoxSlot}
  }
  LOD2Path_Display {
   widgetClassPath=&SinglePropertyView
   parent=HorizontalBox_4
   widget=LOD2Path_Display
   properties {bIsEnabled=false, nameOverride=LOD2, propertyName=PathLOD2, toolTipText=Path for LOD2 Mesh}
   slot {class=HorizontalBoxSlot, properties {size {sizeRule=Fill, value=1}}}
  }
  Spacer_6 {widgetClassPath=&Spacer, parent=VerticalBox_220, widget=Spacer_6, properties {size=<1 10>}, slot {class=VerticalBoxSlot}}
  HorizontalBox_5 {widgetClassPath=&HorizontalBox, parent=VerticalBox_220, widget=HorizontalBox_5, slot {class=VerticalBoxSlot}}
  LOD3_CheckBox {
   widgetClassPath=&EditorUtilityCheckBox
   bIsVariable=true
   parent=HorizontalBox_5
   widget=LOD3_CheckBox
   slot {class=HorizontalBoxSlot}
  }
  LOD3Path_Display {
   widgetClassPath=&SinglePropertyView
   parent=HorizontalBox_5
   widget=LOD3Path_Display
   properties {bIsEnabled=false, nameOverride=LOD3, propertyName=PathLOD3, toolTipText=Path for LOD3 Mesh}
   slot {class=HorizontalBoxSlot, properties {size {sizeRule=Fill, value=1}}}
  }
  Spacer_7 {widgetClassPath=&Spacer, parent=VerticalBox_220, widget=Spacer_7, properties {size=<1 10>}, slot {class=VerticalBoxSlot}}
  HorizontalBox_6 {widgetClassPath=&HorizontalBox, parent=VerticalBox_220, widget=HorizontalBox_6, slot {class=VerticalBoxSlot}}
  SM_Button_ImportLODs {
   widgetClassPath=&UIFrameworkCustomButtonWidget
   bIsVariable=true
   parent=HorizontalBox_6
   widget=SM_Button_ImportLODs
   properties {widgetStyle {
     clickedSlateSound {resourceObject=null}
     disabled {
      bIsDynamicallyLoaded=false
      drawAs=RoundedBox
      imageType=NoImage
      mirroring=NoMirror
      resourceName=None
      resourceObject=null
      tiling=NoTile
      imageSize=<0 0>
      margin=<ltrb 0 0 0 0>
      outlineSettings {
       bUseBrushTransparency=false
       roundingType=FixedRadius
       width=1
       color {colorUseRule=UseColor_ColorTable, specifiedColor=<rgba 0 0 0 0>}
       cornerRadii=<4 4 4 4>
      }
      tintColor {colorUseRule=UseColor_Specified, specifiedColor=<rgba 0.0395 0.0395 0.0395 1>}
      uVRegion {bIsValid=false, max=<0 0>, min=<0 0>}
     }
     disabledForeground {colorUseRule=UseColor_ColorTable, specifiedColor=<rgba 0 0 0 0>}
     hovered {
      bIsDynamicallyLoaded=false
      drawAs=RoundedBox
      imageType=NoImage
      mirroring=NoMirror
      resourceName=None
      resourceObject=null
      tiling=NoTile
      imageSize=<32 32>
      margin=<ltrb 0 0 0 0>
      outlineSettings {
       bUseBrushTransparency=false
       roundingType=FixedRadius
       width=1
       color {colorUseRule=UseColor_ColorTable, specifiedColor=<rgba 0 0 0 0>}
       cornerRadii=<4 4 4 4>
      }
      tintColor {colorUseRule=UseColor_Specified, specifiedColor=<rgba 0.0953 0.0953 0.0953 1>}
      uVRegion {bIsValid=false, max=<0 0>, min=<0 0>}
     }
     hoveredForeground {colorUseRule=UseColor_ColorTable, specifiedColor=<rgba 0 0 0 0>}
     hoveredSlateSound {resourceObject=null}
     normal {
      bIsDynamicallyLoaded=false
      drawAs=RoundedBox
      imageType=NoImage
      mirroring=NoMirror
      resourceName=None
      resourceObject=null
      tiling=NoTile
      imageSize=<32 32>
      margin=<ltrb 0 0 0 0>
      outlineSettings {
       bUseBrushTransparency=false
       roundingType=FixedRadius
       width=1
       color {colorUseRule=UseColor_ColorTable, specifiedColor=<rgba 0 0 0 0>}
       cornerRadii=<4 4 4 4>
      }
      tintColor {colorUseRule=UseColor_Specified, specifiedColor=<rgba 0.0395 0.0395 0.0395 1>}
      uVRegion {bIsValid=false, max=<0 0>, min=<0 0>}
     }
     normalForeground {colorUseRule=UseColor_ColorTable, specifiedColor=<rgba 0 0 0 0>}
     normalPadding=<ltrb 12 1.5 12 1.5>
     pressed {
      bIsDynamicallyLoaded=false
      drawAs=RoundedBox
      imageType=NoImage
      mirroring=NoMirror
      resourceName=None
      resourceObject=null
      tiling=NoTile
      imageSize=<32 32>
      margin=<ltrb 0 0 0 0>
      outlineSettings {
       bUseBrushTransparency=false
       roundingType=FixedRadius
       width=1
       color {colorUseRule=UseColor_ColorTable, specifiedColor=<rgba 0 0 0 0>}
       cornerRadii=<4 4 4 4>
      }
      tintColor {colorUseRule=UseColor_Specified, specifiedColor=<rgba 0.0284 0.0284 0.0284 1>}
      uVRegion {bIsValid=false, max=<0 0>, min=<0 0>}
     }
     pressedForeground {colorUseRule=UseColor_ColorTable, specifiedColor=<rgba 0 0 0 0>}
     pressedPadding=<ltrb 12 2.5 12 0.5>
     pressedSlateSound {resourceObject=null}
    }}
   slot {class=HorizontalBoxSlot, properties {horizontalAlignment=HAlign_Center, size {sizeRule=Fill, value=1}}}
  }
  TextBlock_13 {
   widgetClassPath=&TextBlock
   parent=SM_Button_ImportLODs
   widget=TextBlock_13
   properties {
    text=Import LODs
    font {
     bForceMonospaced=false
     bMaterialIsStencil=false
     fontMaterial=null
     fontObject=&Roboto
     letterSpacing=0
     monospacedWidth=1
     size=10.5
     skewAmount=0
     typefaceFontName=Bold
     outlineSettings {
      bApplyOutlineToDropShadows=false
      bMiteredCorners=false
      bSeparateFillAlpha=false
      outlineMaterial=null
      outlineSize=0
      outlineColor=<rgba 0 0 0 1>
     }
    }
   }
   slot {class=ButtonSlot}
  }
  Spacer_611 {widgetClassPath=&Spacer, parent=VerticalBox_1, widget=Spacer_611, properties {size=<1 10>}, slot {class=VerticalBoxSlot}}
  ExpandableArea_Edit {
   widgetClassPath=&EditorUtilityExpandableArea
   bIsVariable=true
   parent=VerticalBox_1
   widget=ExpandableArea_Edit
   properties {bIsExpanded=true}
   slot {class=VerticalBoxSlot}
  }
  TextBlock_14 {
   widgetClassPath=&TextBlock
   namedSlotHost=ExpandableArea_Edit
   widget=TextBlock_14
   properties {
    text=Edit
    font {
     bForceMonospaced=false
     bMaterialIsStencil=false
     fontMaterial=null
     fontObject=&Roboto
     letterSpacing=0
     monospacedWidth=1
     size=10.5
     skewAmount=0
     typefaceFontName=Bold
     outlineSettings {
      bApplyOutlineToDropShadows=false
      bMiteredCorners=false
      bSeparateFillAlpha=false
      outlineMaterial=null
      outlineSize=0
      outlineColor=<rgba 0 0 0 1>
     }
    }
   }
  }
  Border_1 {
   widgetClassPath=&Border
   namedSlotHost=ExpandableArea_Edit
   widget=Border_1
   properties {
    background {
     bIsDynamicallyLoaded=false
     drawAs=Image
     imageType=NoImage
     mirroring=NoMirror
     resourceName=None
     resourceObject=null
     tiling=NoTile
     imageSize=<32 32>
     margin=<ltrb 0 0 0 0>
     outlineSettings {
      bUseBrushTransparency=false
      roundingType=HalfHeightRadius
      width=0
      color {colorUseRule=UseColor_Specified, specifiedColor=<rgba 0 0 0 0>}
      cornerRadii=<0 0 0 0>
     }
     tintColor {colorUseRule=UseColor_Specified, specifiedColor=<rgba 0.15 0.15 0.15 1>}
     uVRegion {bIsValid=false, max=<0 0>, min=<0 0>}
    }
    padding=<ltrb 5 5 5 5>
   }
  }
  VerticalBox_2 {widgetClassPath=&VerticalBox, parent=Border_1, widget=VerticalBox_2, slot {class=BorderSlot, properties {padding=<ltrb 5 5 5 5>}}}
  HorizontalBox_91 {widgetClassPath=&HorizontalBox, parent=VerticalBox_2, widget=HorizontalBox_91, slot {class=VerticalBoxSlot}}
  SM_Button_AssignLODs {
   widgetClassPath=&UIFrameworkCustomButtonWidget
   bIsVariable=true
   parent=HorizontalBox_91
   widget=SM_Button_AssignLODs
   properties {widgetStyle {
     clickedSlateSound {resourceObject=null}
     disabled {
      bIsDynamicallyLoaded=false
      drawAs=RoundedBox
      imageType=NoImage
      mirroring=NoMirror
      resourceName=None
      resourceObject=null
      tiling=NoTile
      imageSize=<0 0>
      margin=<ltrb 0 0 0 0>
      outlineSettings {
       bUseBrushTransparency=false
       roundingType=FixedRadius
       width=1
       color {colorUseRule=UseColor_ColorTable, specifiedColor=<rgba 0 0 0 0>}
       cornerRadii=<4 4 4 4>
      }
      tintColor {colorUseRule=UseColor_Specified, specifiedColor=<rgba 0.0395 0.0395 0.0395 1>}
      uVRegion {bIsValid=false, max=<0 0>, min=<0 0>}
     }
     disabledForeground {colorUseRule=UseColor_ColorTable, specifiedColor=<rgba 0 0 0 0>}
     hovered {
      bIsDynamicallyLoaded=false
      drawAs=RoundedBox
      imageType=NoImage
      mirroring=NoMirror
      resourceName=None
      resourceObject=null
      tiling=NoTile
      imageSize=<32 32>
      margin=<ltrb 0 0 0 0>
      outlineSettings {
       bUseBrushTransparency=false
       roundingType=FixedRadius
       width=1
       color {colorUseRule=UseColor_ColorTable, specifiedColor=<rgba 0 0 0 0>}
       cornerRadii=<4 4 4 4>
      }
      tintColor {colorUseRule=UseColor_Specified, specifiedColor=<rgba 0.0953 0.0953 0.0953 1>}
      uVRegion {bIsValid=false, max=<0 0>, min=<0 0>}
     }
     hoveredForeground {colorUseRule=UseColor_ColorTable, specifiedColor=<rgba 0 0 0 0>}
     hoveredSlateSound {resourceObject=null}
     normal {
      bIsDynamicallyLoaded=false
      drawAs=RoundedBox
      imageType=NoImage
      mirroring=NoMirror
      resourceName=None
      resourceObject=null
      tiling=NoTile
      imageSize=<32 32>
      margin=<ltrb 0 0 0 0>
      outlineSettings {
       bUseBrushTransparency=false
       roundingType=FixedRadius
       width=1
       color {colorUseRule=UseColor_ColorTable, specifiedColor=<rgba 0 0 0 0>}
       cornerRadii=<4 4 4 4>
      }
      tintColor {colorUseRule=UseColor_Specified, specifiedColor=<rgba 0.0395 0.0395 0.0395 1>}
      uVRegion {bIsValid=false, max=<0 0>, min=<0 0>}
     }
     normalForeground {colorUseRule=UseColor_ColorTable, specifiedColor=<rgba 0 0 0 0>}
     normalPadding=<ltrb 12 1.5 12 1.5>
     pressed {
      bIsDynamicallyLoaded=false
      drawAs=RoundedBox
      imageType=NoImage
      mirroring=NoMirror
      resourceName=None
      resourceObject=null
      tiling=NoTile
      imageSize=<32 32>
      margin=<ltrb 0 0 0 0>
      outlineSettings {
       bUseBrushTransparency=false
       roundingType=FixedRadius
       width=1
       color {colorUseRule=UseColor_ColorTable, specifiedColor=<rgba 0 0 0 0>}
       cornerRadii=<4 4 4 4>
      }
      tintColor {colorUseRule=UseColor_Specified, specifiedColor=<rgba 0.0284 0.0284 0.0284 1>}
      uVRegion {bIsValid=false, max=<0 0>, min=<0 0>}
     }
     pressedForeground {colorUseRule=UseColor_ColorTable, specifiedColor=<rgba 0 0 0 0>}
     pressedPadding=<ltrb 12 2.5 12 0.5>
     pressedSlateSound {resourceObject=null}
    }}
   slot {class=HorizontalBoxSlot, properties {padding=<ltrb 5 5 5 5>, size {sizeRule=Fill, value=1}}}
  }
  TextBlock_16 {
   widgetClassPath=&TextBlock
   parent=SM_Button_AssignLODs
   widget=TextBlock_16
   properties {
    text=Apply Perf Settings
    toolTipText="Previously Assign LODs tool.\r\n\r\nApplies the following settings to a static mesh:\r\n\r\nNanite Nanite Support = True\r\nNanite Fallback = Relative Error\r\nFallback Relative Error = 0.0\r\n\r\nQuality Level MinLODs:\r\n-Low= 2\r\n-Medium = 2\r\n\r\nLOD For Collision = 2\r\n\r\nScreen Sizes for Architecture:\r\n-LOD0 = 1.0\r\n-LOD1 = 0.4\r\n-LOD2 = 0.2\r\n-LOD3 = 0.03\r\n\r\nScreen Sizes for Props:\r\n-LOD0 = 1.0\r\n-LOD1 = 0.2\r\n-LOD2 = 0.08\r\n"
    font {
     bForceMonospaced=false
     bMaterialIsStencil=false
     fontMaterial=null
     fontObject=&Roboto
     letterSpacing=0
     monospacedWidth=1
     size=10.5
     skewAmount=0
     typefaceFontName=Bold
     outlineSettings {
      bApplyOutlineToDropShadows=false
      bMiteredCorners=false
      bSeparateFillAlpha=false
      outlineMaterial=null
      outlineSize=0
      outlineColor=<rgba 0 0 0 1>
     }
    }
   }
   slot {class=ButtonSlot}
  }
  SM_Button_ReduceNanite {
   widgetClassPath=&UIFrameworkCustomButtonWidget
   bIsVariable=true
   parent=HorizontalBox_91
   widget=SM_Button_ReduceNanite
   properties {
    toolTipText="Auto reduces nanite vert count using the Trim Relative Error property. \r\nWill reduce until it's below 200k or less than 10x the vert count of the fallback mesh. Whichever is lower."
    widgetStyle {
     clickedSlateSound {resourceObject=null}
     disabled {
      bIsDynamicallyLoaded=false
      drawAs=RoundedBox
      imageType=NoImage
      mirroring=NoMirror
      resourceName=None
      resourceObject=null
      tiling=NoTile
      imageSize=<0 0>
      margin=<ltrb 0 0 0 0>
      outlineSettings {
       bUseBrushTransparency=false
       roundingType=FixedRadius
       width=1
       color {colorUseRule=UseColor_ColorTable, specifiedColor=<rgba 0 0 0 0>}
       cornerRadii=<4 4 4 4>
      }
      tintColor {colorUseRule=UseColor_Specified, specifiedColor=<rgba 0.0395 0.0395 0.0395 1>}
      uVRegion {bIsValid=false, max=<0 0>, min=<0 0>}
     }
     disabledForeground {colorUseRule=UseColor_ColorTable, specifiedColor=<rgba 0 0 0 0>}
     hovered {
      bIsDynamicallyLoaded=false
      drawAs=RoundedBox
      imageType=NoImage
      mirroring=NoMirror
      resourceName=None
      resourceObject=null
      tiling=NoTile
      imageSize=<32 32>
      margin=<ltrb 0 0 0 0>
      outlineSettings {
       bUseBrushTransparency=false
       roundingType=FixedRadius
       width=1
       color {colorUseRule=UseColor_ColorTable, specifiedColor=<rgba 0 0 0 0>}
       cornerRadii=<4 4 4 4>
      }
      tintColor {colorUseRule=UseColor_Specified, specifiedColor=<rgba 0.0953 0.0953 0.0953 1>}
      uVRegion {bIsValid=false, max=<0 0>, min=<0 0>}
     }
     hoveredForeground {colorUseRule=UseColor_ColorTable, specifiedColor=<rgba 0 0 0 0>}
     hoveredSlateSound {resourceObject=null}
     normal {
      bIsDynamicallyLoaded=false
      drawAs=RoundedBox
      imageType=NoImage
      mirroring=NoMirror
      resourceName=None
      resourceObject=null
      tiling=NoTile
      imageSize=<32 32>
      margin=<ltrb 0 0 0 0>
      outlineSettings {
       bUseBrushTransparency=false
       roundingType=FixedRadius
       width=1
       color {colorUseRule=UseColor_ColorTable, specifiedColor=<rgba 0 0 0 0>}
       cornerRadii=<4 4 4 4>
      }
      tintColor {colorUseRule=UseColor_Specified, specifiedColor=<rgba 0.0395 0.0395 0.0395 1>}
      uVRegion {bIsValid=false, max=<0 0>, min=<0 0>}
     }
     normalForeground {colorUseRule=UseColor_ColorTable, specifiedColor=<rgba 0 0 0 0>}
     normalPadding=<ltrb 12 1.5 12 1.5>
     pressed {
      bIsDynamicallyLoaded=false
      drawAs=RoundedBox
      imageType=NoImage
      mirroring=NoMirror
      resourceName=None
      resourceObject=null
      tiling=NoTile
      imageSize=<32 32>
      margin=<ltrb 0 0 0 0>
      outlineSettings {
       bUseBrushTransparency=false
       roundingType=FixedRadius
       width=1
       color {colorUseRule=UseColor_ColorTable, specifiedColor=<rgba 0 0 0 0>}
       cornerRadii=<4 4 4 4>
      }
      tintColor {colorUseRule=UseColor_Specified, specifiedColor=<rgba 0.0284 0.0284 0.0284 1>}
      uVRegion {bIsValid=false, max=<0 0>, min=<0 0>}
     }
     pressedForeground {colorUseRule=UseColor_ColorTable, specifiedColor=<rgba 0 0 0 0>}
     pressedPadding=<ltrb 12 2.5 12 0.5>
     pressedSlateSound {resourceObject=null}
    }
   }
   slot {class=HorizontalBoxSlot, properties {padding=<ltrb 5 5 5 5>, size {sizeRule=Fill, value=1}}}
  }
  TextBlock_17 {
   widgetClassPath=&TextBlock
   parent=SM_Button_ReduceNanite
   widget=TextBlock_17
   properties {
    text=ReduceNanite
    font {
     bForceMonospaced=false
     bMaterialIsStencil=false
     fontMaterial=null
     fontObject=&Roboto
     letterSpacing=0
     monospacedWidth=1
     size=10.5
     skewAmount=0
     typefaceFontName=Bold
     outlineSettings {
      bApplyOutlineToDropShadows=false
      bMiteredCorners=false
      bSeparateFillAlpha=false
      outlineMaterial=null
      outlineSize=0
      outlineColor=<rgba 0 0 0 1>
     }
    }
   }
   slot {class=ButtonSlot}
  }
  SM_Button_GenerateNavEdges {
   widgetClassPath=&UIFrameworkCustomButtonWidget
   bIsVariable=true
   parent=HorizontalBox_91
   widget=SM_Button_GenerateNavEdges
   properties {
    toolTipText="Generates navigation edge data on static mesh for hurdling system.\r\nShould be run on all window meshes."
    widgetStyle {
     clickedSlateSound {resourceObject=null}
     disabled {
      bIsDynamicallyLoaded=false
      drawAs=RoundedBox
      imageType=NoImage
      mirroring=NoMirror
      resourceName=None
      resourceObject=null
      tiling=NoTile
      imageSize=<0 0>
      margin=<ltrb 0 0 0 0>
      outlineSettings {
       bUseBrushTransparency=false
       roundingType=FixedRadius
       width=1
       color {colorUseRule=UseColor_ColorTable, specifiedColor=<rgba 0 0 0 0>}
       cornerRadii=<4 4 4 4>
      }
      tintColor {colorUseRule=UseColor_Specified, specifiedColor=<rgba 0.0395 0.0395 0.0395 1>}
      uVRegion {bIsValid=false, max=<0 0>, min=<0 0>}
     }
     disabledForeground {colorUseRule=UseColor_ColorTable, specifiedColor=<rgba 0 0 0 0>}
     hovered {
      bIsDynamicallyLoaded=false
      drawAs=RoundedBox
      imageType=NoImage
      mirroring=NoMirror
      resourceName=None
      resourceObject=null
      tiling=NoTile
      imageSize=<32 32>
      margin=<ltrb 0 0 0 0>
      outlineSettings {
       bUseBrushTransparency=false
       roundingType=FixedRadius
       width=1
       color {colorUseRule=UseColor_ColorTable, specifiedColor=<rgba 0 0 0 0>}
       cornerRadii=<4 4 4 4>
      }
      tintColor {colorUseRule=UseColor_Specified, specifiedColor=<rgba 0.0953 0.0953 0.0953 1>}
      uVRegion {bIsValid=false, max=<0 0>, min=<0 0>}
     }
     hoveredForeground {colorUseRule=UseColor_ColorTable, specifiedColor=<rgba 0 0 0 0>}
     hoveredSlateSound {resourceObject=null}
     normal {
      bIsDynamicallyLoaded=false
      drawAs=RoundedBox
      imageType=NoImage
      mirroring=NoMirror
      resourceName=None
      resourceObject=null
      tiling=NoTile
      imageSize=<32 32>
      margin=<ltrb 0 0 0 0>
      outlineSettings {
       bUseBrushTransparency=false
       roundingType=FixedRadius
       width=1
       color {colorUseRule=UseColor_ColorTable, specifiedColor=<rgba 0 0 0 0>}
       cornerRadii=<4 4 4 4>
      }
      tintColor {colorUseRule=UseColor_Specified, specifiedColor=<rgba 0.0395 0.0395 0.0395 1>}
      uVRegion {bIsValid=false, max=<0 0>, min=<0 0>}
     }
     normalForeground {colorUseRule=UseColor_ColorTable, specifiedColor=<rgba 0 0 0 0>}
     normalPadding=<ltrb 12 1.5 12 1.5>
     pressed {
      bIsDynamicallyLoaded=false
      drawAs=RoundedBox
      imageType=NoImage
      mirroring=NoMirror
      resourceName=None
      resourceObject=null
      tiling=NoTile
      imageSize=<32 32>
      margin=<ltrb 0 0 0 0>
      outlineSettings {
       bUseBrushTransparency=false
       roundingType=FixedRadius
       width=1
       color {colorUseRule=UseColor_ColorTable, specifiedColor=<rgba 0 0 0 0>}
       cornerRadii=<4 4 4 4>
      }
      tintColor {colorUseRule=UseColor_Specified, specifiedColor=<rgba 0.0284 0.0284 0.0284 1>}
      uVRegion {bIsValid=false, max=<0 0>, min=<0 0>}
     }
     pressedForeground {colorUseRule=UseColor_ColorTable, specifiedColor=<rgba 0 0 0 0>}
     pressedPadding=<ltrb 12 2.5 12 0.5>
     pressedSlateSound {resourceObject=null}
    }
   }
   slot {class=HorizontalBoxSlot, properties {padding=<ltrb 5 5 5 5>, size {sizeRule=Fill, value=1}}}
  }
  TextBlock_18 {
   widgetClassPath=&TextBlock
   parent=SM_Button_GenerateNavEdges
   widget=TextBlock_18
   properties {
    text="Generate Navigation Edges (42.00+)"
    font {
     bForceMonospaced=false
     bMaterialIsStencil=false
     fontMaterial=null
     fontObject=&Roboto
     letterSpacing=0
     monospacedWidth=1
     size=10.5
     skewAmount=0
     typefaceFontName=Bold
     outlineSettings {
      bApplyOutlineToDropShadows=false
      bMiteredCorners=false
      bSeparateFillAlpha=false
      outlineMaterial=null
      outlineSize=0
      outlineColor=<rgba 0 0 0 1>
     }
    }
   }
   slot {class=ButtonSlot}
  }
  Spacer_79 {widgetClassPath=&Spacer, parent=VerticalBox_83, widget=Spacer_79, properties {size=<1 10>}, slot {class=VerticalBoxSlot}}
  ExpandableArea_BP {
   widgetClassPath=&EditorUtilityExpandableArea
   bIsVariable=true
   parent=VerticalBox_83
   widget=ExpandableArea_BP
   properties {bIsExpanded=true}
   slot {class=VerticalBoxSlot}
  }
  Text_Create {
   widgetClassPath=&TextBlock
   namedSlotHost=ExpandableArea_BP
   widget=Text_Create
   properties {
    text=BP
    font {
     bForceMonospaced=false
     bMaterialIsStencil=false
     fontMaterial=null
     fontObject=&Roboto
     letterSpacing=0
     monospacedWidth=1
     size=13.5
     skewAmount=0
     typefaceFontName=Bold
     outlineSettings {
      bApplyOutlineToDropShadows=false
      bMiteredCorners=false
      bSeparateFillAlpha=false
      outlineMaterial=null
      outlineSize=0
      outlineColor=<rgba 0 0 0 1>
     }
    }
   }
  }
  Border_35 {
   widgetClassPath=&Border
   namedSlotHost=ExpandableArea_BP
   widget=Border_35
   properties {
    background {
     bIsDynamicallyLoaded=false
     drawAs=Image
     imageType=NoImage
     mirroring=NoMirror
     resourceName=None
     resourceObject=null
     tiling=NoTile
     imageSize=<32 32>
     margin=<ltrb 0 0 0 0>
     outlineSettings {
      bUseBrushTransparency=false
      roundingType=HalfHeightRadius
      width=0
      color {colorUseRule=UseColor_Specified, specifiedColor=<rgba 0 0 0 0>}
      cornerRadii=<0 0 0 0>
     }
     tintColor {colorUseRule=UseColor_Specified, specifiedColor=<rgba 0.08 0.08 0.08 1>}
     uVRegion {bIsValid=false, max=<0 0>, min=<0 0>}
    }
    padding=<ltrb 10 10 10 10>
   }
  }
  VerticalBox {
   widgetClassPath=&VerticalBox
   parent=Border_35
   widget=VerticalBox
   slot {class=BorderSlot, properties {padding=<ltrb 10 10 10 10>}}
  }
  ExpandableArea_BP_Create {
   widgetClassPath=&EditorUtilityExpandableArea
   bIsVariable=true
   parent=VerticalBox
   widget=ExpandableArea_BP_Create
   properties {bIsExpanded=true}
   slot {class=VerticalBoxSlot}
  }
  Create {
   widgetClassPath=&TextBlock
   namedSlotHost=ExpandableArea_BP_Create
   widget=Create
   properties {
    text=Create
    font {
     bForceMonospaced=false
     bMaterialIsStencil=false
     fontMaterial=null
     fontObject=&Roboto
     letterSpacing=0
     monospacedWidth=1
     size=10.5
     skewAmount=0
     typefaceFontName=Bold
     outlineSettings {
      bApplyOutlineToDropShadows=false
      bMiteredCorners=false
      bSeparateFillAlpha=false
      outlineMaterial=null
      outlineSize=0
      outlineColor=<rgba 0 0 0 1>
     }
    }
   }
  }
  Border_90 {
   widgetClassPath=&Border
   namedSlotHost=ExpandableArea_BP_Create
   widget=Border_90
   properties {
    background {
     bIsDynamicallyLoaded=false
     drawAs=Image
     imageType=NoImage
     mirroring=NoMirror
     resourceName=None
     resourceObject=null
     tiling=NoTile
     imageSize=<32 32>
     margin=<ltrb 0 0 0 0>
     outlineSettings {
      bUseBrushTransparency=false
      roundingType=HalfHeightRadius
      width=0
      color {colorUseRule=UseColor_Specified, specifiedColor=<rgba 0 0 0 0>}
      cornerRadii=<0 0 0 0>
     }
     tintColor {colorUseRule=UseColor_Specified, specifiedColor=<rgba 0.15 0.15 0.15 1>}
     uVRegion {bIsValid=false, max=<0 0>, min=<0 0>}
    }
    padding=<ltrb 5 5 5 5>
   }
  }
  VerticalBox_65 {
   widgetClassPath=&VerticalBox
   parent=Border_90
   widget=VerticalBox_65
   slot {class=BorderSlot, properties {padding=<ltrb 5 5 5 5>}}
  }
  HorizontalBox {widgetClassPath=&HorizontalBox, parent=VerticalBox_65, widget=HorizontalBox, slot {class=VerticalBoxSlot}}
  BP_Button_AutoPropBlueprint {
   widgetClassPath=&UIFrameworkCustomButtonWidget
   bIsVariable=true
   parent=HorizontalBox
   widget=BP_Button_AutoPropBlueprint
   properties {
    toolTipText="Creates a prop blueprint for selected static meshes.\r\nAutomatically setting the appropriate blockout prop parent based on static mesh size"
    widgetStyle {
     clickedSlateSound {resourceObject=null}
     disabled {
      bIsDynamicallyLoaded=false
      drawAs=RoundedBox
      imageType=NoImage
      mirroring=NoMirror
      resourceName=None
      resourceObject=null
      tiling=NoTile
      imageSize=<0 0>
      margin=<ltrb 0 0 0 0>
      outlineSettings {
       bUseBrushTransparency=false
       roundingType=FixedRadius
       width=1
       color {colorUseRule=UseColor_ColorTable, specifiedColor=<rgba 0 0 0 0>}
       cornerRadii=<4 4 4 4>
      }
      tintColor {colorUseRule=UseColor_Specified, specifiedColor=<rgba 0.0395 0.0395 0.0395 1>}
      uVRegion {bIsValid=false, max=<0 0>, min=<0 0>}
     }
     disabledForeground {colorUseRule=UseColor_ColorTable, specifiedColor=<rgba 0 0 0 0>}
     hovered {
      bIsDynamicallyLoaded=false
      drawAs=RoundedBox
      imageType=NoImage
      mirroring=NoMirror
      resourceName=None
      resourceObject=null
      tiling=NoTile
      imageSize=<32 32>
      margin=<ltrb 0 0 0 0>
      outlineSettings {
       bUseBrushTransparency=false
       roundingType=FixedRadius
       width=1
       color {colorUseRule=UseColor_ColorTable, specifiedColor=<rgba 0 0 0 0>}
       cornerRadii=<4 4 4 4>
      }
      tintColor {colorUseRule=UseColor_Specified, specifiedColor=<rgba 0.0953 0.0953 0.0953 1>}
      uVRegion {bIsValid=false, max=<0 0>, min=<0 0>}
     }
     hoveredForeground {colorUseRule=UseColor_ColorTable, specifiedColor=<rgba 0 0 0 0>}
     hoveredSlateSound {resourceObject=null}
     normal {
      bIsDynamicallyLoaded=false
      drawAs=RoundedBox
      imageType=NoImage
      mirroring=NoMirror
      resourceName=None
      resourceObject=null
      tiling=NoTile
      imageSize=<32 32>
      margin=<ltrb 0 0 0 0>
      outlineSettings {
       bUseBrushTransparency=false
       roundingType=FixedRadius
       width=1
       color {colorUseRule=UseColor_ColorTable, specifiedColor=<rgba 0 0 0 0>}
       cornerRadii=<4 4 4 4>
      }
      tintColor {colorUseRule=UseColor_Specified, specifiedColor=<rgba 0.0395 0.0395 0.0395 1>}
      uVRegion {bIsValid=false, max=<0 0>, min=<0 0>}
     }
     normalForeground {colorUseRule=UseColor_ColorTable, specifiedColor=<rgba 0 0 0 0>}
     normalPadding=<ltrb 12 1.5 12 1.5>
     pressed {
      bIsDynamicallyLoaded=false
      drawAs=RoundedBox
      imageType=NoImage
      mirroring=NoMirror
      resourceName=None
      resourceObject=null
      tiling=NoTile
      imageSize=<32 32>
      margin=<ltrb 0 0 0 0>
      outlineSettings {
       bUseBrushTransparency=false
       roundingType=FixedRadius
       width=1
       color {colorUseRule=UseColor_ColorTable, specifiedColor=<rgba 0 0 0 0>}
       cornerRadii=<4 4 4 4>
      }
      tintColor {colorUseRule=UseColor_Specified, specifiedColor=<rgba 0.0284 0.0284 0.0284 1>}
      uVRegion {bIsValid=false, max=<0 0>, min=<0 0>}
     }
     pressedForeground {colorUseRule=UseColor_ColorTable, specifiedColor=<rgba 0 0 0 0>}
     pressedPadding=<ltrb 12 2.5 12 0.5>
     pressedSlateSound {resourceObject=null}
    }
   }
   slot {class=HorizontalBoxSlot, properties {padding=<ltrb 5 5 5 5>, size {sizeRule=Fill, value=1}}}
  }
  TextBlock_19 {
   widgetClassPath=&TextBlock
   parent=BP_Button_AutoPropBlueprint
   widget=TextBlock_19
   properties {
    text=Auto Prop Blueprint
    font {
     bForceMonospaced=false
     bMaterialIsStencil=false
     fontMaterial=null
     fontObject=&Roboto
     letterSpacing=0
     monospacedWidth=1
     size=10.5
     skewAmount=0
     typefaceFontName=Bold
     outlineSettings {
      bApplyOutlineToDropShadows=false
      bMiteredCorners=false
      bSeparateFillAlpha=false
      outlineMaterial=null
      outlineSize=0
      outlineColor=<rgba 0 0 0 1>
     }
    }
   }
   slot {class=ButtonSlot}
  }
  Spacer_8 {widgetClassPath=&Spacer, parent=VerticalBox, widget=Spacer_8, properties {size=<1 10>}, slot {class=VerticalBoxSlot}}
  ExpandableArea_BP_Edit {
   widgetClassPath=&EditorUtilityExpandableArea
   bIsVariable=true
   parent=VerticalBox
   widget=ExpandableArea_BP_Edit
   properties {bIsExpanded=true}
   slot {class=VerticalBoxSlot}
  }
  Create_1 {
   widgetClassPath=&TextBlock
   namedSlotHost=ExpandableArea_BP_Edit
   widget=Create_1
   properties {
    text=Edit
    font {
     bForceMonospaced=false
     bMaterialIsStencil=false
     fontMaterial=null
     fontObject=&Roboto
     letterSpacing=0
     monospacedWidth=1
     size=10.5
     skewAmount=0
     typefaceFontName=Bold
     outlineSettings {
      bApplyOutlineToDropShadows=false
      bMiteredCorners=false
      bSeparateFillAlpha=false
      outlineMaterial=null
      outlineSize=0
      outlineColor=<rgba 0 0 0 1>
     }
    }
   }
  }
  Border_2 {
   widgetClassPath=&Border
   namedSlotHost=ExpandableArea_BP_Edit
   widget=Border_2
   properties {
    background {
     bIsDynamicallyLoaded=false
     drawAs=Image
     imageType=NoImage
     mirroring=NoMirror
     resourceName=None
     resourceObject=null
     tiling=NoTile
     imageSize=<32 32>
     margin=<ltrb 0 0 0 0>
     outlineSettings {
      bUseBrushTransparency=false
      roundingType=HalfHeightRadius
      width=0
      color {colorUseRule=UseColor_Specified, specifiedColor=<rgba 0 0 0 0>}
      cornerRadii=<0 0 0 0>
     }
     tintColor {colorUseRule=UseColor_Specified, specifiedColor=<rgba 0.15 0.15 0.15 1>}
     uVRegion {bIsValid=false, max=<0 0>, min=<0 0>}
    }
    padding=<ltrb 5 5 5 5>
   }
  }
  VerticalBox_3 {widgetClassPath=&VerticalBox, parent=Border_2, widget=VerticalBox_3, slot {class=BorderSlot, properties {padding=<ltrb 5 5 5 5>}}}
  HorizontalBox_125 {widgetClassPath=&HorizontalBox, parent=VerticalBox_3, widget=HorizontalBox_125, slot {class=VerticalBoxSlot}}
  VerticalBox_667 {
   widgetClassPath=&VerticalBox
   parent=HorizontalBox_125
   widget=VerticalBox_667
   slot {class=HorizontalBoxSlot, properties {size {sizeRule=Fill, value=1}}}
  }
  HorizontalBox_127 {
   widgetClassPath=&HorizontalBox
   parent=VerticalBox_667
   widget=HorizontalBox_127
   slot {class=VerticalBoxSlot, properties {size {sizeRule=Fill, value=1}}}
  }
  BP_Button_CreateBSGVersion {
   widgetClassPath=&UIFrameworkCustomButtonWidget
   bIsVariable=true
   parent=HorizontalBox_127
   widget=BP_Button_CreateBSGVersion
   properties {
    toolTipText="Creates a Building Static Geo (BSG) version of an existing blueprint.\r\nNew BSG will be created within a BSG folder in the current location of the existing BP.\r\n\r\nUse this when needed to create an indestructible version of an existing desctructible BP."
    widgetStyle {
     clickedSlateSound {resourceObject=null}
     disabled {
      bIsDynamicallyLoaded=false
      drawAs=RoundedBox
      imageType=NoImage
      mirroring=NoMirror
      resourceName=None
      resourceObject=null
      tiling=NoTile
      imageSize=<0 0>
      margin=<ltrb 0 0 0 0>
      outlineSettings {
       bUseBrushTransparency=false
       roundingType=FixedRadius
       width=1
       color {colorUseRule=UseColor_ColorTable, specifiedColor=<rgba 0 0 0 0>}
       cornerRadii=<4 4 4 4>
      }
      tintColor {colorUseRule=UseColor_Specified, specifiedColor=<rgba 0.0395 0.0395 0.0395 1>}
      uVRegion {bIsValid=false, max=<0 0>, min=<0 0>}
     }
     disabledForeground {colorUseRule=UseColor_ColorTable, specifiedColor=<rgba 0 0 0 0>}
     hovered {
      bIsDynamicallyLoaded=false
      drawAs=RoundedBox
      imageType=NoImage
      mirroring=NoMirror
      resourceName=None
      resourceObject=null
      tiling=NoTile
      imageSize=<32 32>
      margin=<ltrb 0 0 0 0>
      outlineSettings {
       bUseBrushTransparency=false
       roundingType=FixedRadius
       width=1
       color {colorUseRule=UseColor_ColorTable, specifiedColor=<rgba 0 0 0 0>}
       cornerRadii=<4 4 4 4>
      }
      tintColor {colorUseRule=UseColor_Specified, specifiedColor=<rgba 0.0953 0.0953 0.0953 1>}
      uVRegion {bIsValid=false, max=<0 0>, min=<0 0>}
     }
     hoveredForeground {colorUseRule=UseColor_ColorTable, specifiedColor=<rgba 0 0 0 0>}
     hoveredSlateSound {resourceObject=null}
     normal {
      bIsDynamicallyLoaded=false
      drawAs=RoundedBox
      imageType=NoImage
      mirroring=NoMirror
      resourceName=None
      resourceObject=null
      tiling=NoTile
      imageSize=<32 32>
      margin=<ltrb 0 0 0 0>
      outlineSettings {
       bUseBrushTransparency=false
       roundingType=FixedRadius
       width=1
       color {colorUseRule=UseColor_ColorTable, specifiedColor=<rgba 0 0 0 0>}
       cornerRadii=<4 4 4 4>
      }
      tintColor {colorUseRule=UseColor_Specified, specifiedColor=<rgba 0.0395 0.0395 0.0395 1>}
      uVRegion {bIsValid=false, max=<0 0>, min=<0 0>}
     }
     normalForeground {colorUseRule=UseColor_ColorTable, specifiedColor=<rgba 0 0 0 0>}
     normalPadding=<ltrb 12 1.5 12 1.5>
     pressed {
      bIsDynamicallyLoaded=false
      drawAs=RoundedBox
      imageType=NoImage
      mirroring=NoMirror
      resourceName=None
      resourceObject=null
      tiling=NoTile
      imageSize=<32 32>
      margin=<ltrb 0 0 0 0>
      outlineSettings {
       bUseBrushTransparency=false
       roundingType=FixedRadius
       width=1
       color {colorUseRule=UseColor_ColorTable, specifiedColor=<rgba 0 0 0 0>}
       cornerRadii=<4 4 4 4>
      }
      tintColor {colorUseRule=UseColor_Specified, specifiedColor=<rgba 0.0284 0.0284 0.0284 1>}
      uVRegion {bIsValid=false, max=<0 0>, min=<0 0>}
     }
     pressedForeground {colorUseRule=UseColor_ColorTable, specifiedColor=<rgba 0 0 0 0>}
     pressedPadding=<ltrb 12 2.5 12 0.5>
     pressedSlateSound {resourceObject=null}
    }
   }
   slot {class=HorizontalBoxSlot, properties {padding=<ltrb 5 5 5 5>, size {sizeRule=Fill, value=1}}}
  }
  TextBlock_15 {
   widgetClassPath=&TextBlock
   parent=BP_Button_CreateBSGVersion
   widget=TextBlock_15
   properties {
    text=Create BSG Version
    font {
     bForceMonospaced=false
     bMaterialIsStencil=false
     fontMaterial=null
     fontObject=&Roboto
     letterSpacing=0
     monospacedWidth=1
     size=10.5
     skewAmount=0
     typefaceFontName=Bold
     outlineSettings {
      bApplyOutlineToDropShadows=false
      bMiteredCorners=false
      bSeparateFillAlpha=false
      outlineMaterial=null
      outlineSize=0
      outlineColor=<rgba 0 0 0 1>
     }
    }
   }
   slot {class=ButtonSlot}
  }
  BP_Button_SetCentroid {
   widgetClassPath=&UIFrameworkCustomButtonWidget
   bIsVariable=true
   parent=HorizontalBox_127
   widget=BP_Button_SetCentroid
   properties {
    toolTipText="Sets the \"Centroid Offset\" property on a blueprint to the center of it's primary static mesh component's bounds.\r\nThis drives health bar location on blueprints. Use this when there is a misplaced health bar."
    widgetStyle {
     clickedSlateSound {resourceObject=null}
     disabled {
      bIsDynamicallyLoaded=false
      drawAs=RoundedBox
      imageType=NoImage
      mirroring=NoMirror
      resourceName=None
      resourceObject=null
      tiling=NoTile
      imageSize=<0 0>
      margin=<ltrb 0 0 0 0>
      outlineSettings {
       bUseBrushTransparency=false
       roundingType=FixedRadius
       width=1
       color {colorUseRule=UseColor_ColorTable, specifiedColor=<rgba 0 0 0 0>}
       cornerRadii=<4 4 4 4>
      }
      tintColor {colorUseRule=UseColor_Specified, specifiedColor=<rgba 0.0395 0.0395 0.0395 1>}
      uVRegion {bIsValid=false, max=<0 0>, min=<0 0>}
     }
     disabledForeground {colorUseRule=UseColor_ColorTable, specifiedColor=<rgba 0 0 0 0>}
     hovered {
      bIsDynamicallyLoaded=false
      drawAs=RoundedBox
      imageType=NoImage
      mirroring=NoMirror
      resourceName=None
      resourceObject=null
      tiling=NoTile
      imageSize=<32 32>
      margin=<ltrb 0 0 0 0>
      outlineSettings {
       bUseBrushTransparency=false
       roundingType=FixedRadius
       width=1
       color {colorUseRule=UseColor_ColorTable, specifiedColor=<rgba 0 0 0 0>}
       cornerRadii=<4 4 4 4>
      }
      tintColor {colorUseRule=UseColor_Specified, specifiedColor=<rgba 0.0953 0.0953 0.0953 1>}
      uVRegion {bIsValid=false, max=<0 0>, min=<0 0>}
     }
     hoveredForeground {colorUseRule=UseColor_ColorTable, specifiedColor=<rgba 0 0 0 0>}
     hoveredSlateSound {resourceObject=null}
     normal {
      bIsDynamicallyLoaded=false
      drawAs=RoundedBox
      imageType=NoImage
      mirroring=NoMirror
      resourceName=None
      resourceObject=null
      tiling=NoTile
      imageSize=<32 32>
      margin=<ltrb 0 0 0 0>
      outlineSettings {
       bUseBrushTransparency=false
       roundingType=FixedRadius
       width=1
       color {colorUseRule=UseColor_ColorTable, specifiedColor=<rgba 0 0 0 0>}
       cornerRadii=<4 4 4 4>
      }
      tintColor {colorUseRule=UseColor_Specified, specifiedColor=<rgba 0.0395 0.0395 0.0395 1>}
      uVRegion {bIsValid=false, max=<0 0>, min=<0 0>}
     }
     normalForeground {colorUseRule=UseColor_ColorTable, specifiedColor=<rgba 0 0 0 0>}
     normalPadding=<ltrb 12 1.5 12 1.5>
     pressed {
      bIsDynamicallyLoaded=false
      drawAs=RoundedBox
      imageType=NoImage
      mirroring=NoMirror
      resourceName=None
      resourceObject=null
      tiling=NoTile
      imageSize=<32 32>
      margin=<ltrb 0 0 0 0>
      outlineSettings {
       bUseBrushTransparency=false
       roundingType=FixedRadius
       width=1
       color {colorUseRule=UseColor_ColorTable, specifiedColor=<rgba 0 0 0 0>}
       cornerRadii=<4 4 4 4>
      }
      tintColor {colorUseRule=UseColor_Specified, specifiedColor=<rgba 0.0284 0.0284 0.0284 1>}
      uVRegion {bIsValid=false, max=<0 0>, min=<0 0>}
     }
     pressedForeground {colorUseRule=UseColor_ColorTable, specifiedColor=<rgba 0 0 0 0>}
     pressedPadding=<ltrb 12 2.5 12 0.5>
     pressedSlateSound {resourceObject=null}
    }
   }
   slot {class=HorizontalBoxSlot, properties {padding=<ltrb 5 5 5 5>, size {sizeRule=Fill, value=1}}}
  }
  Text_SetCentroid {
   widgetClassPath=&TextBlock
   parent=BP_Button_SetCentroid
   widget=Text_SetCentroid
   properties {
    text=Set Centroid By Bounds
    font {
     bForceMonospaced=false
     bMaterialIsStencil=false
     fontMaterial=null
     fontObject=&Roboto
     letterSpacing=0
     monospacedWidth=1
     size=10.5
     skewAmount=0
     typefaceFontName=Bold
     outlineSettings {
      bApplyOutlineToDropShadows=false
      bMiteredCorners=false
      bSeparateFillAlpha=false
      outlineMaterial=null
      outlineSize=0
      outlineColor=<rgba 0 0 0 1>
     }
    }
   }
   slot {class=ButtonSlot}
  }
  EditorUtilityExpandableArea_186 {
   widgetClassPath=&EditorUtilityExpandableArea
   bIsVariable=true
   parent=HorizontalBox_127
   widget=EditorUtilityExpandableArea_186
   properties {bIsExpanded=true}
   slot {class=HorizontalBoxSlot, properties {verticalAlignment=VAlign_Center, size {sizeRule=Fill, value=1}}}
  }
  Create_2 {
   widgetClassPath=&TextBlock
   namedSlotHost=EditorUtilityExpandableArea_186
   widget=Create_2
   properties {
    text=Prop Building Settings
    font {
     bForceMonospaced=false
     bMaterialIsStencil=false
     fontMaterial=null
     fontObject=&Roboto
     letterSpacing=0
     monospacedWidth=1
     size=12
     skewAmount=0
     typefaceFontName=Bold
     outlineSettings {
      bApplyOutlineToDropShadows=false
      bMiteredCorners=false
      bSeparateFillAlpha=false
      outlineMaterial=null
      outlineSize=0
      outlineColor=<rgba 0 0 0 1>
     }
    }
   }
  }
  HorizontalBox_1 {widgetClassPath=&HorizontalBox, namedSlotHost=EditorUtilityExpandableArea_186, widget=HorizontalBox_1}
  BP_Button_BuildThrough {
   widgetClassPath=&UIFrameworkCustomButtonWidget
   bIsVariable=true
   parent=HorizontalBox_1
   widget=BP_Button_BuildThrough
   properties {
    toolTipText="Applies building settings to BP to allow players to build through props without destroying them.\r\nUse on large and extra large props."
    widgetStyle {
     clickedSlateSound {resourceObject=null}
     disabled {
      bIsDynamicallyLoaded=false
      drawAs=RoundedBox
      imageType=NoImage
      mirroring=NoMirror
      resourceName=None
      resourceObject=null
      tiling=NoTile
      imageSize=<0 0>
      margin=<ltrb 0 0 0 0>
      outlineSettings {
       bUseBrushTransparency=false
       roundingType=FixedRadius
       width=1
       color {colorUseRule=UseColor_ColorTable, specifiedColor=<rgba 0 0 0 0>}
       cornerRadii=<4 4 4 4>
      }
      tintColor {colorUseRule=UseColor_Specified, specifiedColor=<rgba 0.0395 0.0395 0.0395 1>}
      uVRegion {bIsValid=false, max=<0 0>, min=<0 0>}
     }
     disabledForeground {colorUseRule=UseColor_ColorTable, specifiedColor=<rgba 0 0 0 0>}
     hovered {
      bIsDynamicallyLoaded=false
      drawAs=RoundedBox
      imageType=NoImage
      mirroring=NoMirror
      resourceName=None
      resourceObject=null
      tiling=NoTile
      imageSize=<32 32>
      margin=<ltrb 0 0 0 0>
      outlineSettings {
       bUseBrushTransparency=false
       roundingType=FixedRadius
       width=1
       color {colorUseRule=UseColor_ColorTable, specifiedColor=<rgba 0 0 0 0>}
       cornerRadii=<4 4 4 4>
      }
      tintColor {colorUseRule=UseColor_Specified, specifiedColor=<rgba 0.0953 0.0953 0.0953 1>}
      uVRegion {bIsValid=false, max=<0 0>, min=<0 0>}
     }
     hoveredForeground {colorUseRule=UseColor_ColorTable, specifiedColor=<rgba 0 0 0 0>}
     hoveredSlateSound {resourceObject=null}
     normal {
      bIsDynamicallyLoaded=false
      drawAs=RoundedBox
      imageType=NoImage
      mirroring=NoMirror
      resourceName=None
      resourceObject=null
      tiling=NoTile
      imageSize=<32 32>
      margin=<ltrb 0 0 0 0>
      outlineSettings {
       bUseBrushTransparency=false
       roundingType=FixedRadius
       width=1
       color {colorUseRule=UseColor_ColorTable, specifiedColor=<rgba 0 0 0 0>}
       cornerRadii=<4 4 4 4>
      }
      tintColor {colorUseRule=UseColor_Specified, specifiedColor=<rgba 0.0395 0.0395 0.0395 1>}
      uVRegion {bIsValid=false, max=<0 0>, min=<0 0>}
     }
     normalForeground {colorUseRule=UseColor_ColorTable, specifiedColor=<rgba 0 0 0 0>}
     normalPadding=<ltrb 12 1.5 12 1.5>
     pressed {
      bIsDynamicallyLoaded=false
      drawAs=RoundedBox
      imageType=NoImage
      mirroring=NoMirror
      resourceName=None
      resourceObject=null
      tiling=NoTile
      imageSize=<32 32>
      margin=<ltrb 0 0 0 0>
      outlineSettings {
       bUseBrushTransparency=false
       roundingType=FixedRadius
       width=1
       color {colorUseRule=UseColor_ColorTable, specifiedColor=<rgba 0 0 0 0>}
       cornerRadii=<4 4 4 4>
      }
      tintColor {colorUseRule=UseColor_Specified, specifiedColor=<rgba 0.0284 0.0284 0.0284 1>}
      uVRegion {bIsValid=false, max=<0 0>, min=<0 0>}
     }
     pressedForeground {colorUseRule=UseColor_ColorTable, specifiedColor=<rgba 0 0 0 0>}
     pressedPadding=<ltrb 12 2.5 12 0.5>
     pressedSlateSound {resourceObject=null}
    }
   }
   slot {class=HorizontalBoxSlot, properties {size {sizeRule=Fill, value=1}}}
  }
  Text_SetCentroid_1 {
   widgetClassPath=&TextBlock
   parent=BP_Button_BuildThrough
   widget=Text_SetCentroid_1
   properties {
    text=Build Through
    font {
     bForceMonospaced=false
     bMaterialIsStencil=false
     fontMaterial=null
     fontObject=&Roboto
     letterSpacing=0
     monospacedWidth=1
     size=10.5
     skewAmount=0
     typefaceFontName=Bold
     outlineSettings {
      bApplyOutlineToDropShadows=false
      bMiteredCorners=false
      bSeparateFillAlpha=false
      outlineMaterial=null
      outlineSize=0
      outlineColor=<rgba 0 0 0 1>
     }
    }
   }
   slot {class=ButtonSlot}
  }
  BP_Button_DestroyOnBuild {
   widgetClassPath=&UIFrameworkCustomButtonWidget
   bIsVariable=true
   parent=HorizontalBox_1
   widget=BP_Button_DestroyOnBuild
   properties {
    toolTipText="Applies building settings to BP so they destroy on placement of player builds.\r\nUse on small and medium props."
    widgetStyle {
     clickedSlateSound {resourceObject=null}
     disabled {
      bIsDynamicallyLoaded=false
      drawAs=RoundedBox
      imageType=NoImage
      mirroring=NoMirror
      resourceName=None
      resourceObject=null
      tiling=NoTile
      imageSize=<0 0>
      margin=<ltrb 0 0 0 0>
      outlineSettings {
       bUseBrushTransparency=false
       roundingType=FixedRadius
       width=1
       color {colorUseRule=UseColor_ColorTable, specifiedColor=<rgba 0 0 0 0>}
       cornerRadii=<4 4 4 4>
      }
      tintColor {colorUseRule=UseColor_Specified, specifiedColor=<rgba 0.0395 0.0395 0.0395 1>}
      uVRegion {bIsValid=false, max=<0 0>, min=<0 0>}
     }
     disabledForeground {colorUseRule=UseColor_ColorTable, specifiedColor=<rgba 0 0 0 0>}
     hovered {
      bIsDynamicallyLoaded=false
      drawAs=RoundedBox
      imageType=NoImage
      mirroring=NoMirror
      resourceName=None
      resourceObject=null
      tiling=NoTile
      imageSize=<32 32>
      margin=<ltrb 0 0 0 0>
      outlineSettings {
       bUseBrushTransparency=false
       roundingType=FixedRadius
       width=1
       color {colorUseRule=UseColor_ColorTable, specifiedColor=<rgba 0 0 0 0>}
       cornerRadii=<4 4 4 4>
      }
      tintColor {colorUseRule=UseColor_Specified, specifiedColor=<rgba 0.0953 0.0953 0.0953 1>}
      uVRegion {bIsValid=false, max=<0 0>, min=<0 0>}
     }
     hoveredForeground {colorUseRule=UseColor_ColorTable, specifiedColor=<rgba 0 0 0 0>}
     hoveredSlateSound {resourceObject=null}
     normal {
      bIsDynamicallyLoaded=false
      drawAs=RoundedBox
      imageType=NoImage
      mirroring=NoMirror
      resourceName=None
      resourceObject=null
      tiling=NoTile
      imageSize=<32 32>
      margin=<ltrb 0 0 0 0>
      outlineSettings {
       bUseBrushTransparency=false
       roundingType=FixedRadius
       width=1
       color {colorUseRule=UseColor_ColorTable, specifiedColor=<rgba 0 0 0 0>}
       cornerRadii=<4 4 4 4>
      }
      tintColor {colorUseRule=UseColor_Specified, specifiedColor=<rgba 0.0395 0.0395 0.0395 1>}
      uVRegion {bIsValid=false, max=<0 0>, min=<0 0>}
     }
     normalForeground {colorUseRule=UseColor_ColorTable, specifiedColor=<rgba 0 0 0 0>}
     normalPadding=<ltrb 12 1.5 12 1.5>
     pressed {
      bIsDynamicallyLoaded=false
      drawAs=RoundedBox
      imageType=NoImage
      mirroring=NoMirror
      resourceName=None
      resourceObject=null
      tiling=NoTile
      imageSize=<32 32>
      margin=<ltrb 0 0 0 0>
      outlineSettings {
       bUseBrushTransparency=false
       roundingType=FixedRadius
       width=1
       color {colorUseRule=UseColor_ColorTable, specifiedColor=<rgba 0 0 0 0>}
       cornerRadii=<4 4 4 4>
      }
      tintColor {colorUseRule=UseColor_Specified, specifiedColor=<rgba 0.0284 0.0284 0.0284 1>}
      uVRegion {bIsValid=false, max=<0 0>, min=<0 0>}
     }
     pressedForeground {colorUseRule=UseColor_ColorTable, specifiedColor=<rgba 0 0 0 0>}
     pressedPadding=<ltrb 12 2.5 12 0.5>
     pressedSlateSound {resourceObject=null}
    }
   }
   slot {class=HorizontalBoxSlot, properties {size {sizeRule=Fill, value=1}}}
  }
  Text_SetCentroid_2 {
   widgetClassPath=&TextBlock
   parent=BP_Button_DestroyOnBuild
   widget=Text_SetCentroid_2
   properties {
    text=Destroy On Build
    font {
     bForceMonospaced=false
     bMaterialIsStencil=false
     fontMaterial=null
     fontObject=&Roboto
     letterSpacing=0
     monospacedWidth=1
     size=10.5
     skewAmount=0
     typefaceFontName=Bold
     outlineSettings {
      bApplyOutlineToDropShadows=false
      bMiteredCorners=false
      bSeparateFillAlpha=false
      outlineMaterial=null
      outlineSize=0
      outlineColor=<rgba 0 0 0 1>
     }
    }
   }
   slot {class=ButtonSlot}
  }
 ]
}
