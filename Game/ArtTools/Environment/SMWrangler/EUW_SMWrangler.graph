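# Editor utility widget (SMWrangler) for reviewing LODs on selected static meshes, batch-saving edits and exporting lists of meshes with generated LODs.
# PopulateMeshList hides collision meshes by name, creates LODRow widgets, sizes headers to the highest LOD count and syncs name-column width.
# NotifyDirty and NotifyClean maintain counts driving Save All and status text.
# Export flags reduced LOD1s, opening a dialog alongside Python that re-scans the selection into lock-guarded temp files.
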EditorUtilityWidgetBlueprint /Game/ArtTools/Environment/SMWrangler/EUW_SMWrangler
graphs [
 BuildGeneratedLODsString {nodes [
   BuildGeneratedLODsString() {is=title K2Node_FunctionEntry}
   > Clear(TargetArray=Get QualifiedNames() -> (QualifiedNames:Array of Strings))
   > n1 = For Each Loop(Array:Array of Static Mesh Object References=Get LoadedMeshes() -> (LoadedMeshes:Array of Static Mesh Object References)) -> (Array Element:Static Mesh Object Reference)
   Return Node!(
    execute=@n1.Completed
    ReturnValue:String=JoinStringArray(SourceArray=Get QualifiedNames(), Separator="\\n")
    Count:int=Length(TargetArray:Array of Strings=Get QualifiedNames())
   )
   > Branch(Condition="integer > integer"(A=GetNumLODs(self=@n1), B=MakeLiteralInt!(Value:int=1) -> (ReturnValue:int=0)))
   > n2 = GetLodReductionSettings!(
    execute
    self:Static Mesh Editor Subsystem Object Reference=Get StaticMeshEditorSubsystem()
    StaticMesh:Static Mesh Object Reference=@n1
    LodIndex:int=1
   ) -> (then, OutReductionOptions:Mesh Reduction Settings Structure)
   > Branch(Condition="float < float"(
     A=Break Mesh Reduction Settings!(MeshReductionSettings:"Mesh Reduction Settings Structure (by ref)"=@n2) -> (
      PercentTriangles:float=~
      PercentVertices:float=~
      MaxDeviation:float=~
      PixelError:float=~
      WeldingThreshold:float=~
      HardAngleThreshold:float=~
      BaseLODModel:int
      SilhouetteImportance:EMeshFeatureImportance Enum
      TextureImportance:EMeshFeatureImportance Enum
      ShadingImportance:EMeshFeatureImportance Enum
      bRecalculateNormals:bool=~
      bGenerateUniqueLightmapUVs:bool=~
      bKeepSymmetry:bool=~
      bVisibilityAided:bool=~
      bCullOccluded:bool=~
      TerminationCriterion:EStaticMeshReductionTerimationCriterion Enum
      VisibilityAggressiveness:EMeshFeatureImportance Enum
      VertexColorImportance:EMeshFeatureImportance Enum
     ) {is=title K2Node_BreakStruct}
     B=Make Literal Float(Value=1)
    ))
   > Add(TargetArray=Get QualifiedNames(), NewItem=GetObjectName(Object=@n1)) {is=title K2Node_CallArrayFunction}
  ]}
 EventGraph {nodes [
   "On Clicked (BTN_ExportLODs)"() {is=title K2Node_ComponentBoundEvent}
   > n1 = BuildGeneratedLODsString!(execute, self) -> (then, ReturnValue:String, Count:int=0) {is=title K2Node_CallFunction}
   > Execute Python Script(PythonScript="Build String (Integer)"!(
     AppendTo:String
     Prefix:String="import pathlib,tempfile,os; pathlib.Path(os.path.join(tempfile.gettempdir(),'smwrangler_count.txt')).write_text('"
     InInt:int=@n1.Count
     Suffix:String="', encoding='utf-8')"
    ) -> (ReturnValue:String))
   n2 = Get ActiveExportDialog() -> (ActiveExportDialog:EUW SMWrangler Export Dialog Object Reference)
   > n3 = Is Valid(InputObject=@n2)
   n4 = CreateEditorUtilityDialogWidget!(
    execute=[@n3.Is Not Valid, @n6.then]
    WidgetClass:Editor Utility Dialog Widget Class Reference=&/EUW_SMWrangler_ExportDialog.EUW_SMWrangler_ExportDialog_C
   ) -> (then, ReturnValue:EUW SMWrangler Export Dialog Object Reference)
   > n5 = Cast To EUW_SMWrangler_ExportDialog!(execute, Object=@n4) -> (then, CastFailed, AsEUW SMWrangler Export Dialog:EUW SMWrangler Export Dialog Object Reference) {is=title K2Node_DynamicCast}
   > Set LODCount!(execute, LODCount:int=@n1.Count, self:EUW SMWrangler Export Dialog Object Reference=@n5) -> (then, Output_Get:int=0)
   > Set ActiveExportDialog(ActiveExportDialog:EUW SMWrangler Export Dialog Object Reference=@n5) -> (Output_Get:EUW SMWrangler Export Dialog Object Reference)
   > Set MeshNames!(execute, MeshNames:String=@n1, self:EUW SMWrangler Export Dialog Object Reference=@n5) -> (then, Output_Get:String)
   > Execute Python Script(PythonScript="import unreal, pathlib, tempfile, os; lock = os.path.join(tempfile.gettempdir(), 'smwrangler_dialog.lock'); already_open = os.path.exists(lock); unreal.log('SMWP already_open=' + str(already_open)); sel = unreal.EditorUtilityLibrary.get_selected_assets() if not already_open else []; meshes = [a for a in sel if isinstance(a, unreal.StaticMesh)]; subsystem = unreal.get_editor_subsystem(unreal.StaticMeshEditorSubsystem); excl_pre = ('UCX_',); excl_suf = ('_Col','_Collision','_UCX','_ComplexUCX','_CustomCollision','_ComplexCollision','_Complex_Collision','_Custom_Collision'); is_col = lambda n: n.startswith(excl_pre) or n.endswith(excl_suf); is_lod1_generated = lambda m: subsystem.get_lod_count(m) > 1 and subsystem.get_lod_reduction_settings(m, 1).percent_triangles < 1.0; qualified = [m.get_name() for m in meshes if not is_col(m.get_name()) and is_lod1_generated(m)] if not already_open else []; unreal.log('SMWP qualified=' + str(len(qualified))); tmp = os.path.join(tempfile.gettempdir(), 'smwrangler_meshnames.txt'); cnt_file = os.path.join(tempfile.gettempdir(), 'smwrangler_count.txt'); [pathlib.Path(tmp).write_text(chr(10).join(qualified), encoding='utf-8'),  pathlib.Path(cnt_file).write_text(str(len(qualified)), encoding='utf-8'),  pathlib.Path(lock).write_text('open', encoding='utf-8'),  unreal.log('SMWP wrote ' + str(len(qualified)) + ' names')] if not already_open else unreal.log('SMWP skipped, dialog open')")
   > ShowEditorUtilityDialogInstance!(
    execute
    WidgetInstance:Editor Utility Dialog Widget Object Reference=@n4
    OnClosed:"Delegate (by ref)"
    Title:Text="LOCTEXT(\"Export LODs List\")"
    ButtonLayout:EAppMsgType Enum=OkCancel
    Size:Vector 2D Structure
   ) -> (then)
   CloseDialog!(execute=@n3.Is Valid, self:Editor Utility Dialog Widget Object Reference=@n2, Result:EAppReturnType Enum=Cancel) -> (then)
   > n6 = Set ActiveExportDialog()
   n7 = OnClicked_Event() {is=title K2Node_CustomEvent}
   n8 = Get ActiveExportDialog()
   > n9 = Is Valid(InputObject=@n8)
   PopulateMeshList(execute=[@n9.Is Not Valid, @n10.then])
   CloseDialog(execute=@n9.Is Valid, self=@n8, Result=Cancel)
   > n10 = Set ActiveExportDialog()
   n11 = OnSave_Clicked() {is=title K2Node_CustomEvent}
   > SaveAllChanges()
   Event Construct()
   > Assign On Clicked!(
    execute:Exec
    self:Button Object Reference=Get BTN_LoadSelected() -> (BTN_LoadSelected:Button Object Reference)
    Delegate:"Delegate (by ref)"=@n7
   ) -> (then:Exec) {is=title K2Node_AssignDelegate}
   > Assign On Clicked(self=Get BTN_SaveAll() -> (BTN_SaveAll:Button Object Reference), Delegate=@n11)
   > SetIsEnabled(self=Get BTN_SaveAll(), bInIsEnabled=false)
   > SetIsEnabled(self=Get BTN_ExportLODs() -> (BTN_ExportLODs:Button Object Reference), bInIsEnabled=false)
   Set MeshNames(MeshNames=@n1)
  ]}
 IsCollisionMesh {nodes [
   n1 = IsCollisionMesh!() -> (then, InMeshName:String) {is=title K2Node_FunctionEntry}
   n2 = EndsWith(SourceString=@n1, InSuffix=_Custom_Collision)
   n3 = EndsWith(SourceString=@n1, InSuffix=_ComplexUCX)
   n4 = OR Boolean(
    A=OR Boolean(A=StartsWith(SourceString=@n1, InPrefix=UCX_), B=EndsWith(SourceString=@n1, InSuffix=_Col))
    B=EndsWith(SourceString=@n1, InSuffix=_Collision)
   )
   n5 = OR Boolean(
    A=OR Boolean(A=OR Boolean(A=@n4, B=@n3), B=EndsWith(SourceString=@n1, InSuffix=_CustomCollision))
    B=EndsWith(SourceString=@n1, InSuffix=_ComplexCollision)
   )
   > Return Node!(execute, ReturnValue:bool=OR Boolean(A=OR Boolean(A=@n5, B=@n2), B=EndsWith(SourceString=@n1, InSuffix=_UCX)))
  ]}
 NotifyClean {nodes [
   NotifyClean() {is=title K2Node_FunctionEntry}
   > Set SavedCount(SavedCount:int=int + int(A=Get SavedCount() -> (SavedCount:int=0), B=MakeLiteralInt(Value=1) -> (ReturnValue=0))) -> (Output_Get:int=0)
   n1 = MakeLiteralInt(Value=1) -> (ReturnValue=0)
   n2 = Get DirtyCount() -> (DirtyCount:int=0)
   > Set DirtyCount(DirtyCount:int="Clamp (Integer)"(Value=int - int(A=@n2, B=@n1), Min=0, Max=99) -> (ReturnValue=0)) -> (Output_Get:int=0)
   > UpdateSaveStatus(self=Self-Reference())
   > SetIsEnabled(self=Get BTN_SaveAll(), bInIsEnabled="integer > integer"(A=Get DirtyCount() -> (DirtyCount=0)))
  ]}
 NotifyDirty {nodes [
   n1 = NotifyDirty!() -> (then, WasSaved:bool=~) {is=title K2Node_FunctionEntry}
   > Set DirtyCount(DirtyCount=int + int(A=Get DirtyCount() -> (DirtyCount=0), B=MakeLiteralInt(Value=1) -> (ReturnValue=0))) -> (Output_Get=0)
   > n2 = Branch(Condition=@n1)
   n3 = MakeLiteralInt(Value=1) -> (ReturnValue=0)
   n4 = Get SavedCount() -> (SavedCount=0)
   > n5 = Set SavedCount(SavedCount="Clamp (Integer)"(Value=int - int(A=@n4, B=@n3), Min=0, Max=99) -> (ReturnValue=0)) -> (Output_Get=0)
   UpdateSaveStatus(execute=[@n2.else, @n5.then])
   > SetIsEnabled(self=Get BTN_SaveAll(), bInIsEnabled="integer > integer"(A=Get DirtyCount() -> (DirtyCount=0)))
  ]}
 PopulateMeshList {nodes [
   PopulateMeshList() {is=title K2Node_FunctionEntry}
   > n1 = GetSelectedAssetsOfClass(AssetClass=&StaticMesh)
   > Set SavedCount(SavedCount=0) -> (Output_Get=0)
   > SetVisibility(self=Get TXT_SaveStatus() -> (TXT_SaveStatus:Text Object Reference), InVisibility=Collapsed)
   > n2 = GetSelectedAssetsOfClass(AssetClass=&StaticMesh)
   > ClearChildren(self=Get SB_MeshList() -> (SB_MeshList:Scroll Box Object Reference))
   > Set DirtyCount(DirtyCount=0) -> (Output_Get=0)
   > Clear(TargetArray:Array of Static Mesh Object References=Get LoadedMeshes())
   > Set MaxLODCount(MaxLODCount:int=0) -> (Output_Get:int=0)
   > Set CollisionHiddenCount(CollisionHiddenCount:int=0) -> (Output_Get:int=0)
   n3 = Length(TargetArray:Array of Object References=@n1)
   > n4 = For Loop(LastIndex=int - int(A=@n3, B=MakeLiteralInt(Value=1) -> (ReturnValue=0)))
   n5 = Self-Reference()
   RebuildColumnHeaders!(execute=@n4.Completed, self=@n5, InLODCount:int=Get MaxLODCount() -> (MaxLODCount:int=0)) -> (then)
   > n6 = For Loop(LastIndex=int - int(A=@n3, B=MakeLiteralInt(Value=1) -> (ReturnValue=0)))
   > n7 = Cast To StaticMesh(Object="Get (a copy)"(Array:Array of Object References=@n2, Dimension 1=@n6) -> (Output:Object Reference))
   > n8 = IsCollisionMesh!(execute, self=@n5, InMeshName:String=GetObjectName(Object=@n7)) -> (then, ReturnValue:bool) {is=title K2Node_CallFunction}
   > Branch(Condition=NOT Boolean(A=@n8))
   > Add(TargetArray:Array of Static Mesh Object References=Get LoadedMeshes(), NewItem:"Static Mesh Object Reference (by ref)"=@n7)
   > n9 = Create Widget!(
    execute:Exec
    WorldContextObject:Object Reference
    Class:User Widget Class Reference=&/EUW_SMWrangler_LODRow.EUW_SMWrangler_LODRow_C
    OwningPlayer:Player Controller Object Reference
   ) -> (then:Exec, ReturnValue:EUW SMWrangler LODRow Object Reference)
   > n10 = Cast To EUW_SMWrangler_LODRow!(execute, Object=@n9) -> (then, CastFailed, AsEUW SMWrangler LODRow:EUW SMWrangler LODRow Object Reference) {is=title K2Node_DynamicCast}
   > SetupLODRow!(execute, self, InRow:EUW SMWrangler LODRow Object Reference=@n10, InMesh:Static Mesh Object Reference=@n7) -> (then)
   > n11 = AddChild(self=Get SB_MeshList(), Content=@n10)
   > n12 = Cast To ScrollBoxSlot!(execute, Object=@n11) -> (then, CastFailed, AsScroll Box Slot:Scroll Box Slot Object Reference) {is=title K2Node_DynamicCast}
   > SetPadding!(execute, self:Scroll Box Slot Object Reference=@n12, InPadding:Margin Structure=Make Margin(Left=0, Top=1, Right=0, Bottom=0)) -> (then)
   n13 = Get CollisionHiddenCount() -> (CollisionHiddenCount:int=0)
   n14 = Branch(execute=@n6.Completed, Condition="integer > integer"(A=@n13, B=MakeLiteralInt(Value=0) -> (ReturnValue=0)))
   n15 = Append(A="To String (Integer)"(InInt=@n3), B=" mesh(es) loaded")
   n16 = Append(A=" - ", B=Append(A="To String (Integer)"(InInt=@n13), B=" Collision Meshes Hidden"))
   > n17 = "SetText (Text)"(self=Get TXT_Status() -> (TXT_Status:Text Object Reference), InText="To Text (String)"(InString=Append(A=@n15, B=@n16)))
   SyncMeshNameColumnWidth(execute=[@n17.then, @n19.then], self=@n5)
   > n18 = BuildGeneratedLODsString() -> (Count=0)
   > SetIsEnabled(self=Get BTN_ExportLODs(), bInIsEnabled="integer > integer"(A=@n18.Count))
   > SetIsEnabled(self=Get BTN_SaveAll(), bInIsEnabled=false)
   n19 = "SetText (Text)"(execute=@n14.else, self=Get TXT_Status(), InText="To Text (String)"(InString=@n15))
   n20 = Cast To StaticMesh(execute=@n4.LoopBody, Object="Get (a copy)"(Array=@n2, Dimension 1=@n4))
   > n21 = IsCollisionMesh(self=@n5, InMeshName=GetObjectName(Object=@n20))
   > n22 = Branch(Condition=@n21)
   > Set CollisionHiddenCount(CollisionHiddenCount=int + int(A=Get CollisionHiddenCount() -> (CollisionHiddenCount=0), B=MakeLiteralInt(Value=1) -> (ReturnValue=0))) -> (Output_Get=0)
   n23 = GetNumLODs(self=@n20)
   Branch(execute=@n22.else, Condition="integer > integer"(A=@n23, B=Get MaxLODCount() -> (MaxLODCount=0)))
   > Set MaxLODCount(MaxLODCount=@n23) -> (Output_Get=0)
  ]}
 RebuildColumnHeaders {nodes [
   n1 = RebuildColumnHeaders!() -> (then, InLODCount:int) {is=title K2Node_FunctionEntry}
   n2 = Get HB_ColumnHeaders() -> (HB_ColumnHeaders:Horizontal Box Object Reference)
   > ClearChildren(self=@n2)
   > n3 = Create Widget(Class=&/EUW_SMWrangler_HeaderCell.EUW_SMWrangler_HeaderCell_C) -> (ReturnValue:EUW SMWrangler Header Cell Object Reference)
   > SetLabel!(execute, self:EUW SMWrangler Header Cell Object Reference=@n3, InLabel:Text=MakeLiteralText(Value="LOCTEXT(\"Mesh Name\")")) -> (then)
   > SetMinWidth!(execute, self:EUW SMWrangler Header Cell Object Reference=@n3, InWidth:double=306) -> (then)
   > n4 = AddChild(self=@n2, Content=@n3)
   > n5 = Cast To HorizontalBoxSlot!(execute, Object=@n4) -> (then, CastFailed, AsHorizontal Box Slot:Horizontal Box Slot Object Reference) {is=title K2Node_DynamicCast}
   > SetPadding(self:Horizontal Box Slot Object Reference=@n5, InPadding=Make Margin(Left=2, Top=2, Right=0, Bottom=2))
   > n6 = For Loop(LastIndex=int - int(A=@n1, B=MakeLiteralInt(Value=1) -> (ReturnValue=0)))
   n7 = Create Widget(execute=@n6.Completed, Class=&EUW_SMWrangler_HeaderCell_C)
   > SetLabel(self=@n7, InLabel=MakeLiteralText(Value="LOCTEXT(\"Save\")"))
   > SetMinWidth(self=@n7, InWidth=58)
   > n8 = AddChild(self=@n2, Content=@n7)
   > n9 = Cast To HorizontalBoxSlot(Object=@n8)
   > SetPadding(self=@n9, InPadding=Make Margin(Left=2, Top=2, Right=0, Bottom=2))
   n10 = Create Widget(execute=@n6.LoopBody, Class=&EUW_SMWrangler_HeaderCell_C)
   n11 = Append(A=LOD, B="To String (Integer)"(InInt=@n6))
   > SetLabel(self=@n10, InLabel="To Text (String)"(InString=Append(A=@n11, B=" Type")))
   > SetMinWidth(self=@n10, InWidth=98)
   > n12 = AddChild(self=@n2, Content=@n10)
   > n13 = Cast To HorizontalBoxSlot(Object=@n12)
   > SetPadding(self=@n13, InPadding=Make Margin(Left=4, Top=2, Right=0, Bottom=2))
   > n14 = Create Widget(Class=&EUW_SMWrangler_HeaderCell_C)
   > SetLabel(self=@n14, InLabel=MakeLiteralText(Value="LOCTEXT(\"Verts\")"))
   > SetMinWidth(self=@n14, InWidth=78)
   > n15 = AddChild(self=@n2, Content=@n14)
   > n16 = Cast To HorizontalBoxSlot(Object=@n15)
   > SetPadding(self=@n16, InPadding=Make Margin(Left=2, Top=2, Right=0, Bottom=2))
   > n17 = Create Widget(Class=&EUW_SMWrangler_HeaderCell_C)
   > SetLabel(self=@n17, InLabel=MakeLiteralText(Value="LOCTEXT(\"Screen Size\")"))
   > SetMinWidth(self=@n17, InWidth=108)
   > n18 = AddChild(self=@n2, Content=@n17)
   > n19 = Cast To HorizontalBoxSlot(Object=@n18)
   > SetPadding(self=@n19, InPadding=Make Margin(Left=2, Top=2, Right=0, Bottom=2))
  ]}
 SaveAllChanges {nodes [
   SaveAllChanges() {is=title K2Node_FunctionEntry}
   n1 = Get SB_MeshList()
   > n2 = For Loop(LastIndex=int - int(A=GetChildrenCount!(self:Panel Widget Object Reference=@n1) -> (ReturnValue:int=0), B=MakeLiteralInt(Value=1) -> (ReturnValue=0)))
   Set DirtyCount(execute=@n2.Completed, DirtyCount=0) -> (Output_Get=0)
   > UpdateSaveStatus()
   n3 = Cast To EUW_SMWrangler_LODRow(execute=@n2.LoopBody, Object=GetChildAt!(self:Panel Widget Object Reference=@n1, Index:int=@n2) -> (ReturnValue:Widget Object Reference))
   > Branch(Condition=Get IsDirty!(self:EUW SMWrangler LODRow Object Reference=@n3) -> (IsDirty:bool))
   > SaveMesh(self:EUW SMWrangler LODRow Object Reference=@n3)
  ]}
 SetupLODRow {nodes [
   n1 = SetupLODRow!() -> (then, InRow:EUW SMWrangler LODRow Object Reference, InMesh:Static Mesh Object Reference) {is=title K2Node_FunctionEntry}
   > InitRow!(
    execute
    self:EUW SMWrangler LODRow Object Reference=@n1
    InMesh:Static Mesh Object Reference=@n1.InMesh
    InMaxLODCount:int=Get MaxLODCount() -> (MaxLODCount=0)
    InOwningWidget:EUW SMWrangler Object Reference=Self-Reference()
   ) -> (then)
  ]}
 SyncMeshNameColumnWidth {nodes [
   SyncMeshNameColumnWidth() {is=title K2Node_FunctionEntry}
   > ForceLayoutPrepass(self:Widget Object Reference=Self-Reference())
   > Set MaxMeshNameWidth(MaxMeshNameWidth:double=0) -> (Output_Get:double)
   n1 = Get SB_MeshList()
   n2 = GetChildrenCount(self=@n1) -> (ReturnValue=0)
   > n3 = For Loop(LastIndex=int - int(A=@n2, B=MakeLiteralInt(Value=1) -> (ReturnValue=0)))
   n4 = For Loop(execute=@n3.Completed, LastIndex=int - int(A=@n2, B=MakeLiteralInt(Value=1) -> (ReturnValue=0)))
   n5 = Cast To EUW_SMWrangler_HeaderCell!(execute=@n4.Completed, Object=GetChildAt(self=Get HB_ColumnHeaders(), Index=0)) -> (then, CastFailed, AsEUW SMWrangler Header Cell:EUW SMWrangler Header Cell Object Reference) {is=title K2Node_DynamicCast}
   n6 = float + float(A=Get MaxMeshNameWidth() -> (MaxMeshNameWidth:double), B=Make Literal Float(Value=8))
   > SetMinWidth(self=@n5, InWidth=float + float(A=@n6, B=6))
   n7 = Cast To EUW_SMWrangler_LODRow(execute=@n4.LoopBody, Object=GetChildAt(self=@n1, Index=@n4))
   > Set MinDesiredWidth!(
    execute
    MinDesiredWidth:float=@n6
    self:Text Object Reference=Get TXT_MeshName!(self:EUW SMWrangler LODRow Object Reference=@n7) -> (TXT_MeshName:Text Object Reference)
   ) -> (then, Output_Get:float)
   n8 = Cast To EUW_SMWrangler_LODRow(execute=@n3.LoopBody, Object=GetChildAt(self=@n1, Index=@n3))
   n9 = Get TXT_MeshName(self=@n8)
   n10 = BreakVector(InVec="To Vector (Vector2D)"!(InVector2D:Vector 2D Structure=GetDesiredSize!(self:Widget Object Reference=@n9) -> (ReturnValue:Vector 2D Structure), Z:float=0) -> (ReturnValue:Vector))
   > Branch(Condition="float > float"(A=@n10, B=Get MaxMeshNameWidth()))
   > Set MaxMeshNameWidth(MaxMeshNameWidth=@n10)
  ]}
 UpdateSaveStatus {nodes [
   UpdateSaveStatus() {is=title K2Node_FunctionEntry}
   n1 = Get DirtyCount() -> (DirtyCount=0)
   n2 = "integer > integer"(A=@n1, B=MakeLiteralInt(Value=0) -> (ReturnValue=0))
   n3 = Get SavedCount() -> (SavedCount=0)
   n4 = "integer > integer"(A=@n3, B=MakeLiteralInt(Value=0) -> (ReturnValue=0))
   > n5 = Branch(Condition=NOT Boolean(A=OR Boolean(A=@n2, B=@n4)))
   n6 = Get TXT_SaveStatus()
   > SetVisibility(self=@n6, InVisibility=Collapsed)
   SetVisibility(execute=@n5.else, self=@n6, InVisibility=SelfHitTestInvisible)
   n7 = AND Boolean(A=@n2, B=@n4)
   n8 = Make Linear Color!(R:float=1, G:float=0.55, B:float=0, A:float=1) -> (LinearColor:Linear Color Structure) {is=title K2Node_MakeStruct}
   n9 = Select(
    Option 0:Linear Color Structure=Make Linear Color(R=0.2, G=0.85, B=0.2, A=1)
    Option 1:Linear Color Structure=Make Linear Color(R=1, G=0.55, B=0, A=1)
    Index=@n2
   ) -> (ReturnValue:Linear Color Structure)
   > Set ColorAndOpacity!(
    execute
    ColorAndOpacity:Slate Color Structure=Make Slate Color(SpecifiedColor=Select(Option 0=@n9, Option 1=@n8, Index=@n7))
    self:Text Object Reference=@n6
   ) -> (then, Output_Get:Slate Color Structure)
   n10 = AND Boolean(A=@n2, B=@n4)
   n11 = Append(A=Append(A="To String (Integer)"(InInt=@n1), B=" Unsaved  |  "), B=Append(A="To String (Integer)"(InInt=@n3), B=" Saved"))
   n12 = Select(
    Option 0:String=Append(A="To String (Integer)"(InInt=@n3), B=" Mesh(es) Saved")
    Option 1:String=Append(A="To String (Integer)"(InInt=@n1), B=" Unsaved Mesh(es)")
    Index=@n2
   ) -> (ReturnValue:String)
   > "SetText (Text)"(self=@n6, InText="To Text (String)"(InString=Select(Option 0=@n12, Option 1=@n11, Index=@n10)))
  ]}
]
variables [
 ActiveExportDialog {type=EUW_SMWrangler_ExportDialog_C, default=null}
 CollisionHiddenCount {type=integer}
 CurrentChildIndex {type=integer}
 DirtyCount {type=integer}
 LoadedMeshes {type="array<StaticMesh>", default []}
 MaxLODCount {type=integer}
 MaxMeshNameWidth {type=number}
 PendingScreenSizes {type="map<string>", default {}}
 SavedCount {type=integer}
]
widgets {
 info {rootWidgetClass=&VerticalBox, widgetCount=42}
 widgets [
  RootVerticalBox {widgetClassPath=&VerticalBox, widget=RootVerticalBox}
  Spacer_160 {widgetClassPath=&Spacer, parent=RootVerticalBox, widget=Spacer_160, properties {size=<1 20>}, slot {class=VerticalBoxSlot}}
  HeaderRow {
   widgetClassPath=&HorizontalBox
   parent=RootVerticalBox
   widget=HeaderRow
   slot {class=VerticalBoxSlot, properties {horizontalAlignment=HAlign_Center, padding=<ltrb 0 0 0 2>}}
  }
  BTN_LoadSelected {
   widgetClassPath=&Button
   parent=HeaderRow
   widget=BTN_LoadSelected
   properties {
    toolTipText=Loads the Selected Static Meshes from the Active Content Browser
    widgetStyle {
     clickedSlateSound {resourceObject=null}
     disabled {
      bIsDynamicallyLoaded=false
      drawAs=RoundedBox
      imageType=NoImage
      mirroring=NoMirror
      resourceName=None
      resourceObject=null
      tiling=NoTile
      imageSize=<0 0>
      margin=<ltrb 0 0 0 0>
      outlineSettings {
       bUseBrushTransparency=false
       roundingType=FixedRadius
       width=1
       color {colorUseRule=UseColor_ColorTable, specifiedColor=<rgba 0 0 0 0>}
       cornerRadii=<4 4 4 4>
      }
      tintColor {colorUseRule=UseColor_Specified, specifiedColor=<rgba 0.0395 0.0395 0.0395 1>}
      uVRegion {bIsValid=false, max=<0 0>, min=<0 0>}
     }
     disabledForeground {colorUseRule=UseColor_ColorTable, specifiedColor=<rgba 0 0 0 0>}
     hovered {
      bIsDynamicallyLoaded=false
      drawAs=RoundedBox
      imageType=NoImage
      mirroring=NoMirror
      resourceName=None
      resourceObject=null
      tiling=NoTile
      imageSize=<32 32>
      margin=<ltrb 0 0 0 0>
      outlineSettings {
       bUseBrushTransparency=false
       roundingType=FixedRadius
       width=1
       color {colorUseRule=UseColor_ColorTable, specifiedColor=<rgba 0 0 0 0>}
       cornerRadii=<4 4 4 4>
      }
      tintColor {colorUseRule=UseColor_Specified, specifiedColor=<rgba 0.0953 0.0953 0.0953 1>}
      uVRegion {bIsValid=false, max=<0 0>, min=<0 0>}
     }
     hoveredForeground {colorUseRule=UseColor_ColorTable, specifiedColor=<rgba 0 0 0 0>}
     hoveredSlateSound {resourceObject=null}
     normal {
      bIsDynamicallyLoaded=false
      drawAs=RoundedBox
      imageType=NoImage
      mirroring=NoMirror
      resourceName=None
      resourceObject=null
      tiling=NoTile
      imageSize=<32 32>
      margin=<ltrb 0 0 0 0>
      outlineSettings {
       bUseBrushTransparency=false
       roundingType=FixedRadius
       width=1
       color {colorUseRule=UseColor_ColorTable, specifiedColor=<rgba 0 0 0 0>}
       cornerRadii=<4 4 4 4>
      }
      tintColor {colorUseRule=UseColor_Specified, specifiedColor=<rgba 0.0395 0.0395 0.0395 1>}
      uVRegion {bIsValid=false, max=<0 0>, min=<0 0>}
     }
     normalForeground {colorUseRule=UseColor_ColorTable, specifiedColor=<rgba 0 0 0 0>}
     normalPadding=<ltrb 12 1.5 12 1.5>
     pressed {
      bIsDynamicallyLoaded=false
      drawAs=RoundedBox
      imageType=NoImage
      mirroring=NoMirror
      resourceName=None
      resourceObject=null
      tiling=NoTile
      imageSize=<32 32>
      margin=<ltrb 0 0 0 0>
      outlineSettings {
       bUseBrushTransparency=false
       roundingType=FixedRadius
       width=1
       color {colorUseRule=UseColor_ColorTable, specifiedColor=<rgba 0 0 0 0>}
       cornerRadii=<4 4 4 4>
      }
      tintColor {colorUseRule=UseColor_Specified, specifiedColor=<rgba 0.0284 0.0284 0.0284 1>}
      uVRegion {bIsValid=false, max=<0 0>, min=<0 0>}
     }
     pressedForeground {colorUseRule=UseColor_ColorTable, specifiedColor=<rgba 0 0 0 0>}
     pressedPadding=<ltrb 12 2.5 12 0.5>
     pressedSlateSound {resourceObject=null}
    }
   }
   slot {class=HorizontalBoxSlot, properties {verticalAlignment=VAlign_Center}}
  }
  TXT_LoadLabel {
   widgetClassPath=&TextBlock
   parent=BTN_LoadSelected
   widget=TXT_LoadLabel
   properties {
    text=Load Selected SMs
    font {
     bForceMonospaced=false
     bMaterialIsStencil=false
     fontMaterial=null
     fontObject=&Roboto
     letterSpacing=0
     monospacedWidth=1
     size=18
     skewAmount=0
     typefaceFontName=Bold
     outlineSettings {
      bApplyOutlineToDropShadows=false
      bMiteredCorners=false
      bSeparateFillAlpha=false
      outlineMaterial=null
      outlineSize=0
      outlineColor=<rgba 0 0 0 1>
     }
    }
   }
   slot {class=ButtonSlot}
  }
  BTN_SaveAll {
   widgetClassPath=&Button
   parent=HeaderRow
   widget=BTN_SaveAll
   properties {
    toolTipText="Saves All Edited && Unsaved Static Meshes within the SMWrangler"
    widgetStyle {
     clickedSlateSound {resourceObject=null}
     disabled {
      bIsDynamicallyLoaded=false
      drawAs=RoundedBox
      imageType=NoImage
      mirroring=NoMirror
      resourceName=None
      resourceObject=null
      tiling=NoTile
      imageSize=<0 0>
      margin=<ltrb 0 0 0 0>
      outlineSettings {
       bUseBrushTransparency=false
       roundingType=FixedRadius
       width=1
       color {colorUseRule=UseColor_ColorTable, specifiedColor=<rgba 0 0 0 0>}
       cornerRadii=<4 4 4 4>
      }
      tintColor {colorUseRule=UseColor_Specified, specifiedColor=<rgba 0.0395 0.0395 0.0395 1>}
      uVRegion {bIsValid=false, max=<0 0>, min=<0 0>}
     }
     disabledForeground {colorUseRule=UseColor_ColorTable, specifiedColor=<rgba 0 0 0 0>}
     hovered {
      bIsDynamicallyLoaded=false
      drawAs=RoundedBox
      imageType=NoImage
      mirroring=NoMirror
      resourceName=None
      resourceObject=null
      tiling=NoTile
      imageSize=<32 32>
      margin=<ltrb 0 0 0 0>
      outlineSettings {
       bUseBrushTransparency=false
       roundingType=FixedRadius
       width=1
       color {colorUseRule=UseColor_ColorTable, specifiedColor=<rgba 0 0 0 0>}
       cornerRadii=<4 4 4 4>
      }
      tintColor {colorUseRule=UseColor_Specified, specifiedColor=<rgba 0.0953 0.0953 0.0953 1>}
      uVRegion {bIsValid=false, max=<0 0>, min=<0 0>}
     }
     hoveredForeground {colorUseRule=UseColor_ColorTable, specifiedColor=<rgba 0 0 0 0>}
     hoveredSlateSound {resourceObject=null}
     normal {
      bIsDynamicallyLoaded=false
      drawAs=RoundedBox
      imageType=NoImage
      mirroring=NoMirror
      resourceName=None
      resourceObject=null
      tiling=NoTile
      imageSize=<32 32>
      margin=<ltrb 0 0 0 0>
      outlineSettings {
       bUseBrushTransparency=false
       roundingType=FixedRadius
       width=1
       color {colorUseRule=UseColor_ColorTable, specifiedColor=<rgba 0 0 0 0>}
       cornerRadii=<4 4 4 4>
      }
      tintColor {colorUseRule=UseColor_Specified, specifiedColor=<rgba 0.0395 0.0395 0.0395 1>}
      uVRegion {bIsValid=false, max=<0 0>, min=<0 0>}
     }
     normalForeground {colorUseRule=UseColor_ColorTable, specifiedColor=<rgba 0 0 0 0>}
     normalPadding=<ltrb 12 1.5 12 1.5>
     pressed {
      bIsDynamicallyLoaded=false
      drawAs=RoundedBox
      imageType=NoImage
      mirroring=NoMirror
      resourceName=None
      resourceObject=null
      tiling=NoTile
      imageSize=<32 32>
      margin=<ltrb 0 0 0 0>
      outlineSettings {
       bUseBrushTransparency=false
       roundingType=FixedRadius
       width=1
       color {colorUseRule=UseColor_ColorTable, specifiedColor=<rgba 0 0 0 0>}
       cornerRadii=<4 4 4 4>
      }
      tintColor {colorUseRule=UseColor_Specified, specifiedColor=<rgba 0.0284 0.0284 0.0284 1>}
      uVRegion {bIsValid=false, max=<0 0>, min=<0 0>}
     }
     pressedForeground {colorUseRule=UseColor_ColorTable, specifiedColor=<rgba 0 0 0 0>}
     pressedPadding=<ltrb 12 2.5 12 0.5>
     pressedSlateSound {resourceObject=null}
    }
   }
   slot {class=HorizontalBoxSlot, properties {verticalAlignment=VAlign_Center, padding=<ltrb 8 0 0 0>}}
  }
  TXT_SaveAllLabel {
   widgetClassPath=&TextBlock
   parent=BTN_SaveAll
   widget=TXT_SaveAllLabel
   properties {
    text=Save All
    font {
     bForceMonospaced=false
     bMaterialIsStencil=false
     fontMaterial=null
     fontObject=&Roboto
     letterSpacing=0
     monospacedWidth=1
     size=18
     skewAmount=0
     typefaceFontName=Bold
     outlineSettings {
      bApplyOutlineToDropShadows=false
      bMiteredCorners=false
      bSeparateFillAlpha=false
      outlineMaterial=null
      outlineSize=0
      outlineColor=<rgba 0 0 0 1>
     }
    }
   }
   slot {class=ButtonSlot}
  }
  BTN_ExportLODs {
   widgetClassPath=&Button
   parent=HeaderRow
   widget=BTN_ExportLODs
   properties {
    toolTipText=Opens a Popup Dialog to Export out a list of Meshes with Generated LODs
    widgetStyle {
     clickedSlateSound {resourceObject=null}
     disabled {
      bIsDynamicallyLoaded=false
      drawAs=RoundedBox
      imageType=NoImage
      mirroring=NoMirror
      resourceName=None
      resourceObject=null
      tiling=NoTile
      imageSize=<0 0>
      margin=<ltrb 0 0 0 0>
      outlineSettings {
       bUseBrushTransparency=false
       roundingType=FixedRadius
       width=1
       color {colorUseRule=UseColor_ColorTable, specifiedColor=<rgba 0 0 0 0>}
       cornerRadii=<4 4 4 4>
      }
      tintColor {colorUseRule=UseColor_Specified, specifiedColor=<rgba 0.0395 0.0395 0.0395 1>}
      uVRegion {bIsValid=false, max=<0 0>, min=<0 0>}
     }
     disabledForeground {colorUseRule=UseColor_ColorTable, specifiedColor=<rgba 0 0 0 0>}
     hovered {
      bIsDynamicallyLoaded=false
      drawAs=RoundedBox
      imageType=NoImage
      mirroring=NoMirror
      resourceName=None
      resourceObject=null
      tiling=NoTile
      imageSize=<32 32>
      margin=<ltrb 0 0 0 0>
      outlineSettings {
       bUseBrushTransparency=false
       roundingType=FixedRadius
       width=1
       color {colorUseRule=UseColor_ColorTable, specifiedColor=<rgba 0 0 0 0>}
       cornerRadii=<4 4 4 4>
      }
      tintColor {colorUseRule=UseColor_Specified, specifiedColor=<rgba 0.0953 0.0953 0.0953 1>}
      uVRegion {bIsValid=false, max=<0 0>, min=<0 0>}
     }
     hoveredForeground {colorUseRule=UseColor_ColorTable, specifiedColor=<rgba 0 0 0 0>}
     hoveredSlateSound {resourceObject=null}
     normal {
      bIsDynamicallyLoaded=false
      drawAs=RoundedBox
      imageType=NoImage
      mirroring=NoMirror
      resourceName=None
      resourceObject=null
      tiling=NoTile
      imageSize=<32 32>
      margin=<ltrb 0 0 0 0>
      outlineSettings {
       bUseBrushTransparency=false
       roundingType=FixedRadius
       width=1
       color {colorUseRule=UseColor_ColorTable, specifiedColor=<rgba 0 0 0 0>}
       cornerRadii=<4 4 4 4>
      }
      tintColor {colorUseRule=UseColor_Specified, specifiedColor=<rgba 0.0395 0.0395 0.0395 1>}
      uVRegion {bIsValid=false, max=<0 0>, min=<0 0>}
     }
     normalForeground {colorUseRule=UseColor_ColorTable, specifiedColor=<rgba 0 0 0 0>}
     normalPadding=<ltrb 12 1.5 12 1.5>
     pressed {
      bIsDynamicallyLoaded=false
      drawAs=RoundedBox
      imageType=NoImage
      mirroring=NoMirror
      resourceName=None
      resourceObject=null
      tiling=NoTile
      imageSize=<32 32>
      margin=<ltrb 0 0 0 0>
      outlineSettings {
       bUseBrushTransparency=false
       roundingType=FixedRadius
       width=1
       color {colorUseRule=UseColor_ColorTable, specifiedColor=<rgba 0 0 0 0>}
       cornerRadii=<4 4 4 4>
      }
      tintColor {colorUseRule=UseColor_Specified, specifiedColor=<rgba 0.0284 0.0284 0.0284 1>}
      uVRegion {bIsValid=false, max=<0 0>, min=<0 0>}
     }
     pressedForeground {colorUseRule=UseColor_ColorTable, specifiedColor=<rgba 0 0 0 0>}
     pressedPadding=<ltrb 12 2.5 12 0.5>
     pressedSlateSound {resourceObject=null}
    }
   }
   slot {class=HorizontalBoxSlot, properties {verticalAlignment=VAlign_Center, padding=<ltrb 8 0 0 0>}}
  }
  TXT_ExportLODsLabel {
   widgetClassPath=&TextBlock
   parent=BTN_ExportLODs
   widget=TXT_ExportLODsLabel
   properties {
    text=Export LODs List
    colorAndOpacity {colorUseRule=UseColor_Specified, specifiedColor=<rgba 0.9 0.9 0.9 1>}
    font {
     bForceMonospaced=false
     bMaterialIsStencil=false
     fontMaterial=null
     fontObject=&Roboto
     letterSpacing=0
     monospacedWidth=1
     size=18
     skewAmount=0
     typefaceFontName=Bold
     outlineSettings {
      bApplyOutlineToDropShadows=false
      bMiteredCorners=false
      bSeparateFillAlpha=false
      outlineMaterial=null
      outlineSize=0
      outlineColor=<rgba 0 0 0 1>
     }
    }
   }
   slot {class=ButtonSlot}
  }
  HorizontalBox_0 {
   widgetClassPath=&HorizontalBox
   parent=RootVerticalBox
   widget=HorizontalBox_0
   slot {class=VerticalBoxSlot, properties {horizontalAlignment=HAlign_Center, verticalAlignment=VAlign_Center, padding=<ltrb 0 2 0 2>}}
  }
  TXT_Status {
   widgetClassPath=&TextBlock
   bIsVariable=true
   parent=HorizontalBox_0
   widget=TXT_Status
   properties {
    text=Select Static Mesh assets and click Load Selected
    colorAndOpacity {colorUseRule=UseColor_Specified, specifiedColor=<rgba 0.6 0.6 0.6 1>}
    font {
     bForceMonospaced=false
     bMaterialIsStencil=false
     fontMaterial=null
     fontObject=&Roboto
     letterSpacing=0
     monospacedWidth=1
     size=13
     skewAmount=0
     typefaceFontName=Bold
     outlineSettings {
      bApplyOutlineToDropShadows=false
      bMiteredCorners=false
      bSeparateFillAlpha=false
      outlineMaterial=null
      outlineSize=0
      outlineColor=<rgba 0 0 0 1>
     }
    }
   }
   slot {class=HorizontalBoxSlot, properties {verticalAlignment=VAlign_Center}}
  }
  TXT_SaveStatus {
   widgetClassPath=&TextBlock
   bIsVariable=true
   parent=HorizontalBox_0
   widget=TXT_SaveStatus
   properties {
    text=Text Block
    visibility=Collapsed
    colorAndOpacity {colorUseRule=UseColor_Specified, specifiedColor=<rgba 0.6 0.6 0.6 1>}
    font {
     bForceMonospaced=false
     bMaterialIsStencil=false
     fontMaterial=null
     fontObject=&Roboto
     letterSpacing=0
     monospacedWidth=1
     size=13
     skewAmount=0
     typefaceFontName=Bold
     outlineSettings {
      bApplyOutlineToDropShadows=false
      bMiteredCorners=false
      bSeparateFillAlpha=false
      outlineMaterial=null
      outlineSize=0
      outlineColor=<rgba 0 0 0 1>
     }
    }
   }
   slot {class=HorizontalBoxSlot, properties {verticalAlignment=VAlign_Center, padding=<ltrb 32 0 0 0>}}
  }
  HB_ColumnHeaders {
   widgetClassPath=&HorizontalBox
   bIsVariable=true
   parent=RootVerticalBox
   widget=HB_ColumnHeaders
   slot {class=VerticalBoxSlot, properties {horizontalAlignment=HAlign_Center, padding=<ltrb 0 0 0 1>}}
  }
  Border_0 {
   widgetClassPath=&Border
   parent=HB_ColumnHeaders
   widget=Border_0
   properties {
    verticalAlignment=VAlign_Center
    background {
     bIsDynamicallyLoaded=false
     drawAs=Image
     imageType=NoImage
     mirroring=NoMirror
     resourceName=None
     resourceObject=null
     tiling=NoTile
     imageSize=<32 32>
     margin=<ltrb 0 0 0 0>
     outlineSettings {
      bUseBrushTransparency=false
      roundingType=HalfHeightRadius
      width=0
      color {colorUseRule=UseColor_Specified, specifiedColor=<rgba 0 0 0 0>}
      cornerRadii=<0 0 0 0>
     }
     tintColor {colorUseRule=UseColor_Specified, specifiedColor=<rgba 0.0417 0.0417 0.0417 1>}
     uVRegion {bIsValid=false, max=<0 0>, min=<0 0>}
    }
    padding=<ltrb 4 3 4 3>
   }
   slot {class=HorizontalBoxSlot, properties {padding=<ltrb 2 2 0 2>}}
  }
  HDR_MeshName {
   widgetClassPath=&TextBlock
   parent=Border_0
   widget=HDR_MeshName
   properties {
    justification=Center
    minDesiredWidth=298
    text=Mesh Name
    colorAndOpacity {colorUseRule=UseColor_Specified, specifiedColor=<rgba 0.9 0.9 0.9 1>}
    font {
     bForceMonospaced=false
     bMaterialIsStencil=false
     fontMaterial=null
     fontObject=&Roboto
     letterSpacing=0
     monospacedWidth=1
     size=11
     skewAmount=0
     typefaceFontName=Bold
     outlineSettings {
      bApplyOutlineToDropShadows=false
      bMiteredCorners=false
      bSeparateFillAlpha=false
      outlineMaterial=null
      outlineSize=0
      outlineColor=<rgba 0 0 0 1>
     }
    }
   }
   slot {class=BorderSlot, properties {verticalAlignment=VAlign_Center, padding=<ltrb 0 0 0 0>}}
  }
  Border_1 {
   widgetClassPath=&Border
   parent=HB_ColumnHeaders
   widget=Border_1
   properties {
    verticalAlignment=VAlign_Center
    background {
     bIsDynamicallyLoaded=false
     drawAs=Image
     imageType=NoImage
     mirroring=NoMirror
     resourceName=None
     resourceObject=null
     tiling=NoTile
     imageSize=<32 32>
     margin=<ltrb 0 0 0 0>
     outlineSettings {
      bUseBrushTransparency=false
      roundingType=HalfHeightRadius
      width=0
      color {colorUseRule=UseColor_Specified, specifiedColor=<rgba 0 0 0 0>}
      cornerRadii=<0 0 0 0>
     }
     tintColor {colorUseRule=UseColor_Specified, specifiedColor=<rgba 0.0417 0.0417 0.0417 1>}
     uVRegion {bIsValid=false, max=<0 0>, min=<0 0>}
    }
    padding=<ltrb 4 3 4 3>
   }
   slot {class=HorizontalBoxSlot, properties {padding=<ltrb 4 2 0 2>}}
  }
  HDR_L0Type {
   widgetClassPath=&TextBlock
   parent=Border_1
   widget=HDR_L0Type
   properties {
    justification=Center
    minDesiredWidth=98
    text=LOD0 Type
    colorAndOpacity {colorUseRule=UseColor_Specified, specifiedColor=<rgba 0.9 0.9 0.9 1>}
    font {
     bForceMonospaced=false
     bMaterialIsStencil=false
     fontMaterial=null
     fontObject=&Roboto
     letterSpacing=0
     monospacedWidth=1
     size=11
     skewAmount=0
     typefaceFontName=Bold
     outlineSettings {
      bApplyOutlineToDropShadows=false
      bMiteredCorners=false
      bSeparateFillAlpha=false
      outlineMaterial=null
      outlineSize=0
      outlineColor=<rgba 0 0 0 1>
     }
    }
   }
   slot {class=BorderSlot, properties {verticalAlignment=VAlign_Center, padding=<ltrb 0 0 0 0>}}
  }
  Border_2 {
   widgetClassPath=&Border
   parent=HB_ColumnHeaders
   widget=Border_2
   properties {
    verticalAlignment=VAlign_Center
    background {
     bIsDynamicallyLoaded=false
     drawAs=Image
     imageType=NoImage
     mirroring=NoMirror
     resourceName=None
     resourceObject=null
     tiling=NoTile
     imageSize=<32 32>
     margin=<ltrb 0 0 0 0>
     outlineSettings {
      bUseBrushTransparency=false
      roundingType=HalfHeightRadius
      width=0
      color {colorUseRule=UseColor_Specified, specifiedColor=<rgba 0 0 0 0>}
      cornerRadii=<0 0 0 0>
     }
     tintColor {colorUseRule=UseColor_Specified, specifiedColor=<rgba 0.0417 0.0417 0.0417 1>}
     uVRegion {bIsValid=false, max=<0 0>, min=<0 0>}
    }
    padding=<ltrb 4 3 4 3>
   }
   slot {class=HorizontalBoxSlot, properties {padding=<ltrb 2 2 0 2>}}
  }
  HDR_L0Verts {
   widgetClassPath=&TextBlock
   parent=Border_2
   widget=HDR_L0Verts
   properties {
    justification=Center
    minDesiredWidth=78
    text=Verts
    colorAndOpacity {colorUseRule=UseColor_Specified, specifiedColor=<rgba 0.9 0.9 0.9 1>}
    font {
     bForceMonospaced=false
     bMaterialIsStencil=false
     fontMaterial=null
     fontObject=&Roboto
     letterSpacing=0
     monospacedWidth=1
     size=11
     skewAmount=0
     typefaceFontName=Bold
     outlineSettings {
      bApplyOutlineToDropShadows=false
      bMiteredCorners=false
      bSeparateFillAlpha=false
      outlineMaterial=null
      outlineSize=0
      outlineColor=<rgba 0 0 0 1>
     }
    }
   }
   slot {class=BorderSlot, properties {verticalAlignment=VAlign_Center, padding=<ltrb 0 0 0 0>}}
  }
  Border_3 {
   widgetClassPath=&Border
   parent=HB_ColumnHeaders
   widget=Border_3
   properties {
    verticalAlignment=VAlign_Center
    background {
     bIsDynamicallyLoaded=false
     drawAs=Image
     imageType=NoImage
     mirroring=NoMirror
     resourceName=None
     resourceObject=null
     tiling=NoTile
     imageSize=<32 32>
     margin=<ltrb 0 0 0 0>
     outlineSettings {
      bUseBrushTransparency=false
      roundingType=HalfHeightRadius
      width=0
      color {colorUseRule=UseColor_Specified, specifiedColor=<rgba 0 0 0 0>}
      cornerRadii=<0 0 0 0>
     }
     tintColor {colorUseRule=UseColor_Specified, specifiedColor=<rgba 0.0417 0.0417 0.0417 1>}
     uVRegion {bIsValid=false, max=<0 0>, min=<0 0>}
    }
    padding=<ltrb 4 3 4 3>
   }
   slot {class=HorizontalBoxSlot, properties {padding=<ltrb 2 2 0 2>}}
  }
  HDR_L0Screen {
   widgetClassPath=&TextBlock
   parent=Border_3
   widget=HDR_L0Screen
   properties {
    justification=Center
    minDesiredWidth=108
    text=Screen Size
    colorAndOpacity {colorUseRule=UseColor_Specified, specifiedColor=<rgba 0.9 0.9 0.9 1>}
    font {
     bForceMonospaced=false
     bMaterialIsStencil=false
     fontMaterial=null
     fontObject=&Roboto
     letterSpacing=0
     monospacedWidth=1
     size=11
     skewAmount=0
     typefaceFontName=Bold
     outlineSettings {
      bApplyOutlineToDropShadows=false
      bMiteredCorners=false
      bSeparateFillAlpha=false
      outlineMaterial=null
      outlineSize=0
      outlineColor=<rgba 0 0 0 1>
     }
    }
   }
   slot {class=BorderSlot, properties {verticalAlignment=VAlign_Center, padding=<ltrb 0 0 0 0>}}
  }
  Border_4 {
   widgetClassPath=&Border
   parent=HB_ColumnHeaders
   widget=Border_4
   properties {
    verticalAlignment=VAlign_Center
    background {
     bIsDynamicallyLoaded=false
     drawAs=Image
     imageType=NoImage
     mirroring=NoMirror
     resourceName=None
     resourceObject=null
     tiling=NoTile
     imageSize=<32 32>
     margin=<ltrb 0 0 0 0>
     outlineSettings {
      bUseBrushTransparency=false
      roundingType=HalfHeightRadius
      width=0
      color {colorUseRule=UseColor_Specified, specifiedColor=<rgba 0 0 0 0>}
      cornerRadii=<0 0 0 0>
     }
     tintColor {colorUseRule=UseColor_Specified, specifiedColor=<rgba 0.0417 0.0417 0.0417 1>}
     uVRegion {bIsValid=false, max=<0 0>, min=<0 0>}
    }
    padding=<ltrb 4 3 4 3>
   }
   slot {class=HorizontalBoxSlot, properties {padding=<ltrb 4 2 0 2>}}
  }
  HDR_L1Type {
   widgetClassPath=&TextBlock
   parent=Border_4
   widget=HDR_L1Type
   properties {
    justification=Center
    minDesiredWidth=98
    text=LOD1 Type
    colorAndOpacity {colorUseRule=UseColor_Specified, specifiedColor=<rgba 0.9 0.9 0.9 1>}
    font {
     bForceMonospaced=false
     bMaterialIsStencil=false
     fontMaterial=null
     fontObject=&Roboto
     letterSpacing=0
     monospacedWidth=1
     size=11
     skewAmount=0
     typefaceFontName=Bold
     outlineSettings {
      bApplyOutlineToDropShadows=false
      bMiteredCorners=false
      bSeparateFillAlpha=false
      outlineMaterial=null
      outlineSize=0
      outlineColor=<rgba 0 0 0 1>
     }
    }
   }
   slot {class=BorderSlot, properties {verticalAlignment=VAlign_Center, padding=<ltrb 0 0 0 0>}}
  }
  Border_5 {
   widgetClassPath=&Border
   parent=HB_ColumnHeaders
   widget=Border_5
   properties {
    verticalAlignment=VAlign_Center
    background {
     bIsDynamicallyLoaded=false
     drawAs=Image
     imageType=NoImage
     mirroring=NoMirror
     resourceName=None
     resourceObject=null
     tiling=NoTile
     imageSize=<32 32>
     margin=<ltrb 0 0 0 0>
     outlineSettings {
      bUseBrushTransparency=false
      roundingType=HalfHeightRadius
      width=0
      color {colorUseRule=UseColor_Specified, specifiedColor=<rgba 0 0 0 0>}
      cornerRadii=<0 0 0 0>
     }
     tintColor {colorUseRule=UseColor_Specified, specifiedColor=<rgba 0.0417 0.0417 0.0417 1>}
     uVRegion {bIsValid=false, max=<0 0>, min=<0 0>}
    }
    padding=<ltrb 4 3 4 3>
   }
   slot {class=HorizontalBoxSlot, properties {padding=<ltrb 2 2 0 2>}}
  }
  HDR_L1Verts {
   widgetClassPath=&TextBlock
   parent=Border_5
   widget=HDR_L1Verts
   properties {
    justification=Center
    minDesiredWidth=78
    text=Verts
    colorAndOpacity {colorUseRule=UseColor_Specified, specifiedColor=<rgba 0.9 0.9 0.9 1>}
    font {
     bForceMonospaced=false
     bMaterialIsStencil=false
     fontMaterial=null
     fontObject=&Roboto
     letterSpacing=0
     monospacedWidth=1
     size=11
     skewAmount=0
     typefaceFontName=Bold
     outlineSettings {
      bApplyOutlineToDropShadows=false
      bMiteredCorners=false
      bSeparateFillAlpha=false
      outlineMaterial=null
      outlineSize=0
      outlineColor=<rgba 0 0 0 1>
     }
    }
   }
   slot {class=BorderSlot, properties {verticalAlignment=VAlign_Center, padding=<ltrb 0 0 0 0>}}
  }
  Border_6 {
   widgetClassPath=&Border
   parent=HB_ColumnHeaders
   widget=Border_6
   properties {
    verticalAlignment=VAlign_Center
    background {
     bIsDynamicallyLoaded=false
     drawAs=Image
     imageType=NoImage
     mirroring=NoMirror
     resourceName=None
     resourceObject=null
     tiling=NoTile
     imageSize=<32 32>
     margin=<ltrb 0 0 0 0>
     outlineSettings {
      bUseBrushTransparency=false
      roundingType=HalfHeightRadius
      width=0
      color {colorUseRule=UseColor_Specified, specifiedColor=<rgba 0 0 0 0>}
      cornerRadii=<0 0 0 0>
     }
     tintColor {colorUseRule=UseColor_Specified, specifiedColor=<rgba 0.0417 0.0417 0.0417 1>}
     uVRegion {bIsValid=false, max=<0 0>, min=<0 0>}
    }
    padding=<ltrb 4 3 4 3>
   }
   slot {class=HorizontalBoxSlot, properties {padding=<ltrb 2 2 0 2>}}
  }
  HDR_L1Screen {
   widgetClassPath=&TextBlock
   parent=Border_6
   widget=HDR_L1Screen
   properties {
    justification=Center
    minDesiredWidth=108
    text=Screen Size
    colorAndOpacity {colorUseRule=UseColor_Specified, specifiedColor=<rgba 0.9 0.9 0.9 1>}
    font {
     bForceMonospaced=false
     bMaterialIsStencil=false
     fontMaterial=null
     fontObject=&Roboto
     letterSpacing=0
     monospacedWidth=1
     size=11
     skewAmount=0
     typefaceFontName=Bold
     outlineSettings {
      bApplyOutlineToDropShadows=false
      bMiteredCorners=false
      bSeparateFillAlpha=false
      outlineMaterial=null
      outlineSize=0
      outlineColor=<rgba 0 0 0 1>
     }
    }
   }
   slot {class=BorderSlot, properties {verticalAlignment=VAlign_Center, padding=<ltrb 0 0 0 0>}}
  }
  Border_7 {
   widgetClassPath=&Border
   parent=HB_ColumnHeaders
   widget=Border_7
   properties {
    verticalAlignment=VAlign_Center
    background {
     bIsDynamicallyLoaded=false
     drawAs=Image
     imageType=NoImage
     mirroring=NoMirror
     resourceName=None
     resourceObject=null
     tiling=NoTile
     imageSize=<32 32>
     margin=<ltrb 0 0 0 0>
     outlineSettings {
      bUseBrushTransparency=false
      roundingType=HalfHeightRadius
      width=0
      color {colorUseRule=UseColor_Specified, specifiedColor=<rgba 0 0 0 0>}
      cornerRadii=<0 0 0 0>
     }
     tintColor {colorUseRule=UseColor_Specified, specifiedColor=<rgba 0.0417 0.0417 0.0417 1>}
     uVRegion {bIsValid=false, max=<0 0>, min=<0 0>}
    }
    padding=<ltrb 4 3 4 3>
   }
   slot {class=HorizontalBoxSlot, properties {padding=<ltrb 4 2 0 2>}}
  }
  HDR_L2Type {
   widgetClassPath=&TextBlock
   parent=Border_7
   widget=HDR_L2Type
   properties {
    justification=Center
    minDesiredWidth=98
    text=LOD2 Type
    colorAndOpacity {colorUseRule=UseColor_Specified, specifiedColor=<rgba 0.9 0.9 0.9 1>}
    font {
     bForceMonospaced=false
     bMaterialIsStencil=false
     fontMaterial=null
     fontObject=&Roboto
     letterSpacing=0
     monospacedWidth=1
     size=11
     skewAmount=0
     typefaceFontName=Bold
     outlineSettings {
      bApplyOutlineToDropShadows=false
      bMiteredCorners=false
      bSeparateFillAlpha=false
      outlineMaterial=null
      outlineSize=0
      outlineColor=<rgba 0 0 0 1>
     }
    }
   }
   slot {class=BorderSlot, properties {verticalAlignment=VAlign_Center, padding=<ltrb 0 0 0 0>}}
  }
  Border_8 {
   widgetClassPath=&Border
   parent=HB_ColumnHeaders
   widget=Border_8
   properties {
    verticalAlignment=VAlign_Center
    background {
     bIsDynamicallyLoaded=false
     drawAs=Image
     imageType=NoImage
     mirroring=NoMirror
     resourceName=None
     resourceObject=null
     tiling=NoTile
     imageSize=<32 32>
     margin=<ltrb 0 0 0 0>
     outlineSettings {
      bUseBrushTransparency=false
      roundingType=HalfHeightRadius
      width=0
      color {colorUseRule=UseColor_Specified, specifiedColor=<rgba 0 0 0 0>}
      cornerRadii=<0 0 0 0>
     }
     tintColor {colorUseRule=UseColor_Specified, specifiedColor=<rgba 0.0417 0.0417 0.0417 1>}
     uVRegion {bIsValid=false, max=<0 0>, min=<0 0>}
    }
    padding=<ltrb 4 3 4 3>
   }
   slot {class=HorizontalBoxSlot, properties {padding=<ltrb 2 2 0 2>}}
  }
  HDR_L2Verts {
   widgetClassPath=&TextBlock
   parent=Border_8
   widget=HDR_L2Verts
   properties {
    justification=Center
    minDesiredWidth=78
    text=Verts
    colorAndOpacity {colorUseRule=UseColor_Specified, specifiedColor=<rgba 0.9 0.9 0.9 1>}
    font {
     bForceMonospaced=false
     bMaterialIsStencil=false
     fontMaterial=null
     fontObject=&Roboto
     letterSpacing=0
     monospacedWidth=1
     size=11
     skewAmount=0
     typefaceFontName=Bold
     outlineSettings {
      bApplyOutlineToDropShadows=false
      bMiteredCorners=false
      bSeparateFillAlpha=false
      outlineMaterial=null
      outlineSize=0
      outlineColor=<rgba 0 0 0 1>
     }
    }
   }
   slot {class=BorderSlot, properties {verticalAlignment=VAlign_Center, padding=<ltrb 0 0 0 0>}}
  }
  Border_9 {
   widgetClassPath=&Border
   parent=HB_ColumnHeaders
   widget=Border_9
   properties {
    verticalAlignment=VAlign_Center
    background {
     bIsDynamicallyLoaded=false
     drawAs=Image
     imageType=NoImage
     mirroring=NoMirror
     resourceName=None
     resourceObject=null
     tiling=NoTile
     imageSize=<32 32>
     margin=<ltrb 0 0 0 0>
     outlineSettings {
      bUseBrushTransparency=false
      roundingType=HalfHeightRadius
      width=0
      color {colorUseRule=UseColor_Specified, specifiedColor=<rgba 0 0 0 0>}
      cornerRadii=<0 0 0 0>
     }
     tintColor {colorUseRule=UseColor_Specified, specifiedColor=<rgba 0.0417 0.0417 0.0417 1>}
     uVRegion {bIsValid=false, max=<0 0>, min=<0 0>}
    }
    padding=<ltrb 4 3 4 3>
   }
   slot {class=HorizontalBoxSlot, properties {padding=<ltrb 2 2 0 2>}}
  }
  HDR_L2Screen {
   widgetClassPath=&TextBlock
   parent=Border_9
   widget=HDR_L2Screen
   properties {
    justification=Center
    minDesiredWidth=108
    text=Screen Size
    colorAndOpacity {colorUseRule=UseColor_Specified, specifiedColor=<rgba 0.9 0.9 0.9 1>}
    font {
     bForceMonospaced=false
     bMaterialIsStencil=false
     fontMaterial=null
     fontObject=&Roboto
     letterSpacing=0
     monospacedWidth=1
     size=11
     skewAmount=0
     typefaceFontName=Bold
     outlineSettings {
      bApplyOutlineToDropShadows=false
      bMiteredCorners=false
      bSeparateFillAlpha=false
      outlineMaterial=null
      outlineSize=0
      outlineColor=<rgba 0 0 0 1>
     }
    }
   }
   slot {class=BorderSlot, properties {verticalAlignment=VAlign_Center, padding=<ltrb 0 0 0 0>}}
  }
  Border_10 {
   widgetClassPath=&Border
   parent=HB_ColumnHeaders
   widget=Border_10
   properties {
    verticalAlignment=VAlign_Center
    background {
     bIsDynamicallyLoaded=false
     drawAs=Image
     imageType=NoImage
     mirroring=NoMirror
     resourceName=None
     resourceObject=null
     tiling=NoTile
     imageSize=<32 32>
     margin=<ltrb 0 0 0 0>
     outlineSettings {
      bUseBrushTransparency=false
      roundingType=HalfHeightRadius
      width=0
      color {colorUseRule=UseColor_Specified, specifiedColor=<rgba 0 0 0 0>}
      cornerRadii=<0 0 0 0>
     }
     tintColor {colorUseRule=UseColor_Specified, specifiedColor=<rgba 0.0417 0.0417 0.0417 1>}
     uVRegion {bIsValid=false, max=<0 0>, min=<0 0>}
    }
    padding=<ltrb 4 3 4 3>
   }
   slot {class=HorizontalBoxSlot, properties {padding=<ltrb 4 2 0 2>}}
  }
  HDR_L3Type {
   widgetClassPath=&TextBlock
   parent=Border_10
   widget=HDR_L3Type
   properties {
    justification=Center
    minDesiredWidth=98
    text=LOD3 Type
    colorAndOpacity {colorUseRule=UseColor_Specified, specifiedColor=<rgba 0.9 0.9 0.9 1>}
    font {
     bForceMonospaced=false
     bMaterialIsStencil=false
     fontMaterial=null
     fontObject=&Roboto
     letterSpacing=0
     monospacedWidth=1
     size=11
     skewAmount=0
     typefaceFontName=Bold
     outlineSettings {
      bApplyOutlineToDropShadows=false
      bMiteredCorners=false
      bSeparateFillAlpha=false
      outlineMaterial=null
      outlineSize=0
      outlineColor=<rgba 0 0 0 1>
     }
    }
   }
   slot {class=BorderSlot, properties {verticalAlignment=VAlign_Center, padding=<ltrb 0 0 0 0>}}
  }
  Border_11 {
   widgetClassPath=&Border
   parent=HB_ColumnHeaders
   widget=Border_11
   properties {
    verticalAlignment=VAlign_Center
    background {
     bIsDynamicallyLoaded=false
     drawAs=Image
     imageType=NoImage
     mirroring=NoMirror
     resourceName=None
     resourceObject=null
     tiling=NoTile
     imageSize=<32 32>
     margin=<ltrb 0 0 0 0>
     outlineSettings {
      bUseBrushTransparency=false
      roundingType=HalfHeightRadius
      width=0
      color {colorUseRule=UseColor_Specified, specifiedColor=<rgba 0 0 0 0>}
      cornerRadii=<0 0 0 0>
     }
     tintColor {colorUseRule=UseColor_Specified, specifiedColor=<rgba 0.0417 0.0417 0.0417 1>}
     uVRegion {bIsValid=false, max=<0 0>, min=<0 0>}
    }
    padding=<ltrb 4 3 4 3>
   }
   slot {class=HorizontalBoxSlot, properties {padding=<ltrb 2 2 0 2>}}
  }
  HDR_L3Verts {
   widgetClassPath=&TextBlock
   parent=Border_11
   widget=HDR_L3Verts
   properties {
    justification=Center
    minDesiredWidth=78
    text=Verts
    colorAndOpacity {colorUseRule=UseColor_Specified, specifiedColor=<rgba 0.9 0.9 0.9 1>}
    font {
     bForceMonospaced=false
     bMaterialIsStencil=false
     fontMaterial=null
     fontObject=&Roboto
     letterSpacing=0
     monospacedWidth=1
     size=11
     skewAmount=0
     typefaceFontName=Bold
     outlineSettings {
      bApplyOutlineToDropShadows=false
      bMiteredCorners=false
      bSeparateFillAlpha=false
      outlineMaterial=null
      outlineSize=0
      outlineColor=<rgba 0 0 0 1>
     }
    }
   }
   slot {class=BorderSlot, properties {verticalAlignment=VAlign_Center, padding=<ltrb 0 0 0 0>}}
  }
  Border_12 {
   widgetClassPath=&Border
   parent=HB_ColumnHeaders
   widget=Border_12
   properties {
    verticalAlignment=VAlign_Center
    background {
     bIsDynamicallyLoaded=false
     drawAs=Image
     imageType=NoImage
     mirroring=NoMirror
     resourceName=None
     resourceObject=null
     tiling=NoTile
     imageSize=<32 32>
     margin=<ltrb 0 0 0 0>
     outlineSettings {
      bUseBrushTransparency=false
      roundingType=HalfHeightRadius
      width=0
      color {colorUseRule=UseColor_Specified, specifiedColor=<rgba 0 0 0 0>}
      cornerRadii=<0 0 0 0>
     }
     tintColor {colorUseRule=UseColor_Specified, specifiedColor=<rgba 0.0417 0.0417 0.0417 1>}
     uVRegion {bIsValid=false, max=<0 0>, min=<0 0>}
    }
    padding=<ltrb 4 3 4 3>
   }
   slot {class=HorizontalBoxSlot, properties {padding=<ltrb 2 2 0 2>}}
  }
  HDR_L3Screen {
   widgetClassPath=&TextBlock
   parent=Border_12
   widget=HDR_L3Screen
   properties {
    justification=Center
    minDesiredWidth=108
    text=Screen Size
    colorAndOpacity {colorUseRule=UseColor_Specified, specifiedColor=<rgba 0.9 0.9 0.9 1>}
    font {
     bForceMonospaced=false
     bMaterialIsStencil=false
     fontMaterial=null
     fontObject=&Roboto
     letterSpacing=0
     monospacedWidth=1
     size=11
     skewAmount=0
     typefaceFontName=Bold
     outlineSettings {
      bApplyOutlineToDropShadows=false
      bMiteredCorners=false
      bSeparateFillAlpha=false
      outlineMaterial=null
      outlineSize=0
      outlineColor=<rgba 0 0 0 1>
     }
    }
   }
   slot {class=BorderSlot, properties {verticalAlignment=VAlign_Center, padding=<ltrb 0 0 0 0>}}
  }
  Border_13 {
   widgetClassPath=&Border
   parent=HB_ColumnHeaders
   widget=Border_13
   properties {
    verticalAlignment=VAlign_Center
    background {
     bIsDynamicallyLoaded=false
     drawAs=Image
     imageType=NoImage
     mirroring=NoMirror
     resourceName=None
     resourceObject=null
     tiling=NoTile
     imageSize=<32 32>
     margin=<ltrb 0 0 0 0>
     outlineSettings {
      bUseBrushTransparency=false
      roundingType=HalfHeightRadius
      width=0
      color {colorUseRule=UseColor_Specified, specifiedColor=<rgba 0 0 0 0>}
      cornerRadii=<0 0 0 0>
     }
     tintColor {colorUseRule=UseColor_Specified, specifiedColor=<rgba 0.0417 0.0417 0.0417 1>}
     uVRegion {bIsValid=false, max=<0 0>, min=<0 0>}
    }
    padding=<ltrb 4 3 4 3>
   }
   slot {class=HorizontalBoxSlot, properties {padding=<ltrb 2 2 0 2>}}
  }
  HDR_Save {
   widgetClassPath=&TextBlock
   parent=Border_13
   widget=HDR_Save
   properties {
    justification=Center
    minDesiredWidth=50
    text=Save
    colorAndOpacity {colorUseRule=UseColor_Specified, specifiedColor=<rgba 0.9 0.9 0.9 1>}
    font {
     bForceMonospaced=false
     bMaterialIsStencil=false
     fontMaterial=null
     fontObject=&Roboto
     letterSpacing=0
     monospacedWidth=1
     size=11
     skewAmount=0
     typefaceFontName=Bold
     outlineSettings {
      bApplyOutlineToDropShadows=false
      bMiteredCorners=false
      bSeparateFillAlpha=false
      outlineMaterial=null
      outlineSize=0
      outlineColor=<rgba 0 0 0 1>
     }
    }
   }
   slot {class=BorderSlot, properties {verticalAlignment=VAlign_Center, padding=<ltrb 0 0 0 0>}}
  }
  SB_MeshList {
   widgetClassPath=&ScrollBox
   bIsVariable=true
   parent=RootVerticalBox
   widget=SB_MeshList
   properties {
    widgetBarStyle {
     thickness=8
     draggedThumbImage {
      bIsDynamicallyLoaded=false
      drawAs=RoundedBox
      imageType=NoImage
      mirroring=NoMirror
      resourceName=None
      resourceObject=null
      tiling=NoTile
      imageSize=<0 0>
      margin=<ltrb 0 0 0 0>
      outlineSettings {
       bUseBrushTransparency=false
       roundingType=FixedRadius
       width=0
       color {colorUseRule=UseColor_Specified, specifiedColor=<rgba 0 0 0 0>}
       cornerRadii=<4 4 4 4>
      }
      tintColor {colorUseRule=UseColor_Specified, specifiedColor=<rgba 0.2159 0.2159 0.2159 1>}
      uVRegion {bIsValid=false, max=<0 0>, min=<0 0>}
     }
     horizontalBackgroundImage {
      bIsDynamicallyLoaded=false
      drawAs=NoDrawType
      imageType=NoImage
      mirroring=NoMirror
      resourceName=None
      resourceObject=null
      tiling=NoTile
      imageSize=<0 0>
      margin=<ltrb 0 0 0 0>
      outlineSettings {
       bUseBrushTransparency=false
       roundingType=HalfHeightRadius
       width=0
       color {colorUseRule=UseColor_Specified, specifiedColor=<rgba 0 0 0 0>}
       cornerRadii=<0 0 0 0>
      }
      tintColor {colorUseRule=UseColor_Specified, specifiedColor=<rgba 1 1 1 1>}
      uVRegion {bIsValid=false, max=<0 0>, min=<0 0>}
     }
     horizontalBottomSlotImage {
      bIsDynamicallyLoaded=false
      drawAs=NoDrawType
      imageType=NoImage
      mirroring=NoMirror
      resourceName=None
      resourceObject=null
      tiling=NoTile
      imageSize=<0 0>
      margin=<ltrb 0 0 0 0>
      outlineSettings {
       bUseBrushTransparency=false
       roundingType=HalfHeightRadius
       width=0
       color {colorUseRule=UseColor_Specified, specifiedColor=<rgba 0 0 0 0>}
       cornerRadii=<0 0 0 0>
      }
      tintColor {colorUseRule=UseColor_Specified, specifiedColor=<rgba 1 1 1 1>}
      uVRegion {bIsValid=false, max=<0 0>, min=<0 0>}
     }
     horizontalTopSlotImage {
      bIsDynamicallyLoaded=false
      drawAs=NoDrawType
      imageType=NoImage
      mirroring=NoMirror
      resourceName=None
      resourceObject=null
      tiling=NoTile
      imageSize=<0 0>
      margin=<ltrb 0 0 0 0>
      outlineSettings {
       bUseBrushTransparency=false
       roundingType=HalfHeightRadius
       width=0
       color {colorUseRule=UseColor_Specified, specifiedColor=<rgba 0 0 0 0>}
       cornerRadii=<0 0 0 0>
      }
      tintColor {colorUseRule=UseColor_Specified, specifiedColor=<rgba 1 1 1 1>}
      uVRegion {bIsValid=false, max=<0 0>, min=<0 0>}
     }
     hoveredThumbImage {
      bIsDynamicallyLoaded=false
      drawAs=RoundedBox
      imageType=NoImage
      mirroring=NoMirror
      resourceName=None
      resourceObject=null
      tiling=NoTile
      imageSize=<0 0>
      margin=<ltrb 0 0 0 0>
      outlineSettings {
       bUseBrushTransparency=false
       roundingType=FixedRadius
       width=0
       color {colorUseRule=UseColor_Specified, specifiedColor=<rgba 0 0 0 0>}
       cornerRadii=<4 4 4 4>
      }
      tintColor {colorUseRule=UseColor_Specified, specifiedColor=<rgba 0.2159 0.2159 0.2159 1>}
      uVRegion {bIsValid=false, max=<0 0>, min=<0 0>}
     }
     normalThumbImage {
      bIsDynamicallyLoaded=false
      drawAs=RoundedBox
      imageType=NoImage
      mirroring=NoMirror
      resourceName=None
      resourceObject=null
      tiling=NoTile
      imageSize=<0 0>
      margin=<ltrb 0 0 0 0>
      outlineSettings {
       bUseBrushTransparency=false
       roundingType=FixedRadius
       width=0
       color {colorUseRule=UseColor_Specified, specifiedColor=<rgba 0 0 0 0>}
       cornerRadii=<4 4 4 4>
      }
      tintColor {colorUseRule=UseColor_Specified, specifiedColor=<rgba 0.0953 0.0953 0.0953 1>}
      uVRegion {bIsValid=false, max=<0 0>, min=<0 0>}
     }
     verticalBackgroundImage {
      bIsDynamicallyLoaded=false
      drawAs=NoDrawType
      imageType=NoImage
      mirroring=NoMirror
      resourceName=None
      resourceObject=null
      tiling=NoTile
      imageSize=<0 0>
      margin=<ltrb 0 0 0 0>
      outlineSettings {
       bUseBrushTransparency=false
       roundingType=HalfHeightRadius
       width=0
       color {colorUseRule=UseColor_Specified, specifiedColor=<rgba 0 0 0 0>}
       cornerRadii=<0 0 0 0>
      }
      tintColor {colorUseRule=UseColor_Specified, specifiedColor=<rgba 1 1 1 1>}
      uVRegion {bIsValid=false, max=<0 0>, min=<0 0>}
     }
     verticalBottomSlotImage {
      bIsDynamicallyLoaded=false
      drawAs=NoDrawType
      imageType=NoImage
      mirroring=NoMirror
      resourceName=None
      resourceObject=null
      tiling=NoTile
      imageSize=<0 0>
      margin=<ltrb 0 0 0 0>
      outlineSettings {
       bUseBrushTransparency=false
       roundingType=HalfHeightRadius
       width=0
       color {colorUseRule=UseColor_Specified, specifiedColor=<rgba 0 0 0 0>}
       cornerRadii=<0 0 0 0>
      }
      tintColor {colorUseRule=UseColor_Specified, specifiedColor=<rgba 1 1 1 1>}
      uVRegion {bIsValid=false, max=<0 0>, min=<0 0>}
     }
     verticalTopSlotImage {
      bIsDynamicallyLoaded=false
      drawAs=NoDrawType
      imageType=NoImage
      mirroring=NoMirror
      resourceName=None
      resourceObject=null
      tiling=NoTile
      imageSize=<0 0>
      margin=<ltrb 0 0 0 0>
      outlineSettings {
       bUseBrushTransparency=false
       roundingType=HalfHeightRadius
       width=0
       color {colorUseRule=UseColor_Specified, specifiedColor=<rgba 0 0 0 0>}
       cornerRadii=<0 0 0 0>
      }
      tintColor {colorUseRule=UseColor_Specified, specifiedColor=<rgba 1 1 1 1>}
      uVRegion {bIsValid=false, max=<0 0>, min=<0 0>}
     }
    }
    widgetStyle {
     barThickness=8
     bottomShadowBrush {
      bIsDynamicallyLoaded=false
      drawAs=Box
      imageType=FullColor
      mirroring=NoMirror
      resourceName=../../../Engine/Content/Slate/Common/ScrollBoxShadowBottom.png
      resourceObject=null
      tiling=NoTile
      imageSize=<16 8>
      margin=<ltrb 0.5 0 0.5 1>
      outlineSettings {
       bUseBrushTransparency=false
       roundingType=HalfHeightRadius
       width=0
       color {colorUseRule=UseColor_Specified, specifiedColor=<rgba 0 0 0 0>}
       cornerRadii=<0 0 0 0>
      }
      tintColor {colorUseRule=UseColor_Specified, specifiedColor=<rgba 1 1 1 1>}
      uVRegion {bIsValid=false, max=<0 0>, min=<0 0>}
     }
     horizontalScrolledContentPadding=<ltrb 0 0 1 0>
     leftShadowBrush {
      bIsDynamicallyLoaded=false
      drawAs=Box
      imageType=FullColor
      mirroring=NoMirror
      resourceName=../../../Engine/Content/Slate/Common/ScrollBoxShadowLeft.png
      resourceObject=null
      tiling=NoTile
      imageSize=<8 16>
      margin=<ltrb 1 0.5 0 0.5>
      outlineSettings {
       bUseBrushTransparency=false
       roundingType=HalfHeightRadius
       width=0
       color {colorUseRule=UseColor_Specified, specifiedColor=<rgba 0 0 0 0>}
       cornerRadii=<0 0 0 0>
      }
      tintColor {colorUseRule=UseColor_Specified, specifiedColor=<rgba 1 1 1 1>}
      uVRegion {bIsValid=false, max=<0 0>, min=<0 0>}
     }
     rightShadowBrush {
      bIsDynamicallyLoaded=false
      drawAs=Box
      imageType=FullColor
      mirroring=NoMirror
      resourceName=../../../Engine/Content/Slate/Common/ScrollBoxShadowRight.png
      resourceObject=null
      tiling=NoTile
      imageSize=<8 16>
      margin=<ltrb 0 0.5 1 0.5>
      outlineSettings {
       bUseBrushTransparency=false
       roundingType=HalfHeightRadius
       width=0
       color {colorUseRule=UseColor_Specified, specifiedColor=<rgba 0 0 0 0>}
       cornerRadii=<0 0 0 0>
      }
      tintColor {colorUseRule=UseColor_Specified, specifiedColor=<rgba 1 1 1 1>}
      uVRegion {bIsValid=false, max=<0 0>, min=<0 0>}
     }
     topShadowBrush {
      bIsDynamicallyLoaded=false
      drawAs=Box
      imageType=FullColor
      mirroring=NoMirror
      resourceName=../../../Engine/Content/Slate/Common/ScrollBoxShadowTop.png
      resourceObject=null
      tiling=NoTile
      imageSize=<16 8>
      margin=<ltrb 0.5 1 0.5 0>
      outlineSettings {
       bUseBrushTransparency=false
       roundingType=HalfHeightRadius
       width=0
       color {colorUseRule=UseColor_Specified, specifiedColor=<rgba 0 0 0 0>}
       cornerRadii=<0 0 0 0>
      }
      tintColor {colorUseRule=UseColor_Specified, specifiedColor=<rgba 1 1 1 1>}
      uVRegion {bIsValid=false, max=<0 0>, min=<0 0>}
     }
     verticalScrolledContentPadding=<ltrb 0 0 0 1>
    }
   }
   slot {class=VerticalBoxSlot, properties {horizontalAlignment=HAlign_Center, size {sizeRule=Fill, value=1}}}
  }
 ]
}
